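# Material function for the OmelettePop character body that animates blinking on the chick eyes texture, outputting color and an alpha Mask.
# Two time-driven Frac loops at different rates combine through Min and a threshold, producing brief, irregularly spaced blinks without a noise texture.
# The result switches a two-row FlipBook between frames, using UV channel 1 adjusted by eye scale and offset parameters.
# The clamped FlipBook UVs sample the eye texture.
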MaterialFunction /BRCosmetics/Characters/Player/Female/Medium/Bodies/F_MED_OmelettePop/Materials/MF_OmelettePop_Blink
graph {expressions {
  MaterialExpressionAdd_0 {
   class=MaterialExpressionAdd
   inputs [
    A {from=MaterialExpressionMaterialFunctionCall_2, output=UVs}
    B {from=MaterialExpressionConvert_0}
   ]
  }
  MaterialExpressionClamp_0 {
   class=MaterialExpressionClamp
   inputs [
    None {from=MaterialExpressionMaterialFunctionCall_6, output=UVs}
    Min {from="", output=""}
    Max {from="", output=""}
   ]
   properties {maxDefault=2}
  }
  MaterialExpressionConstant_0 {class=MaterialExpressionConstant, properties {r=2}}
  MaterialExpressionConstant_1 {class=MaterialExpressionConstant, properties {r=1}}
  MaterialExpressionConstant_2 {class=MaterialExpressionConstant, properties {r=1}}
  MaterialExpressionConstant_3 {class=MaterialExpressionConstant}
  MaterialExpressionConstant_4 {class=MaterialExpressionConstant, properties {r=0.5}}
  MaterialExpressionConstant_6 {class=MaterialExpressionConstant, properties {r=0.1965}}
  MaterialExpressionConstant_7 {class=MaterialExpressionConstant, properties {r=0.3}}
  MaterialExpressionConstant_8 {class=MaterialExpressionConstant, properties {r=0.06}}
  MaterialExpressionConvert_0 {
   class=MaterialExpressionConvert
   inputs [
    None {from=MaterialExpressionScalarParameter_3}
    None {from=MaterialExpressionScalarParameter_4}
   ]
   properties {
    nodeName=Make Float2
    convertInputs [
     {
      type=Scalar
      expressionInput {
       expression=&self:MaterialExpressionScalarParameter_3
       inputName=None
       mask=0
       maskA=0
       maskB=0
       maskG=0
       maskR=0
       outputIndex=0
      }
     }
     {
      type=Scalar
      expressionInput {
       expression=&self:MaterialExpressionScalarParameter_4
       inputName=None
       mask=0
       maskA=0
       maskB=0
       maskG=0
       maskR=0
       outputIndex=0
      }
     }
    ]
    convertOutputs [{type=Vector2}]
   }
  }
  MaterialExpressionFrac_1 {class=MaterialExpressionFrac, inputs [None {from=MaterialExpressionMultiply_4, output=""}], properties {}}
  MaterialExpressionFrac_2 {class=MaterialExpressionFrac, inputs [None {from=MaterialExpressionMultiply_1, output=""}], properties {}}
  MaterialExpressionFunctionOutput_0 {class=MaterialExpressionFunctionOutput, inputs [None {from=MaterialExpressionTextureSample_1, output=RGB}]}
  MaterialExpressionFunctionOutput_1 {class=MaterialExpressionFunctionOutput, inputs [None {from=MaterialExpressionTextureSample_1, output=A}], properties {outputName=Mask}}
  MaterialExpressionIf_1 {
   class=MaterialExpressionIf
   inputs [
    A {from=MaterialExpressionMin_0, output=""}
    B {from=MaterialExpressionConstant_8, output=""}
    "A > B" {from=MaterialExpressionConstant_3, output=""}
    "A == B" {from="", output=""}
    "A < B" {from=MaterialExpressionConstant_4, output=""}
   ]
   properties {}
  }
  MaterialExpressionMaterialFunctionCall_2 {
   class=MaterialExpressionMaterialFunctionCall
   inputs [
    UVs {from=MaterialExpressionTextureCoordinate_0}
    Texture Scale {from=MaterialExpressionScalarParameter_2}
   ]
   properties {materialFunction=&ScaleUVsByCenter}
  }
  MaterialExpressionMaterialFunctionCall_6 {
   class=MaterialExpressionMaterialFunctionCall
   inputs [
    "Animation  Phase (0-1)" {from=MaterialExpressionIf_1}
    Number of Rows {from=MaterialExpressionConstant_0}
    Number of Columns {from=MaterialExpressionConstant_1}
    Texture {from=""}
    UVs {from=MaterialExpressionAdd_0}
    MipBias/Level {from=""}
    "Use Mip Bias (T) Level (F)" {from=""}
    "Clamp Anim (see tooltip)" {from=""}
   ]
   properties {materialFunction=&FlipBook}
  }
  MaterialExpressionMin_0 {
   class=MaterialExpressionMin
   inputs [
    A {from=MaterialExpressionFrac_2, output=""}
    B {from=MaterialExpressionFrac_1, output=""}
   ]
   properties {}
  }
  MaterialExpressionMultiply_0 {
   class=MaterialExpressionMultiply
   inputs [
    A {from=MaterialExpressionTime_0}
    B {from=MaterialExpressionConstant_2}
   ]
  }
  MaterialExpressionMultiply_1 {
   class=MaterialExpressionMultiply
   inputs [
    A {from=MaterialExpressionMultiply_0}
    B {from=MaterialExpressionConstant_7}
   ]
  }
  MaterialExpressionMultiply_4 {
   class=MaterialExpressionMultiply
   inputs [
    A {from=MaterialExpressionMultiply_0}
    B {from=MaterialExpressionConstant_6}
   ]
  }
  MaterialExpressionScalarParameter_2 {class=MaterialExpressionScalarParameter, properties {defaultValue=1, parameterName=Eyes Scale}}
  MaterialExpressionScalarParameter_3 {class=MaterialExpressionScalarParameter, properties {parameterName=Eyes Offset X}}
  MaterialExpressionScalarParameter_4 {class=MaterialExpressionScalarParameter, properties {parameterName=Eyes Offset Y}}
  MaterialExpressionTextureCoordinate_0 {class=MaterialExpressionTextureCoordinate, properties {coordinateIndex=1}}
  MaterialExpressionTextureSample_1 {
   class=MaterialExpressionTextureSample
   inputs [
    UVs {from=MaterialExpressionClamp_0}
    Tex {from=""}
    Apply View MipBias {from=""}
   ]
   properties {texture=/BRCosmetics/Characters/Player/Female/Medium/Bodies/F_MED_OmelettePop/Textures/T_OmelettePop_ChickEyes_D}
  }
  MaterialExpressionTime_0 {class=MaterialExpressionTime}
 }}
