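# Skeletal character material for the OmelettePop body cosmetic, built on the 2023 character parent with a fluffy fabric layer and a blink overlay.
# PreFX output feeds the fluffy fabric function, with the tech art mask multiplied by the inverted blink mask to exclude the blink area.
# The blink function's output is then lerped into Base Color before the attributes pass to the shared PostFX function.
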Material /BRCosmetics/Characters/Player/Female/Medium/Bodies/F_MED_OmelettePop/Materials/M_OmelettePop
graph {
 properties {
  bHasPixelAnimation=true
  bUseMaterialAttributes=true
  bUsedWithMeshDeformer=true
  bUsedWithMorphTargets=true
  bUsedWithSkeletalMesh=true
  pixelDepthOffsetMode=PDOM_Legacy
  shadingModel=MSM_FromMaterialExpression
  usedShadingModels=Default Lit | Subsurface | Clear Coat | Subsurface Profile | Cloth
 }
 nodes [
  n1 = MaterialFunctionCall() {materialFunction=&MF_CharacterParent_2023_PreFX}
  n2 = MaterialFunctionCall() {materialFunction=&/MF_OmelettePop_Blink}
  n3 = OneMinus(None=@n2.Mask)
  n4 = GetMaterialAttributes(None=MaterialFunctionCall!(Material=@n1.Result, TechArtMask=Multiply(A=@n1.Tech Art Mask, B=@n3)) {materialFunction=/Game/Global/Materials/Character_Parent_2023/FluffyFabric_Parent_2023/MF_FluffyFabric}.Result) {attributeGetTypes [#1]}
  MaterialFunctionCall(
   Material=SetMaterialAttributes!(MaterialAttributes=@n4.MaterialAttributes, Base Color=LinearInterpolate(A=@n4.Base Color, B=@n2.Result, Alpha=@n2.Result)) {attributeSetTypes [#1]}
   M=@n1.M
   Pre Skinned Local Position=@n1.M
   Distance Blend=@n1.M
   Fresnel=@n1.M
  ) {materialFunction=&MF_CharacterParent_2023_PostFX}
 ]
}
