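# Physics asset for the belt and hanging bag of the OmelettePop female medium body cosmetic.
# Front and back belt capsule chains run from small MSTR_belt anchor spheres to belt5 and belt7, which cross-constrain each other and the dyn_chicketnbag capsule in a loop.
# Swing limits generally loosen toward the bag, with twist mostly locked along the chains.
# Unconstrained pelvis, spine, left thigh, left leg-bag and left jacket capsules supply surrounding body shapes.
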PhysicsAsset /BRCosmetics/Characters/Player/Female/Medium/Bodies/F_MED_OmelettePop/Meshes/F_MED_OmelettePop_BeltBackPac_Physics
bodies {
 MSTR_belt_bk [{shapeType=Sphere, radius=0.505, shapeName=MSTR_belt_bk_sphere}]
 MSTR_belt_fr [{shapeType=Sphere, radius=0.505, shapeName=MSTR_belt_fr_sphere}]
 belt02 [{
   shapeType=Capsule
   length=3.5725
   radius=2.7175
   shapeName=belt1_capsule
   center=<3.297 -1.2769 0.03>
   rotation=<rot -0.0359 -88.3678 -88.9621>
  }]
 belt1 [{
   shapeType=Capsule
   length=3.8024
   radius=2.4877
   shapeName=belt1_capsule
   center=<3.297 -1.2769 0.03>
   rotation=<rot -0.0359 -88.3678 -88.9621>
  }]
 belt2 [{
   shapeType=Capsule
   length=9.9833
   radius=2.0972
   shapeName=belt2_capsule
   center=<7.4371 0.9371 -0.325>
   rotation=<rot 0.2182 -88.8997 -85.6731>
  }]
 belt3 [{
   shapeType=Capsule
   length=9.4436
   radius=2.6368
   shapeName=belt2_capsule
   center=<5.4571 1.2568 0.0341>
   rotation=<rot -0.2412 -92.3692 -92.1653>
  }]
 belt4 [{
   shapeType=Capsule
   length=7.0453
   radius=2.5988
   shapeName=belt1_capsule
   center=<5.4189 -0.6571 -0.27>
   rotation=<rot -0.3184 -85.1344 -86.5138>
  }]
 belt5 [{shapeType=Capsule, length=4.0045, radius=1.8846, shapeName=belt5_capsule, rotation=<rot -20.5286 -67.2687 0.2568>}]
 belt6 [{shapeType=Capsule, length=8.5992, radius=2.0972, shapeName=belt2_capsule, center=<5.8049 0.1873 -1.0749>, rotation=<rot 0 -86.018 -86.887>}]
 belt7 [{
   shapeType=Capsule
   length=4.0045
   radius=1.8846
   shapeName=belt5_capsule
   center=<0.8606 0.3986 -0.1497>
   rotation=<rot 13.1286 -38.3766 -73.9962>
  }]
 def_legBag_l [{
   shapeType=Capsule
   length=2.8991
   radius=8.5192
   shapeName=def_legBag_l_capsule
   center=<-2.5192 -0.6057 0.0423>
   rotation=<rot -0.0988 0.7358 -90.0007>
  }]
 dyn_chicketnbag [{
   shapeType=Capsule
   length=4.3611
   radius=5.5225
   shapeName=dyn_chicketnbag_capsule
   center=<2.901 -0.2368 0.0202>
   rotation=<rot 85.9188 -49.3311 43.6173>
  }]
 dyn_jacket_front_03_l [{
   shapeType=Capsule
   length=9.4692
   radius=9.1815
   shapeName=dyn_jacket_front_03_l_capsule
   center=<-3.8139 6.2889 0.3006>
   rotation=<rot -89.6218 0 0.9194>
  }]
 dyn_jacket_side_03_l [{
   shapeType=Capsule
   length=9.4692
   radius=9.1815
   shapeName=dyn_jacket_front_03_l_capsule
   center=<-2.7365 7.7854 0.9916>
   rotation=<rot -89.6218 0 0.9194>
  }]
 pelvis [
  {
   shapeType=Capsule
   length=8.6932
   radius=14.8465
   shapeName=pelvis_capsule
   center=<-1.3803 -2.2207 -0.6549>
   rotation=<rot 0 170.0376 -178.1206>
  }
  {shapeType=Capsule, length=15.1284, radius=10.2503, shapeName=pelvis_capsule, center=<3.1096 -8.4893 0>}
 ]
 spine_02 [{shapeType=Capsule, length=6.2517, radius=14.7508, shapeName=spine_02_capsule, center=<0.6101 -1.0301 0>}]
 spine_03 [{shapeType=Capsule, length=9.3935, radius=11.609, shapeName=spine_02_capsule, center=<1.7057 -0.6306 0>}]
 thigh_twist_01_l [{
   shapeType=Capsule
   length=20.3835
   radius=9.9873
   shapeName=thigh_twist_01_l_capsule
   center=<10.4323 -1.4044 1.4407>
   rotation=<rot -89.914 -179.9963 179.9963>
  }]
}
constraints [
 {bone1Name=belt2, bone2Name=MSTR_belt_bk, swing1LimitDegrees=7, swing2LimitDegrees=7, twistLimitDegrees=45, twistMotion=Locked}
 {bone1Name=belt1, bone2Name=MSTR_belt_fr, swing1LimitDegrees=10, swing2LimitDegrees=5, twistLimitDegrees=45, twistMotion=Locked}
 {bone1Name=belt3, bone2Name=belt2, swing1LimitDegrees=10, swing2LimitDegrees=7, twistLimitDegrees=45, twistMotion=Locked}
 {bone1Name=belt6, bone2Name=belt3, swing1LimitDegrees=15, swing2LimitDegrees=20, twistLimitDegrees=45, twistMotion=Locked}
 {bone1Name=belt7, bone2Name=belt6, swing1LimitDegrees=25, swing2LimitDegrees=25, twistLimitDegrees=45}
 {bone1Name=belt5, bone2Name=belt4, swing1LimitDegrees=45, swing2LimitDegrees=45, twistLimitDegrees=60}
 {bone1Name=belt7, bone2Name=belt5, swing1LimitDegrees=45, swing2LimitDegrees=45, twistLimitDegrees=45}
 {bone1Name=belt5, bone2Name=belt7, swing1LimitDegrees=45, swing2LimitDegrees=45, twistLimitDegrees=45}
 {bone1Name=belt02, bone2Name=belt1, swing1LimitDegrees=10, swing2LimitDegrees=7, twistLimitDegrees=45, twistMotion=Locked}
 {bone1Name=belt4, bone2Name=belt02, swing1LimitDegrees=15, swing2LimitDegrees=15, twistLimitDegrees=45, twistMotion=Locked}
 {bone1Name=belt7, bone2Name=dyn_chicketnbag, swing1LimitDegrees=45, swing2LimitDegrees=45, twistLimitDegrees=45}
 {bone1Name=belt5, bone2Name=dyn_chicketnbag, swing1LimitDegrees=15, swing2LimitDegrees=15, twistLimitDegrees=15}
 {bone1Name=dyn_chicketnbag, bone2Name=belt5, swing1LimitDegrees=45, swing2LimitDegrees=45, twistLimitDegrees=45, twistMotion=Locked}
]
