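# Control Rig for the F_MED_OmelettePop character body, adding corrective and secondary deformation to jacket, belt, pants, feet, hood and body.
# Parent constraints, halved upper-arm twist and a shared DeformRig function run first, then a sequence layers bone offsets.
# Thigh and arm spherical pose readers, remapped bone transforms and gameplay variables (sliding, axe swing, skydive, parachute) weight those offsets.
# Kalman filtering and alpha interpolation smooth the axe-swing and parachute blends.
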ControlRigBlueprint /BRCosmetics/Characters/Player/Female/Medium/Bodies/F_MED_OmelettePop/Meshes/F_MED_OmelettePop_CtrlRig
graphs [
 RigVMFunctionLibrary {links [], nodes []}
 RigVMFunction_Sequence_1_ContainedGraph {nodes [
   n1 = RigVMFunctionEntryNode()
   RigVMFunction_Sequence = &RigVMFunction_Sequence(ExecuteContext=@n1.ExecuteContext)
   n2 = &RigVMFunction_Sequence(ExecuteContext=@RigVMFunction_Sequence.B)
   n3 = &RigVMFunction_Sequence(ExecuteContext=@n2.B)
   n4 = &RigVMFunction_Sequence(ExecuteContext=@n3.B)
   RigVMFunctionReturnNode(A=@RigVMFunction_Sequence.A, B=@n2.A, C=@n3.A, D=@n4.A, E=@n4.B)
  ]}
 RigVMModel {nodes [
   n1 = &RigVMFunction_MathBoolToFloat(Value=RigVMVariableNode() {variable=isSliding}.Value)
   &RigUnit_ParentConstraint(
    AdvancedSettings=(InterpolationType=Average,RotationOrderForFilter=XZY)
    Child=(Type=Bone,Name="MSTR_jacket_side_l")
    ExecutePin=()
    Filter=(TranslationFilter=(bX=true,bY=true,bZ=true),RotationFilter=(bX=true,bY=true,bZ=true),ScaleFilter=(bX=True,bY=True,bZ=True))
    Parents=((Item=(Type=Bone,Name="pelvis"),Weight=1),(Item=(Type=Bone,Name="spine_03"),Weight=1))
    Weight=@n1.Result
   )
   > &RigUnit_ParentConstraint(
    AdvancedSettings=(InterpolationType=Average,RotationOrderForFilter=XZY)
    Child=(Type=Bone,Name="MSTR_jacket_side_r")
    Filter=(TranslationFilter=(bX=true,bY=true,bZ=true),RotationFilter=(bX=true,bY=true,bZ=true),ScaleFilter=(bX=True,bY=True,bZ=True))
    Parents=((Item=(Type=Bone,Name="pelvis"),Weight=1),(Item=(Type=Bone,Name="spine_03"),Weight=1))
    Weight=@n1.Result
   )
   > &RigUnit_ParentConstraint(
    AdvancedSettings=(InterpolationType=Average,RotationOrderForFilter=XZY)
    Child=(Type=Bone,Name="MSTR_jacket_front_l")
    Filter=(TranslationFilter=(bX=true,bY=true,bZ=true),RotationFilter=(bX=true,bY=true,bZ=true),ScaleFilter=(bX=True,bY=True,bZ=True))
    Parents=((Item=(Type=Bone,Name="spine_03"),Weight=1),(Item=(Type=Bone,Name="pelvis"),Weight=1))
    bMaintainOffset=true
    Weight=@n1.Result
   )
   > &RigUnit_ParentConstraint(
    AdvancedSettings=(InterpolationType=Average,RotationOrderForFilter=XZY)
    Child=(Type=Bone,Name="MSTR_jacket_front_r")
    Filter=(TranslationFilter=(bX=true,bY=true,bZ=true),RotationFilter=(bX=true,bY=true,bZ=true),ScaleFilter=(bX=True,bY=True,bZ=True))
    Parents=((Item=(Type=Bone,Name="spine_03"),Weight=1),(Item=(Type=Bone,Name="pelvis"),Weight=1))
    Weight=@n1.Result
   )
   RigUnit_BeginExecution = &RigUnit_BeginExecution()
   &RigUnit_ModifyTransforms(
    ItemToModify=((Item=(Type=Bone,Name="def_legBag_l"),Transform=(Rotation=(X=0,Y=0,Z=0,W=1),Translation=(X=1,Y=0,Z=0),Scale3D=(X=1,Y=1,Z=1))),(Item=(Type=Bone,Name="def_regBag_r"),Transform=(Rotation=(X=0,Y=0,Z=0,W=1),Translation=(X=-1,Y=0,Z=0),Scale3D=(X=1,Y=1,Z=1))))
    Weight=1
    WeightMinimum=0
    ExecutePin=@RigUnit_BeginExecution.ExecutePin
   )
   n2 = &RigUnit_GetRelativeTransformForItem(Child=(Type=Bone,Name="upperarm_twist_01_l"), Parent=(Type=Bone,Name="upperarm_l"), bChildInitial=False, bParentInitial=False)
   RigVMFunction_MathQuaternionFromEuler = &RigVMFunction_MathQuaternionFromEuler(
    Euler=(X=0,Y=0,Z=0)
    Euler.X=&RigVMFunction_MathDoubleMul(B=0.5, A=&RigVMFunction_MathQuaternionToEuler(RotationOrder=ZYX, Value=@n2.RelativeTransform.Rotation).Result.X).Result
   )
   > &RigUnit_ModifyTransforms(
    ItemToModify=((Item=(Type=Bone,Name="upperarm_twist_01_l"),Transform=(Rotation=(X=-0.3827,Y=0,Z=0,W=0.9239),Translation=(X=0,Y=0,Z=0),Scale3D=(X=1,Y=1,Z=1))))
    Weight=1
    WeightMinimum=0
    ItemToModify.0.Transform.Rotation=@RigVMFunction_MathQuaternionFromEuler.Result
   )
   n3 = &RigUnit_GetRelativeTransformForItem(Child=(Type=Bone,Name="upperarm_twist_01_r"), Parent=(Type=Bone,Name="upperarm_r"), bChildInitial=False, bParentInitial=False)
   n4 = &RigVMFunction_MathQuaternionToEuler(RotationOrder=ZYX, Value=@n3.RelativeTransform.Rotation)
   > &RigUnit_ModifyTransforms(
    ItemToModify=((Item=(Type=Bone,Name="upperarm_twist_01_r"),Transform=(Rotation=(X=-0.3827,Y=0,Z=0,W=0.9239),Translation=(X=0,Y=0,Z=0),Scale3D=(X=1,Y=1,Z=1))))
    Weight=1
    WeightMinimum=0
    ItemToModify.0.Transform.Rotation=&RigVMFunction_MathQuaternionFromEuler(Euler=(X=0,Y=0,Z=0), Euler.X=&RigVMFunction_MathDoubleMul(B=0.5, A=@n4.Result.X).Result).Result
   )
   > &RigUnit_ParentConstraint(
    AdvancedSettings=(InterpolationType=Average,RotationOrderForFilter=XZY)
    Child=(Type=Bone,Name="MSTR_jacket_bk_l")
    Filter=(TranslationFilter=(bX=true,bY=true,bZ=true),RotationFilter=(bX=true,bY=true,bZ=true),ScaleFilter=(bX=True,bY=True,bZ=True))
    Parents=((Item=(Type=Bone,Name="spine_02"),Weight=1),(Item=(Type=Bone,Name="pelvis"),Weight=1))
    Weight=0.5
    bMaintainOffset=true
   )
   > &RigUnit_ParentConstraint(
    AdvancedSettings=(InterpolationType=Average,RotationOrderForFilter=XZY)
    Child=(Type=Bone,Name="MSTR_jacket_bk_r")
    Filter=(TranslationFilter=(bX=true,bY=true,bZ=true),RotationFilter=(bX=True,bY=true,bZ=true),ScaleFilter=(bX=True,bY=True,bZ=True))
    Parents=((Item=(Type=Bone,Name="spine_02"),Weight=1),(Item=(Type=Bone,Name="pelvis"),Weight=1))
    Weight=0.5
    bMaintainOffset=true
   )
   > &RigUnit_ParentConstraint(
    AdvancedSettings=(InterpolationType=Average,RotationOrderForFilter=XZY)
    Child=(Type=Bone,Name="MSTR_jacket_bk")
    Filter=(TranslationFilter=(bX=True,bY=true,bZ=true),RotationFilter=(bX=True,bY=true,bZ=true),ScaleFilter=(bX=True,bY=True,bZ=True))
    Parents=((Item=(Type=Bone,Name="spine_02"),Weight=1),(Item=(Type=Bone,Name="pelvis"),Weight=1))
    Weight=0.5
    bMaintainOffset=true
   )
   > &RigUnit_ParentConstraint(
    AdvancedSettings=(InterpolationType=Average,RotationOrderForFilter=XZY)
    Child=(Type=Bone,Name="MSTR_jacket_side_l")
    Filter=(TranslationFilter=(bX=true,bY=true,bZ=false),RotationFilter=(bX=true,bY=true,bZ=true),ScaleFilter=(bX=True,bY=True,bZ=True))
    Parents=((Item=(Type=Bone,Name="pelvis"),Weight=0.6),(Item=(Type=Bone,Name="spine_03"),Weight=0.2),(Item=(Type=Bone,Name="thigh_l"),Weight=0))
    Weight=0.5
    bMaintainOffset=true
   )
   > &RigUnit_ParentConstraint(
    AdvancedSettings=(InterpolationType=Average,RotationOrderForFilter=XZY)
    Child=(Type=Bone,Name="MSTR_jacket_side_r")
    Filter=(TranslationFilter=(bX=true,bY=true,bZ=false),RotationFilter=(bX=true,bY=true,bZ=true),ScaleFilter=(bX=True,bY=True,bZ=True))
    Parents=((Item=(Type=Bone,Name="pelvis"),Weight=0.6),(Item=(Type=Bone,Name="spine_03"),Weight=0.2),(Item=(Type=Bone,Name="thigh_r"),Weight=0))
    Weight=0.5
   )
   > &RigUnit_ParentConstraint(
    AdvancedSettings=(InterpolationType=Average,RotationOrderForFilter=XZY)
    Child=(Type=Bone,Name="MSTR_jacket_front_l")
    Filter=(TranslationFilter=(bX=true,bY=false,bZ=false),RotationFilter=(bX=true,bY=true,bZ=true),ScaleFilter=(bX=True,bY=True,bZ=True))
    Parents=((Item=(Type=Bone,Name="spine_03"),Weight=0.2),(Item=(Type=Bone,Name="pelvis"),Weight=0.8))
    Weight=0.5
    bMaintainOffset=true
   )
   > &RigUnit_ParentConstraint(
    AdvancedSettings=(InterpolationType=Average,RotationOrderForFilter=XZY)
    Child=(Type=Bone,Name="MSTR_jacket_front_r")
    Filter=(TranslationFilter=(bX=true,bY=false,bZ=false),RotationFilter=(bX=true,bY=true,bZ=true),ScaleFilter=(bX=True,bY=True,bZ=True))
    Parents=((Item=(Type=Bone,Name="spine_03"),Weight=0.2),(Item=(Type=Bone,Name="pelvis"),Weight=0.8))
    Weight=0.5
   )
   > FN_DeformRig_FNC = RigVMFunctionReferenceNode(
    DeformRig_ Curve="(deformRig_l_upperarm_up_alpha_72_#1=False,deformRig_r_upperarm_up_alpha_73_#2=False,deformRig_l_upperarm_fr_alpha_74_#3=False,deformRig_r_upperarm_fr_alpha_75_CB51B8 ...(truncated)"
    DeformRig_Bool="(upperarm_1_#4=True,elbow_3_#5=True,wrist_7_#6=True,groin_8_#7=True,glute_17_#8=true,knee_9_40C408324CA261 ...(truncated)"
    DeformRig_Debug=(upperarm_debug_18_#4=False,wrist_debug_20_#6=False,lat_debug_26_#9=False,groinGlute_debug_27_#7=False)
    DeformRig_Value="(upperarm_up_18_#10=(Rotation=(X=0.0,Y=-0.0,Z=0.0,W=1.0),Translation=(X=0.0,Y=0.0,Z=8.0),Scale3D=(X=0.70,Y=1.0,Z=1.0)),upperarm_fwd_fr_19_#11=(Rotation=(X=0.0,Y=-0.0,Z=0.0,W=1.0),Translation=(X=0.0, ...(truncated)"
   )
   > RigVMFunction_Sequence = RigVMAggregateNode()
   n5 = &RigUnit_SphericalPoseReader(
    ActiveRegionScaleFactors=(PositiveWidth=3.5,NegativeWidth=0.5,PositiveHeight=3,NegativeHeight=3)
    ActiveRegionSize=0.1
    Debug=(bDrawDebug=false,DebugScale=25,DebugSegments=20,DebugThickness=0.25)
    DriverAxis=(X=-1,Y=0,Z=0)
    DriverItem=(Type=Bone,Name="thigh_l")
    FalloffRegionScaleFactors=(PositiveWidth=1,NegativeWidth=1,PositiveHeight=1,NegativeHeight=1)
    FalloffSize=0.8
    OptionalParentItem=(Type=Bone,Name="spine_03")
    RotationOffset=(X=110,Y=0,Z=-30)
    ExecutePin=@RigVMFunction_Sequence.A
   )
   > &RigUnit_ModifyTransforms(
    ItemToModify=((Item=(Type=Bone,Name="MSTR_jacket_front_l"),Transform=(Rotation=(X=0.7127,Y=0.1753,Z=-0.1617,W=0.6597),Translation=(X=0.0698,Y=8.9674,Z=-6.417),Scale3D=(X=1,Y=1,Z=1))))
    WeightMinimum=0
    Weight=@n5.OutputParam
   )
   > &RigUnit_ModifyTransforms(
    ItemToModify="((Item=(Type=Bone,Name=\"MSTR_jacket_side_l\"),Transform=(Rotation=(X=0.4364,Y=-0.009,Z=-0.0632,W=0.8975),Translation=(X=1.7745,Y=4.5559,Z=-6.1624),Scale3D=(X=1,Y=1,Z=1))),(Item=(Type=Bone,Name=\"MSTR_jacket_bk_l\"),Transform=(Rotation=(X=0.228,Y=0.0205,Z=-0.1 ...(truncated)"
    WeightMinimum=0
    Weight=@n5.OutputParam
   )
   n6 = &RigVMFunction_MathBoolNot(Value=RigVMVariableNode() {variable=isBaseBallSliding}.Value)
   n7 = RigVMVariableNode() {variable=isRBAN}
   Cast = &RigVMFunction_MathBoolToFloat(Value=&RigVMFunction_MathBoolAnd(B=@n6.Result, A=@n7.Value).Result)
   > &RigUnit_ModifyTransforms(
    ItemToModify="((Item=(Type=Bone,Name=\"MSTR_bag\"),Transform=(Rotation=(X=0,Y=0,Z=0,W=1),Translation=(X=0,Y=0,Z=0),Scale3D=(X=1,Y=1,Z=1))),(Item=(Type=Bone,Name=\"MSTR_belt_fr\"),Transform=(Rotation=(X=0.0498,Y=0.2135,Z=0.1187,W=0.9684),Translation=(X=-1.3462,Y=-0.3392,Z=0. ...(truncated)"
    WeightMinimum=0
    Weight=&RigVMFunction_MathDoubleMul(B=@Cast.Result, A=@n5.OutputParam).Result
   )
   n8 = &RigVMFunction_MathBoolAnd(A=@n6.Result, B=&RigVMFunction_MathBoolNot(Value=@n7.Value).Result)
   > &RigUnit_ModifyTransforms(
    ItemToModify="((Item=(Type=Bone,Name=\"MSTR_bag\"),Transform=(Rotation=(X=0,Y=0,Z=0,W=1),Translation=(X=0,Y=0,Z=0),Scale3D=(X=1,Y=1,Z=1))),(Item=(Type=Bone,Name=\"MSTR_belt_fr\"),Transform=(Rotation=(X=-0.006,Y=-0.2886,Z=0.0191,W=-0.9572),Translation=(X=0,Y=0,Z=0),Scale3D=( ...(truncated)"
    WeightMinimum=0
    Weight=&RigVMFunction_MathDoubleMul(B=&RigVMFunction_MathBoolToFloat(Value=@n8.Result).Result, A=@n5.OutputParam).Result
   )
   > n9 = &RigUnit_SphericalPoseReader(
    ActiveRegionScaleFactors=(PositiveWidth=3.5,NegativeWidth=0.5,PositiveHeight=3,NegativeHeight=3)
    ActiveRegionSize=0.1
    Debug=(bDrawDebug=false,DebugScale=25,DebugSegments=20,DebugThickness=0.25)
    DriverAxis=(X=1,Y=0,Z=0)
    DriverItem=(Type=Bone,Name="thigh_r")
    FalloffRegionScaleFactors=(PositiveWidth=1,NegativeWidth=1,PositiveHeight=1,NegativeHeight=1)
    FalloffSize=0.8
    FlipHeightScaling=false
    FlipWidthScaling=true
    OptionalParentItem=(Type=Bone,Name="spine_03")
    RotationOffset=(X=-70,Y=0,Z=-30)
   )
   > &RigUnit_ModifyTransforms(
    ItemToModify=((Item=(Type=Bone,Name="MSTR_jacket_front_r"),Transform=(Rotation=(X=0.7131,Y=0.1752,Z=0.1618,W=-0.6592),Translation=(X=0.0698,Y=8.9674,Z=6.417),Scale3D=(X=1,Y=1,Z=1))))
    WeightMinimum=0
    Weight=@n9.OutputParam
   )
   > &RigUnit_ModifyTransforms(
    ItemToModify="((Item=(Type=Bone,Name=\"MSTR_jacket_side_r\"),Transform=(Rotation=(X=-0.4342,Y=0.044,Z=-0.0462,W=0.8985),Translation=(X=1.7745,Y=4.5559,Z=6.1624),Scale3D=(X=1,Y=1,Z=1))),(Item=(Type=Bone,Name=\"MSTR_jacket_bk_r\"),Transform=(Rotation=(X=-0.228,Y=-0.0205,Z=-0. ...(truncated)"
    WeightMinimum=0
    Weight=@n9.OutputParam
   )
   n10 = &RigVMFunction_MathDoubleMul(
    B=&RigVMFunction_MathDoubleRemap(
     SourceMaximum=70
     SourceMinimum=45
     TargetMaximum=1
     TargetMinimum=0
     Value=&RigUnit_GetRelativeTransformForItem(Child=(Type=Bone,Name="pelvis"), Parent=(Type=Bone,Name="root"), bChildInitial=False, bParentInitial=False).RelativeTransform.Translation.Z
    ).Result
    A=@n9.OutputParam
   )
   > &RigUnit_ModifyTransforms(
    ItemToModify="((Item=(Type=Bone,Name=\"MSTR_jacket_front_r\"),Transform=(Rotation=(X=0,Y=0,Z=0.0436,W=0.999),Translation=(X=0,Y=0,Z=2),Scale3D=(X=1,Y=1,Z=1))),(Item=(Type=Bone,Name=\"MSTR_jacket_side_r\"),Transform=(Rotation=(X=-0.0872,Y=0,Z=0,W=0.9962),Translation=(X=0,Y=0 ...(truncated)"
    WeightMinimum=0
    Weight=@n10.Result
   )
   > &RigUnit_ModifyTransforms(
    ItemToModify="((Item=(Type=Bone,Name=\"MSTR_jacket_front_l\"),Transform=(Rotation=(X=0,Y=0,Z=0.0436,W=0.999),Translation=(X=0,Y=0,Z=-2),Scale3D=(X=1,Y=1,Z=1))),(Item=(Type=Bone,Name=\"MSTR_jacket_side_l\"),Transform=(Rotation=(X=0.0872,Y=0,Z=0,W=0.9962),Translation=(X=0,Y=0 ...(truncated)"
    WeightMinimum=0
    Weight=@n10.Result
   )
   > &RigUnit_ModifyTransforms(
    ItemToModify=((Item=(Type=Bone,Name="MSTR_jacket_bk"),Transform=(Rotation=(X=0,Y=0,Z=0.0872,W=0.9962),Translation=(X=0,Y=-2,Z=0),Scale3D=(X=1,Y=1,Z=1))))
    WeightMinimum=0
    Weight=@n10.Result
   )
   &RigUnit_ParentConstraint(
    AdvancedSettings=(InterpolationType=Average,RotationOrderForFilter=XZY)
    Child=(Type=Bone,Name="MSTR_jacket_bk_l")
    Filter=(TranslationFilter=(bX=True,bY=true,bZ=true),RotationFilter=(bX=true,bY=true,bZ=true),ScaleFilter=(bX=True,bY=True,bZ=True))
    Parents=((Item=(Type=Bone,Name="spine_03"),Weight=1),(Item=(Type=Bone,Name="pelvis"),Weight=0.5))
    Weight=@n1.Result
    ExecutePin=@RigVMFunction_Sequence.B
   )
   > &RigUnit_ParentConstraint(
    AdvancedSettings=(InterpolationType=Average,RotationOrderForFilter=XZY)
    Child=(Type=Bone,Name="MSTR_jacket_bk_r")
    Filter=(TranslationFilter=(bX=true,bY=true,bZ=true),RotationFilter=(bX=True,bY=true,bZ=true),ScaleFilter=(bX=True,bY=True,bZ=True))
    Parents=((Item=(Type=Bone,Name="spine_03"),Weight=1),(Item=(Type=Bone,Name="pelvis"),Weight=0.5))
    Weight=@n1.Result
   )
   > &RigUnit_ParentConstraint(
    AdvancedSettings=(InterpolationType=Average,RotationOrderForFilter=XZY)
    Child=(Type=Bone,Name="MSTR_jacket_bk")
    Filter=(TranslationFilter=(bX=true,bY=true,bZ=true),RotationFilter=(bX=true,bY=true,bZ=true),ScaleFilter=(bX=True,bY=True,bZ=True))
    Parents=((Item=(Type=Bone,Name="spine_03"),Weight=1),(Item=(Type=Bone,Name="pelvis"),Weight=0.5))
    bMaintainOffset=true
    Weight=@n1.Result
   )
   n11 = RigVMVariableNode() {variable=isBaseBallSliding}
   n12 = &RigVMFunction_MathBoolToFloat(Value=&RigVMFunction_MathBoolAnd(B=&RigVMFunction_MathBoolNot(Value=@n11.Value).Result, A=RigVMVariableNode() {variable=isSliding}.Value).Result)
   > &RigUnit_ModifyTransforms(
    ItemToModify=((Item=(Type=Bone,Name="MSTR_jacket_front_l"),Transform=(Rotation=(X=0.0871,Y=-0.0042,Z=-0.0477,W=0.9951),Translation=(X=0,Y=-0.5566,Z=0),Scale3D=(X=1,Y=1,Z=1))))
    WeightMinimum=0
    Weight=@n12.Result
   )
   > &RigUnit_ModifyTransforms(
    ItemToModify="((Item=(Type=Bone,Name=\"MSTR_jacket_side_l\"),Transform=(Rotation=(X=0.0871,Y=-0.0036,Z=-0.0407,W=0.9954),Translation=(X=-0.0909,Y=-1.1099,Z=0),Scale3D=(X=1,Y=1,Z=1))),(Item=(Type=Bone,Name=\"MSTR_jacket_bk_l\"),Transform=(Rotation=(X=0,Y=0,Z=0,W=1),Translati ...(truncated)"
    WeightMinimum=0
    Weight=@n12.Result
   )
   > &RigUnit_ModifyTransforms(
    ItemToModify=((Item=(Type=Bone,Name="MSTR_jacket_front_r"),Transform=(Rotation=(X=-0.0879,Y=-0.0259,Z=-0.0319,W=0.9953),Translation=(X=-0.0219,Y=-0.3576,Z=0.0125),Scale3D=(X=1,Y=1,Z=1))))
    WeightMinimum=0
    Weight=@n12.Result
   )
   > &RigUnit_ModifyTransforms(
    ItemToModify="((Item=(Type=Bone,Name=\"MSTR_jacket_side_r\"),Transform=(Rotation=(X=-0.0871,Y=0.0017,Z=-0.0191,W=0.996),Translation=(X=0,Y=-0.7501,Z=0),Scale3D=(X=1,Y=1,Z=1))),(Item=(Type=Bone,Name=\"MSTR_jacket_bk\"),Transform=(Rotation=(X=0.1094,Y=-0.1252,Z=0.2338,W=-0.95 ...(truncated)"
    WeightMinimum=0
    Weight=@n12.Result
   )
   n13 = &RigVMFunction_MathBoolToFloat(Value=RigVMVariableNode() {variable=isBaseBallSliding}.Value)
   > &RigUnit_ModifyTransforms(
    ItemToModify=((Item=(Type=Bone,Name="MSTR_jacket_front_l"),Transform=(Rotation=(X=0.0792,Y=0.0911,Z=0.0792,W=0.9895),Translation=(X=-2.6886,Y=-2.8418,Z=2.9193),Scale3D=(X=1,Y=1,Z=1))))
    WeightMinimum=0
    Weight=@n13.Result
   )
   > &RigUnit_ModifyTransforms(
    ItemToModify="((Item=(Type=Bone,Name=\"MSTR_jacket_side_l\"),Transform=(Rotation=(X=0.0052,Y=-0.0377,Z=0.137,W=0.9898),Translation=(X=-0.8642,Y=-1.2178,Z=2.3017),Scale3D=(X=1,Y=1,Z=1))),(Item=(Type=Bone,Name=\"MSTR_jacket_bk_l\"),Transform=(Rotation=(X=0,Y=0,Z=0,W=1),Transl ...(truncated)"
    WeightMinimum=0
    Weight=@n13.Result
   )
   > &RigUnit_ModifyTransforms(
    ItemToModify=((Item=(Type=Bone,Name="MSTR_jacket_front_r"),Transform=(Rotation=(X=0,Y=0,Z=0,W=1),Translation=(X=-2.6695,Y=1.127,Z=-4.1354),Scale3D=(X=1,Y=1,Z=1))))
    WeightMinimum=0
    Weight=@n13.Result
   )
   > &RigUnit_ModifyTransforms(
    ItemToModify="((Item=(Type=Bone,Name=\"MSTR_jacket_side_r\"),Transform=(Rotation=(X=-0.064,Y=-0.0139,Z=-0.1003,W=-0.9928),Translation=(X=-1.9055,Y=-2.2932,Z=0.1469),Scale3D=(X=1,Y=1,Z=1))),(Item=(Type=Bone,Name=\"MSTR_jacket_bk\"),Transform=(Rotation=(X=0,Y=0,Z=0,W=1),Trans ...(truncated)"
    WeightMinimum=0
    Weight=@n13.Result
   )
   > &RigUnit_ModifyTransforms(
    ItemToModify="((Item=(Type=Bone,Name=\"MSTR_belt_bk\"),Transform=(Rotation=(X=0.1787,Y=-0.1034,Z=-0.0739,W=-0.9757),Translation=(X=0.171,Y=-0.1537,Z=3.3218),Scale3D=(X=1,Y=1,Z=1))),(Item=(Type=Bone,Name=\"MSTR_belt_fr\"),Transform=(Rotation=(X=0.0044,Y=-0.402,Z=0.0665,W=-0. ...(truncated)"
    WeightMinimum=0
    Weight=@n13.Result
   )
   n14 = &RigVMFunction_MathQuaternionToEuler(
    RotationOrder=ZYX
    Value=&RigUnit_GetRelativeTransformForItem(Child=(Type=Bone,Name="spine_04"), Parent=(Type=Bone,Name="pelvis"), bChildInitial=False, bParentInitial=False).RelativeTransform.Rotation
   )
   > &RigUnit_ModifyTransforms(
    ItemToModify="((Item=(Type=Bone,Name=\"MSTR_jacket_front_r\"),Transform=(Rotation=(X=0.7115,Y=0.1041,Z=0.3975,W=-0.5701),Translation=(X=-0.2482,Y=-7.7976,Z=0.3032),Scale3D=(X=1,Y=1,Z=1))),(Item=(Type=Bone,Name=\"MSTR_jacket_side_r\"),Transform=(Rotation=(X=0.7516,Y=-0.2184, ...(truncated)"
    WeightMinimum=0
    Weight=&RigVMFunction_MathDoubleMul(
     B=&RigVMFunction_MathDoubleRemap(SourceMaximum=-60, SourceMinimum=0, TargetMaximum=1, TargetMinimum=0, Value=@n14.Result.X).Result
     A=@n13.Result
    ).Result
   )
   > &RigUnit_ModifyTransforms(
    ItemToModify="((Item=(Type=Bone,Name=\"MSTR_jacket_front_l\"),Transform=(Rotation=(X=0.0899,Y=0.0656,Z=-0.0494,W=0.9926),Translation=(X=9.2,Y=-9,Z=-1.5),Scale3D=(X=1,Y=1,Z=1))),(Item=(Type=Bone,Name=\"MSTR_jacket_side_l\"),Transform=(Rotation=(X=0.0949,Y=0.1304,Z=-0.2195,W= ...(truncated)"
    WeightMinimum=0
    Weight=&RigVMFunction_MathDoubleMul(
     B=&RigVMFunction_MathDoubleRemap(SourceMaximum=30, SourceMinimum=0, TargetMaximum=1, TargetMinimum=0, Value=@n14.Result.X).Result
     A=@n13.Result
    ).Result
   )
   n15 = &RigUnit_SphericalPoseReader(
    ActiveRegionScaleFactors=(PositiveWidth=1,NegativeWidth=1,PositiveHeight=1,NegativeHeight=1)
    ActiveRegionSize=0.1
    Debug=(bDrawDebug=false,bDraw2D=False,bDrawLocalAxes=False,DebugScale=25,DebugSegments=20,DebugThickness=0.25)
    DriverAxis=(X=-1,Y=0,Z=0)
    DriverItem=(Type=Bone,Name="thigh_l")
    FalloffRegionScaleFactors=(PositiveWidth=1,NegativeWidth=1,PositiveHeight=1,NegativeHeight=1)
    FalloffSize=0.25
    FlipHeightScaling=False
    FlipWidthScaling=False
    OptionalParentItem=(Type=Bone,Name="None")
    RotationOffset=(X=165,Y=0,Z=0)
    ExecutePin=@RigVMFunction_Sequence.C
   )
   > n16 = &RigUnit_SphericalPoseReader(
    ActiveRegionScaleFactors=(PositiveWidth=1,NegativeWidth=1,PositiveHeight=1,NegativeHeight=1)
    ActiveRegionSize=0.1
    Debug=(bDrawDebug=false,bDraw2D=False,bDrawLocalAxes=False,DebugScale=25,DebugSegments=20,DebugThickness=0.25)
    DriverAxis=(X=1,Y=0,Z=0)
    DriverItem=(Type=Bone,Name="thigh_r")
    FalloffRegionScaleFactors=(PositiveWidth=1,NegativeWidth=1,PositiveHeight=1,NegativeHeight=1)
    FalloffSize=0.25
    FlipHeightScaling=False
    FlipWidthScaling=False
    OptionalParentItem=(Type=Bone,Name="None")
    RotationOffset=(X=-20,Y=0,Z=0)
   )
   Multiply = &RigVMFunction_MathDoubleMul(A=@n15.OutputParam, B=@n16.OutputParam)
   > &RigUnit_ModifyTransforms(
    ItemToModify="((Item=(Type=Bone,Name=\"MSTR_jacket_front_l\"),Transform=(Rotation=(X=-0.3189,Y=-0.1465,Z=-0.7387,W=0.5755),Translation=(X=-0.1376,Y=2.6927,Z=2.5249),Scale3D=(X=1,Y=1,Z=1))),(Item=(Type=Bone,Name=\"MSTR_jacket_side_l\"),Transform=(Rotation=(X=0.262,Y=-0.6271, ...(truncated)"
    WeightMinimum=0
    Weight=@Multiply.Result
   )
   > &RigUnit_ModifyTransforms(
    ItemToModify="((Item=(Type=Bone,Name=\"MSTR_bag\"),Transform=(Rotation=(X=0,Y=0,Z=0,W=1),Translation=(X=0,Y=0,Z=0),Scale3D=(X=1,Y=1,Z=1))),(Item=(Type=Bone,Name=\"MSTR_belt_bk\"),Transform=(Rotation=(X=-0.0681,Y=-0.4959,Z=0.1166,W=-0.8578),Translation=(X=0.0485,Y=0.0453,Z=- ...(truncated)"
    WeightMinimum=0
    Weight=@Multiply.Result
   )
   > &RigUnit_ModifyTransforms(
    ItemToModify="((Item=(Type=Bone,Name=\"MSTR_jacket_front_l\"),Transform=(Rotation=(X=0,Y=0,Z=-0.115,W=0.9934),Translation=(X=-0.3762,Y=-2.4411,Z=0),Scale3D=(X=1,Y=1,Z=1))),(Item=(Type=Bone,Name=\"MSTR_jacket_front_r\"),Transform=(Rotation=(X=0,Y=0,Z=0.1394,W=-0.9902),Transl ...(truncated)"
    WeightMinimum=0
    Weight=&RigVMFunction_MathDoubleRemap(
     SourceMaximum=0.8
     SourceMinimum=1
     TargetMaximum=1
     TargetMinimum=0
     Value=&RigUnit_GetTransform(Item=(Type=Bone,Name="spine_02"), Space=LocalSpace, bInitial=False).Transform.Scale3D.X
    ).Result
   )
   n17 = &RigVMFunction_MathBoolNot(Value=RigVMVariableNode() {variable=isSliding}.Value)
   n18 = RigVMVariableNode() {variable=isSwingAxe}
   KalmanFilter = &RigVMFunction_KalmanFloat(
    BufferSize=8
    Value=&RigVMFunction_MathBoolToFloat(Value=&RigVMFunction_MathBoolAnd(A=@n17.Result, B=@n18.Value).Result).Result
   ) {template="KalmanFilter::Execute(in Value,in BufferSize,out Result)"}
   n19 = &RigVMFunction_MathDoubleRemap(
    SourceMaximum=60
    SourceMinimum=90
    TargetMaximum=0
    TargetMinimum=1
    Value=&RigUnit_GetRelativeTransformForItem(Child=(Type=Bone,Name="pelvis"), Parent=(Type=Bone,Name="root"), bChildInitial=False, bParentInitial=False).RelativeTransform.Translation.Z
   )
   n20 = &RigVMFunction_MathDoubleMul(A=@KalmanFilter.Result, B=@n19.Result)
   > &RigUnit_ParentConstraint(
    AdvancedSettings=(InterpolationType=Average,RotationOrderForFilter=XZY)
    Child=(Type=Bone,Name="MSTR_jacket_front_r")
    Filter=(TranslationFilter=(bX=true,bY=true,bZ=false),RotationFilter=(bX=true,bY=true,bZ=true),ScaleFilter=(bX=True,bY=True,bZ=True))
    Parents=((Item=(Type=Bone,Name="pelvis"),Weight=1))
    Weight=&RigVMFunction_MathDoubleRemap(SourceMaximum=1, SourceMinimum=0, TargetMaximum=1, TargetMinimum=0, Value=@n20.Result).Result
   )
   > &RigUnit_ParentConstraint(
    AdvancedSettings=(InterpolationType=Average,RotationOrderForFilter=XZY)
    Child=(Type=Bone,Name="MSTR_jacket_front_r")
    Filter=(TranslationFilter=(bX=false,bY=false,bZ=true),RotationFilter=(bX=True,bY=True,bZ=True),ScaleFilter=(bX=True,bY=True,bZ=True))
    Parents=((Item=(Type=Bone,Name="pelvis"),Weight=0.5),(Item=(Type=Bone,Name="thigh_r"),Weight=0.5))
    Weight=@n20.Result
   )
   > &RigUnit_ParentConstraint(
    AdvancedSettings=(InterpolationType=Average,RotationOrderForFilter=XZY)
    Child=(Type=Bone,Name="MSTR_jacket_side_r")
    Filter=(TranslationFilter=(bX=true,bY=true,bZ=true),RotationFilter=(bX=true,bY=true,bZ=True),ScaleFilter=(bX=True,bY=True,bZ=True))
    Parents=((Item=(Type=Bone,Name="pelvis"),Weight=1),(Item=(Type=Bone,Name="thigh_r"),Weight=1))
    Weight=@n20.Result
   )
   n21 = &RigVMFunction_MathDoubleRemap(
    SourceMaximum=70
    SourceMinimum=-70
    TargetMaximum=1
    TargetMinimum=-1
    Value=&RigVMFunction_MathQuaternionToEuler(
     RotationOrder=ZYX
     Value=&RigUnit_GetRelativeTransformForItem(Child=(Type=Bone,Name="spine_05"), Parent=(Type=Bone,Name="pelvis"), bChildInitial=False, bParentInitial=False).RelativeTransform.Rotation
    ).Result.X
   )
   n22 = &RigVMFunction_MathDoubleClamp(Maximum=0, Minimum=-1, Value=@n21.Result)
   n23 = &RigVMFunction_MathDoubleMul(
    B=@KalmanFilter.Result
    A=&RigVMFunction_MathDoubleRemap(SourceMaximum=0, SourceMinimum=1, TargetMaximum=1, TargetMinimum=0, Value=@n19.Result).Result
   )
   > &RigUnit_ModifyTransforms(
    ItemToModify="((Item=(Type=Bone,Name=\"MSTR_jacket_front_l\"),Transform=(Rotation=(X=-0.315,Y=-0.1946,Z=0.1736,W=0.9125),Translation=(X=0.1854,Y=-1.3861,Z=0.8652),Scale3D=(X=1,Y=1,Z=1))),(Item=(Type=Bone,Name=\"MSTR_jacket_front_r\"),Transform=(Rotation=(X=-0.0409,Y=0.1423, ...(truncated)"
    WeightMinimum=0
    Weight=&RigVMFunction_MathDoubleMul(B=&RigVMFunction_MathDoubleMul(B=-1, A=@n22.Result).Result, A=@n23.Result).Result
   )
   > &RigUnit_ModifyTransforms(
    ItemToModify="((Item=(Type=Bone,Name=\"MSTR_jacket_front_l\"),Transform=(Rotation=(X=-0.0738,Y=0.0641,Z=-0.3746,W=0.922),Translation=(X=-5.2218,Y=1.2509,Z=2.345),Scale3D=(X=1,Y=1,Z=1))),(Item=(Type=Bone,Name=\"MSTR_jacket_front_r\"),Transform=(Rotation=(X=-0.0798,Y=-0.073,Z ...(truncated)"
    WeightMinimum=0
    Weight=&RigVMFunction_MathDoubleMul(A=@n21.Result, B=@n23.Result).Result
   )
   &RigUnit_ModifyTransforms(
    ItemToModify=((Item=(Type=Bone,Name="MSTR_pants_bk_l"),Transform=(Rotation=(X=0,Y=0,Z=0,W=1),Translation=(X=-5,Y=-10,Z=0),Scale3D=(X=0.1,Y=1,Z=1))))
    WeightMinimum=0
    Weight=&RigVMFunction_MathDoubleRemap(SourceMaximum=0, SourceMinimum=1, TargetMaximum=1, TargetMinimum=0, Value=RigVMVariableNode() {variable=CalfL}.Value).Result
    ExecutePin=@RigVMFunction_Sequence.D
   )
   > &RigUnit_ModifyTransforms(
    ItemToModify=((Item=(Type=Bone,Name="MSTR_pants_bk_r"),Transform=(Rotation=(X=0,Y=0,Z=0,W=1),Translation=(X=-5,Y=-10,Z=0),Scale3D=(X=0.1,Y=1,Z=1))))
    WeightMinimum=0
    Weight=&RigVMFunction_MathDoubleRemap(SourceMaximum=0, SourceMinimum=1, TargetMaximum=1, TargetMinimum=0, Value=RigVMVariableNode() {variable=CalfR}.Value).Result
   )
   > &RigUnit_ParentConstraint(
    AdvancedSettings=(InterpolationType=Average,RotationOrderForFilter=XZY)
    Child=(Type=Bone,Name="def_foot_corr_fr_l")
    Filter=(TranslationFilter=(bX=True,bY=True,bZ=True),RotationFilter=(bX=false,bY=false,bZ=false),ScaleFilter=(bX=True,bY=True,bZ=True))
    Parents=((Item=(Type=Bone,Name="foot_l"),Weight=0.5),(Item=(Type=Bone,Name="calf_twist_01_l"),Weight=0.5))
    Weight=1
    bMaintainOffset=true
   )
   > &RigUnit_ParentConstraint(
    AdvancedSettings=(InterpolationType=Average,RotationOrderForFilter=XZY)
    Child=(Type=Bone,Name="def_foot_corr_fr_r")
    Filter=(TranslationFilter=(bX=true,bY=True,bZ=True),RotationFilter=(bX=false,bY=false,bZ=false),ScaleFilter=(bX=True,bY=True,bZ=True))
    Parents=((Item=(Type=Bone,Name="foot_r"),Weight=0.5),(Item=(Type=Bone,Name="calf_twist_01_r"),Weight=0.5))
    Weight=1
    bMaintainOffset=true
   )
   n24 = &RigVMFunction_MathQuaternionToEuler(
    RotationOrder=ZYX
    Value=&RigUnit_GetTransform(Item=(Type=Bone,Name="foot_r"), Space=LocalSpace, bInitial=False).Transform.Rotation
   )
   > &RigUnit_ModifyTransforms(
    ItemToModify=((Item=(Type=Bone,Name="def_foot_corr_fr_r"),Transform=(Rotation=(X=0,Y=0,Z=0.1736,W=0.9848),Translation=(X=2,Y=-3,Z=0.0001),Scale3D=(X=0.8497,Y=1,Z=1))))
    WeightMinimum=0
    Weight=&RigVMFunction_MathDoubleRemap(SourceMaximum=-60, SourceMinimum=-20, TargetMaximum=1, TargetMinimum=0, Value=@n24.Result.Z).Result
   )
   n25 = &RigVMFunction_MathQuaternionToEuler(
    RotationOrder=ZYX
    Value=&RigUnit_GetTransform(Item=(Type=Bone,Name="foot_l"), Space=LocalSpace, bInitial=False).Transform.Rotation
   )
   > &RigUnit_ModifyTransforms(
    ItemToModify=((Item=(Type=Bone,Name="def_foot_corr_fr_l"),Transform=(Rotation=(X=0,Y=0,Z=0.1736,W=0.9848),Translation=(X=2,Y=-3,Z=0),Scale3D=(X=0.7472,Y=1,Z=1))))
    WeightMinimum=0
    Weight=&RigVMFunction_MathDoubleRemap(SourceMaximum=-60, SourceMinimum=-20, TargetMaximum=1, TargetMinimum=0, Value=@n25.Result.Z).Result
   )
   > &RigUnit_ModifyTransforms(
    ItemToModify="((Item=(Type=Bone,Name=\"def_foot_corr_bk_r\"),Transform=(Rotation=(X=0.0644,Y=0.0453,Z=0.3246,W=-0.9426),Translation=(X=6.5683,Y=-3.8648,Z=-0.2327),Scale3D=(X=1,Y=1,Z=1))),(Item=(Type=Bone,Name=\"def_foot_corr_fr_r\"),Transform=(Rotation=(X=0,Y=0,Z=0.2164,W=0 ...(truncated)"
    WeightMinimum=0
    Weight=&RigVMFunction_MathDoubleRemap(SourceMaximum=65, SourceMinimum=0, TargetMaximum=1, TargetMinimum=0, Value=@n24.Result.Z).Result
   )
   > &RigUnit_ModifyTransforms(
    ItemToModify="((Item=(Type=Bone,Name=\"def_foot_corr_bk_l\"),Transform=(Rotation=(X=0.0926,Y=-0.0867,Z=-0.1063,W=0.9862),Translation=(X=5.9568,Y=-2.6848,Z=-0.4558),Scale3D=(X=1,Y=1,Z=1))),(Item=(Type=Bone,Name=\"def_foot_corr_fr_l\"),Transform=(Rotation=(X=0,Y=0,Z=0.2128,W= ...(truncated)"
    WeightMinimum=0
    Weight=&RigVMFunction_MathDoubleRemap(SourceMaximum=65, SourceMinimum=0, TargetMaximum=1, TargetMinimum=0, Value=@n25.Result.Z).Result
   )
   n26 = &RigUnit_GetRelativeTransformForItem(Child=(Type=Bone,Name="head"), Parent=(Type=Bone,Name="spine_05"), bChildInitial=False, bParentInitial=False)
   > &RigUnit_ModifyTransforms(
    ItemToModify=((Item=(Type=Bone,Name="dyn_hoodie"),Transform=(Rotation=(X=0.0664,Y=0.122,Z=0.0038,W=0.9903),Translation=(X=-0.6452,Y=-2.0233,Z=-0.0926),Scale3D=(X=1,Y=1,Z=1))))
    WeightMinimum=0
    Weight=&RigVMFunction_MathDoubleRemap(
     SourceMaximum=-50
     SourceMinimum=0
     TargetMaximum=1
     TargetMinimum=0
     Value=&RigVMFunction_MathQuaternionToEuler(RotationOrder=ZYX, Value=@n26.RelativeTransform.Rotation).Result.Z
    ).Result
   )
   RigVMFunction_MathQuaternionToEuler = &RigVMFunction_MathQuaternionToEuler(
    RotationOrder=ZYX
    Value=&RigUnit_GetRelativeTransformForItem(Child=(Type=Bone,Name="spine_03"), Parent=(Type=Bone,Name="pelvis"), bChildInitial=False, bParentInitial=False).RelativeTransform.Rotation
   )
   Remap = &RigVMFunction_MathDoubleRemap(SourceMaximum=-60, SourceMinimum=-20, TargetMaximum=1, TargetMinimum=0, Value=@RigVMFunction_MathQuaternionToEuler.Result.Z)
   > &RigUnit_ModifyTransforms(
    ItemToModify="((Item=(Type=Bone,Name=\"MSTR_jacket_bk_l\"),Transform=(Rotation=(X=-0.1509,Y=0.0231,Z=0.2767,W=-0.9487),Translation=(X=-2.1697,Y=-7.9956,Z=-2.9849),Scale3D=(X=1,Y=1,Z=1))),(Item=(Type=Bone,Name=\"MSTR_jacket_bk_r\"),Transform=(Rotation=(X=0.111,Y=0.2427,Z=0.2 ...(truncated)"
    WeightMinimum=0
    Weight=@Remap.Result
   )
   > &RigUnit_ModifyTransforms(
    ItemToModify="((Item=(Type=Bone,Name=\"MSTR_belt_fr\"),Transform=(Rotation=(X=0,Y=0,Z=0,W=1),Translation=(X=0,Y=-3,Z=0),Scale3D=(X=1,Y=1,Z=1))),(Item=(Type=Bone,Name=\"MSTR_jacket_side_l\"),Transform=(Rotation=(X=0,Y=0,Z=-0.2164,W=0.9763),Translation=(X=0,Y=0,Z=0),Scale3D=( ...(truncated)"
    WeightMinimum=0
    Weight=&RigVMFunction_MathBoolToFloat(Value=RigVMVariableNode() {variable=isSkyDiveFloat}.Value).Result
   )
   > n27 = &RigUnit_SphericalPoseReader(
    ActiveRegionScaleFactors=(PositiveWidth=1,NegativeWidth=2,PositiveHeight=2,NegativeHeight=2)
    ActiveRegionSize=0.1
    Debug=(bDrawDebug=false,bDraw2D=False,bDrawLocalAxes=False,DebugScale=25,DebugSegments=20,DebugThickness=0.25)
    DriverAxis=(X=1,Y=0,Z=0)
    DriverItem=(Type=Bone,Name="upperarm_l")
    FalloffRegionScaleFactors=(PositiveWidth=0.7,NegativeWidth=1,PositiveHeight=1,NegativeHeight=1)
    FalloffSize=0.3
    FlipHeightScaling=False
    FlipWidthScaling=False
    OptionalParentItem=(Type=Bone,Name="clavicle_l")
    RotationOffset=(X=0,Y=-55,Z=0)
   )
   > n28 = &RigUnit_SphericalPoseReader(
    ActiveRegionScaleFactors=(PositiveWidth=1,NegativeWidth=2,PositiveHeight=2,NegativeHeight=2)
    ActiveRegionSize=0.1
    Debug=(bDrawDebug=false,bDraw2D=False,bDrawLocalAxes=False,DebugScale=25,DebugSegments=20,DebugThickness=0.25)
    DriverAxis=(X=-1,Y=0,Z=0)
    DriverItem=(Type=Bone,Name="upperarm_r")
    FalloffRegionScaleFactors=(PositiveWidth=0.7,NegativeWidth=1,PositiveHeight=1,NegativeHeight=1)
    FalloffSize=0.3
    FlipHeightScaling=true
    FlipWidthScaling=false
    OptionalParentItem=(Type=Bone,Name="clavicle_r")
    RotationOffset=(X=0,Y=125,Z=0)
   )
   n29 = &RigVMFunction_MathBoolToFloat(Value=RigVMVariableNode() {variable=isParachute}.Value)
   > &RigUnit_ModifyTransforms(
    ItemToModify="((Item=(Type=Bone,Name=\"MSTR_jacket_front_l\"),Transform=(Rotation=(X=0,Y=0,Z=-0.0436,W=0.999),Translation=(X=-1,Y=-1,Z=-1),Scale3D=(X=1,Y=1,Z=1))),(Item=(Type=Bone,Name=\"MSTR_jacket_front_r\"),Transform=(Rotation=(X=0,Y=0,Z=-0.0436,W=0.999),Translation=(X=- ...(truncated)"
    WeightMinimum=0
    Weight=&RigVMFunction_MathDoubleMul(B=@n29.Result, A=&RigVMFunction_MathDoubleMul(A=@n27.OutputParam, B=@n28.OutputParam).Result).Result
   )
   RigVMFunction_AlphaInterp = &RigVMFunction_AlphaInterp(InterpSpeedDecreasing=10, InterpSpeedIncreasing=10, Scale=1, bClampResult=False, bInterpResult=True, Value=@n29.Result)
   > &RigUnit_ModifyTransforms(
    ItemToModify="((Item=(Type=Bone,Name=\"MSTR_jacket_side_l\"),Transform=(Rotation=(X=0,Y=0,Z=-0.1305,W=0.9914),Translation=(X=0,Y=0,Z=0),Scale3D=(X=1,Y=1,Z=1))),(Item=(Type=Bone,Name=\"MSTR_jacket_side_r\"),Transform=(Rotation=(X=0,Y=0,Z=-0.1305,W=0.9914),Translation=(X=0,Y= ...(truncated)"
    WeightMinimum=0
    Weight=@RigVMFunction_AlphaInterp.Result
   )
   > &RigUnit_ParentConstraint(
    AdvancedSettings=(InterpolationType=Average,RotationOrderForFilter=XZY)
    Child=(Type=Bone,Name="def_beltRing")
    Parents="((Item=(Type=Bone,Name=\"spine_04\"),Weight=0.0029),(Item=(Type=Bone,Name=\"spine_02\"),Weight=0.0277),(Item=(Type=Bone,Name=\"pelvis\"),Weight=0.7076),(Item=(Type=Bone,Name=\"thigh_twist_01_l\"),Weight=0.0404),(Item=(Type=Bone,Name=\"thigh_twist_01_r\"),Weight=0.04 ...(truncated)"
    Weight=1
   )
   n30 = &RigVMFunction_MathDoubleRemap(
    SourceMaximum=-50
    SourceMinimum=-15
    TargetMaximum=1
    TargetMinimum=0
    Value=&RigVMFunction_MathQuaternionToEuler(
     RotationOrder=ZYX
     Value=&RigUnit_GetRelativeTransformForItem(Child=(Type=Bone,Name="thigh_r"), Parent=(Type=Bone,Name="thigh_r"), bChildInitial=False, bParentInitial=true).RelativeTransform.Rotation
    ).Result.Z
   )
   n31 = &RigVMFunction_MathDoubleRemap(
    SourceMaximum=-50
    SourceMinimum=-15
    TargetMaximum=1
    TargetMinimum=0
    Value=&RigVMFunction_MathQuaternionToEuler(
     RotationOrder=ZYX
     Value=&RigUnit_GetRelativeTransformForItem(Child=(Type=Bone,Name="thigh_r"), Parent=(Type=Bone,Name="thigh_r"), bChildInitial=False, bParentInitial=true).RelativeTransform.Rotation
    ).Result.Z
   )
   > &RigUnit_ModifyTransforms(
    ItemToModify=((Item=(Type=Bone,Name="def_beltRing"),Transform=(Rotation=(X=0,Y=0,Z=0,W=1),Translation=(X=0,Y=0,Z=0),Scale3D=(X=0.8,Y=0.8,Z=0.8))))
    WeightMinimum=0
    Weight=&RigVMFunction_MathDoubleClamp(Value=&RigVMFunction_MathDoubleAdd(A=@n30.Result, B=@n31.Result).Result).Result
   )
   n32 = RigVMVariableNode() {variable=isRBAN}
   > &RigUnit_ModifyTransforms(
    ItemToModify=((Item=(Type=Bone,Name="def_altSpineFixLOD"),Transform=(Rotation=(X=0,Y=0,Z=0,W=1),Translation=(X=0,Y=0,Z=0),Scale3D=(X=0,Y=0,Z=0))))
    WeightMinimum=0
    Weight=&RigVMFunction_MathBoolToFloat(Value=&RigVMFunction_MathBoolNot(Value=@n32.Value).Result).Result
   )
   RigVMVariableNode() {variable=isSliding}
  ]}
]
hierarchy [
 root {loc=<0 0 0>, rot=<0 0 0 1>}
 pelvis {parent=root, loc=<0 -0.5301 98.0571>, rot=<0.0127 0.707 -0.0127 -0.707>}
 spine_01 {parent=pelvis, loc=<6.1208 -0.9328 0>, rot=<0 0 0.034 -0.9994>}
 spine_02 {parent=spine_01, loc=<6.5997 -0.2095 0>, rot=<0 0 0.0238 -0.9997>}
 spine_03 {parent=spine_02, loc=<6.4914 -0.1237 0>, rot=<0 0 0.0149 -0.9999>}
 spine_04 {parent=spine_03, loc=<6.9598 0.0794 0>, rot=<0 0 -0.0132 -0.9999>}
 spine_05 {parent=spine_04, loc=<6.9987 0 0>, rot=<0 0 -0.0015 -1>}
 neck_01 {parent=spine_05, loc=<11.3138 0.1529 0>, rot=<0 0 -0.2003 -0.9797>}
 neck_02 {parent=neck_01, loc=<4.5627 0 0>, rot=<0 0 0.0504 -0.9987>}
 head {parent=neck_02, loc=<5.3914 0 0>, rot=<0 0 0.115 -0.9934>}
 clavicle_l {parent=spine_05, loc=<5.4831 -0.1053 -3.2946>, rot=<0.0465 -0.7289 -0.009 -0.6829>}
 upperarm_l {parent=clavicle_l, loc=<9.8658 -1.2136 -0.4105>, rot=<0.0164 -0.3245 -0.0269 -0.9454>}
 lowerarm_l {parent=upperarm_l, loc=<26.465 -0.8345 0.259>, rot=<0.0036 0.0161 -0.2298 -0.9731>}
 hand_l {parent=lowerarm_l, loc=<22.7355 -0.3203 -0.6105>, rot=<-0.6784 -0.0223 0.0735 -0.7307>}
 thumb_01_l {parent=hand_l, loc=<1.3118 -0.3191 -2.459>, rot=<0.6546 -0.4042 0.0664 -0.6354>}
 thumb_02_l {parent=thumb_01_l, loc=<5.9336 -0.0448 0.0401>, rot=<-0.0037 0.0213 0.298 -0.9543>}
 thumb_03_l {parent=thumb_02_l, loc=<2.2223 0.3682 -0.0907>, rot=<0.0118 -0.013 0.0066 -0.9998>}
 index_metacarpal_l {parent=hand_l, loc=<1.6535 0.0183 -2.96>, rot=<0.0158 -0.0124 -0.0173 -0.9996>}
 index_01_l {parent=index_metacarpal_l, loc=<8.3008 0.0357 -0.0524>, rot=<-0.0044 0.0485 0.0924 -0.9945>}
 index_02_l {parent=index_01_l, loc=<3.3587 -0.3004 0.0452>, rot=<-0.002 0.0027 0.3225 -0.9466>}
 index_03_l {parent=index_02_l, loc=<2.5578 -0.123 -0.0252>, rot=<-0.002 0.0067 0.2558 -0.9667>}
 middle_metacarpal_l {parent=hand_l, loc=<2.1856 -0.1725 -0.5984>, rot=<0.006 0.0168 -0.0329 -0.9993>}
 middle_01_l {parent=middle_metacarpal_l, loc=<7.3275 -0.2447 0.0129>, rot=<-0.0208 0.0042 0.0858 -0.9961>}
 middle_02_l {parent=middle_01_l, loc=<4.0123 0.1401 0.1175>, rot=<-0.0218 0.0258 0.3512 -0.9357>}
 middle_03_l {parent=middle_02_l, loc=<2.6077 0.0534 0.1373>, rot=<0.0082 0.0188 0.1708 -0.9851>}
 ring_metacarpal_l {parent=hand_l, loc=<2.2074 -0.3625 1.3638>, rot=<-0.0284 0.0722 -0.0458 -0.9959>}
 ring_01_l {parent=ring_metacarpal_l, loc=<7.0295 -0.6197 -0.1446>, rot=<-0.0489 -0.0252 0.1892 -0.9804>}
 ring_02_l {parent=ring_01_l, loc=<3.6174 0.4647 -0.1027>, rot=<0 0 0.2806 -0.9598>}
 ring_03_l {parent=ring_02_l, loc=<2.4152 0.114 0.0199>, rot=<0 0 0.1189 -0.9929>}
 pinky_metacarpal_l {parent=hand_l, loc=<1.7616 -0.9753 3.3814>, rot=<-0.0397 0.0814 -0.069 -0.9935>}
 pinky_01_l {parent=pinky_metacarpal_l, loc=<7.1709 -0.4831 -0.1094>, rot=<-0.0654 -0.0035 0.2242 -0.9723>}
 pinky_02_l {parent=pinky_01_l, loc=<2.8274 0.0893 -0.1043>, rot=<0.0037 0.0011 0.2854 -0.9584>}
 pinky_03_l {parent=pinky_02_l, loc=<2.2046 0.0307 -0.0078>, rot=<-0.0002 0.0098 0.1445 -0.9895>}
 deform_wrist_dn_base_l {parent=hand_l, loc=<0 -1.5326 0>, rot=<0 0 0 -1>}
 deform_wrist_dn_l {parent=deform_wrist_dn_base_l, loc=<0 0 0>, rot=<0 0 0 1>}
 deform_wrist_up_base_l {parent=hand_l, loc=<0 2.3009 0>, rot=<0 0 0 -1>}
 deform_wrist_up_l {parent=deform_wrist_up_base_l, loc=<0 0 0>, rot=<0 0 0 1>}
 lowerarm_twist_02_l {parent=lowerarm_l, loc=<10.6707 0 0>, rot=<0 0 0 1>}
 lowerarm_twist_01_l {parent=lowerarm_l, loc=<18.316 0 0>, rot=<0 0 0 1>}
 deform_elbow_in_base_l {parent=lowerarm_l, loc=<0.1834 3.6062 -0.2572>, rot=<0 0 0 -1>}
 deform_elbow_in_l {parent=deform_elbow_in_base_l, loc=<0 0 0>, rot=<0 0 0 1>}
 deform_elbow_out_base_l {parent=lowerarm_l, loc=<0 -4 0>, rot=<0 0 0 1>}
 deform_elbow_out_l {parent=deform_elbow_out_base_l, loc=<0 0 0>, rot=<0 0 0 1>}
 upperarm_twist_01_l {parent=upperarm_l, loc=<7.2135 -0.1134 0>, rot=<0 0 0 1>}
 deform_upperarm_bk_base_l {parent=upperarm_twist_01_l, loc=<-4.3788 -6.3931 -1.4877>, rot=<0 0 0 -1>}
 deform_upperarm_bk_l {parent=deform_upperarm_bk_base_l, loc=<0 0 0>, rot=<0 0 0 1>}
 deform_upperarm_fr_base_l {parent=upperarm_twist_01_l, loc=<-3.4526 7.5068 -0.2668>, rot=<0 0 0 -1>}
 deform_upperarm_fr_l {parent=deform_upperarm_fr_base_l, loc=<0 0 0>, rot=<0 0 0 1>}
 deform_upperarm_up_base_l {parent=upperarm_twist_01_l, loc=<-6.3503 0.4825 4.8307>, rot=<0 0 0 -1>}
 deform_upperarm_up_l {parent=deform_upperarm_up_base_l, loc=<0 0 0>, rot=<0 0 0 1>}
 upperarm_twist_02_l {parent=upperarm_l, loc=<19.7998 -0.1134 0>, rot=<0 0 0 1>}
 dyn_armRing_l {parent=upperarm_twist_02_l, loc=<1.5155 -1.4039 7.4429>, rot=<0.6778 0.0803 0.0961 -0.7245>}
 clavicle_r {parent=spine_05, loc=<5.4831 -0.1053 3.2946>, rot=<0.7289 0.0465 -0.6829 0.009>}
 upperarm_r {parent=clavicle_r, loc=<-9.8658 1.2136 0.4105>, rot=<0.0164 -0.3245 -0.0269 -0.9454>}
 lowerarm_r {parent=upperarm_r, loc=<-26.465 0.8345 -0.259>, rot=<0.0036 0.0161 -0.2298 -0.9731>}
 hand_r {parent=lowerarm_r, loc=<-22.7355 0.3203 0.6106>, rot=<-0.6784 -0.0223 0.0735 -0.7307>}
 thumb_01_r {parent=hand_r, loc=<-1.3118 0.3191 2.459>, rot=<0.6546 -0.4042 0.0664 -0.6354>}
 thumb_02_r {parent=thumb_01_r, loc=<-5.9336 0.0448 -0.0401>, rot=<-0.0037 0.0213 0.298 -0.9543>}
 thumb_03_r {parent=thumb_02_r, loc=<-2.2223 -0.3682 0.0907>, rot=<0.0118 -0.013 0.0066 -0.9998>}
 index_metacarpal_r {parent=hand_r, loc=<-1.6535 -0.0183 2.96>, rot=<0.0158 -0.0124 -0.0173 -0.9996>}
 index_01_r {parent=index_metacarpal_r, loc=<-8.3008 -0.0357 0.0524>, rot=<-0.0044 0.0485 0.0924 -0.9945>}
 index_02_r {parent=index_01_r, loc=<-3.3587 0.3004 -0.0452>, rot=<-0.002 0.0027 0.3225 -0.9466>}
 index_03_r {parent=index_02_r, loc=<-2.5578 0.123 0.0252>, rot=<-0.002 0.0067 0.2558 -0.9667>}
 middle_metacarpal_r {parent=hand_r, loc=<-2.1856 0.1725 0.5984>, rot=<0.006 0.0168 -0.0329 -0.9993>}
 middle_01_r {parent=middle_metacarpal_r, loc=<-7.3275 0.2447 -0.0129>, rot=<-0.0208 0.0042 0.0858 -0.9961>}
 middle_02_r {parent=middle_01_r, loc=<-4.0123 -0.1401 -0.1175>, rot=<-0.0218 0.0258 0.3512 -0.9357>}
 middle_03_r {parent=middle_02_r, loc=<-2.6077 -0.0534 -0.1373>, rot=<0.0082 0.0188 0.1708 -0.9851>}
 ring_metacarpal_r {parent=hand_r, loc=<-2.2074 0.3625 -1.3638>, rot=<-0.0284 0.0722 -0.0458 -0.9959>}
 ring_01_r {parent=ring_metacarpal_r, loc=<-7.0295 0.6197 0.1446>, rot=<-0.0489 -0.0252 0.1892 -0.9804>}
 ring_02_r {parent=ring_01_r, loc=<-3.6174 -0.4647 0.1027>, rot=<0 0 0.2806 -0.9598>}
 ring_03_r {parent=ring_02_r, loc=<-2.4152 -0.114 -0.0199>, rot=<0 0 0.1189 -0.9929>}
 pinky_metacarpal_r {parent=hand_r, loc=<-1.7616 0.9753 -3.3814>, rot=<-0.0397 0.0814 -0.069 -0.9935>}
 pinky_01_r {parent=pinky_metacarpal_r, loc=<-7.1709 0.4831 0.1094>, rot=<-0.0654 -0.0035 0.2242 -0.9723>}
 pinky_02_r {parent=pinky_01_r, loc=<-2.8274 -0.0893 0.1043>, rot=<0.0037 0.0011 0.2854 -0.9584>}
 pinky_03_r {parent=pinky_02_r, loc=<-2.2046 -0.0307 0.0078>, rot=<-0.0002 0.0098 0.1445 -0.9895>}
 deform_wrist_dn_base_r {parent=hand_r, loc=<-0.0001 1.5328 0>, rot=<0 0 0 -1>}
 deform_wrist_dn_r {parent=deform_wrist_dn_base_r, loc=<0 0 0>, rot=<0 0 0 1>}
 deform_wrist_up_base_r {parent=hand_r, loc=<-0.0002 -2.3007 -0.0001>, rot=<0 0 0 -1>}
 deform_wrist_up_r {parent=deform_wrist_up_base_r, loc=<0 0 0>, rot=<0 0 0 1>}
 lowerarm_twist_02_r {parent=lowerarm_r, loc=<-10.6707 0 0>, rot=<0 0 0 -1>}
 lowerarm_twist_01_r {parent=lowerarm_r, loc=<-18.316 0 0>, rot=<0 0 0 -1>}
 deform_elbow_in_base_r {parent=lowerarm_r, loc=<-0.1834 -3.6062 0.2572>, rot=<0 0 0 -1>}
 deform_elbow_in_r {parent=deform_elbow_in_base_r, loc=<0 0 0>, rot=<0 0 0 1>}
 deform_elbow_out_base_r {parent=lowerarm_r, loc=<0 4 -0.0001>, rot=<0 0 0 -1>}
 deform_elbow_out_r {parent=deform_elbow_out_base_r, loc=<0 0 0>, rot=<0 0 0 1>}
 upperarm_twist_01_r {parent=upperarm_r, loc=<-7.2135 0.1134 0>, rot=<0 0 0 -1>}
 deform_upperarm_bk_base_r {parent=upperarm_twist_01_r, loc=<3.849 5.824 -0.333>, rot=<0 0 0 -1>}
 deform_upperarm_bk_r {parent=deform_upperarm_bk_base_r, loc=<0 0 0>, rot=<0 0 0 1>}
 deform_upperarm_fr_base_r {parent=upperarm_twist_01_r, loc=<3 -5.554 -0.675>, rot=<0 0 0 -1>}
 deform_upperarm_fr_r {parent=deform_upperarm_fr_base_r, loc=<0 0 0>, rot=<0 0 0 1>}
 deform_upperarm_up_base_r {parent=upperarm_twist_01_r, loc=<6.35 -0.482 -4.831>, rot=<0 0 0 -1>}
 deform_upperarm_up_r {parent=deform_upperarm_up_base_r, loc=<0 0 0>, rot=<0 0 0 1>}
 upperarm_twist_02_r {parent=upperarm_r, loc=<-19.7998 0.1134 0>, rot=<0 0 0 -1>}
 dyn_armRing_r {parent=upperarm_twist_02_r, loc=<-1.5154 1.4039 -7.443>, rot=<-0.0803 0.6778 0.7245 0.0961>}
 def_egg06 {parent=clavicle_r, loc=<-2.992 10.753 3.7529>, rot=<0.9271 0.0685 -0.3448 0.13>}
 def_egg05 {parent=clavicle_r, loc=<-6.4914 8.9097 -0.18>, rot=<0.9271 0.0685 -0.3448 0.13>}
 def_egg04 {parent=clavicle_r, loc=<-1.9777 -10.1365 2.68>, rot=<-0.0586 -0.295 -0.0027 -0.9537>}
 attach_backpack {parent=spine_05, loc=<-0.123 -13.473 0>, rot=<-0.0093 -0.707 0.0093 -0.707>}
 Backpack_BR {parent=attach_backpack, type=Null, loc=<0 -1.7198 0.0352>, rot=<-0.0238 0 0 0.9997>}
 attach_cape {parent=spine_05, loc=<9.017 -11.714 0>, rot=<0.0616 -0.7044 -0.0616 -0.7044>}
 cape {parent=attach_cape, type=Null, loc=<0 -1.3698 0>, rot=<0 0 0 1>}
 dyn_hoodie {parent=spine_05, loc=<12.3798 -9.6335 -0.1322>, rot=<0.4574 -0.5346 0.5564 -0.4422>}
 deform_pec_aim_base_l {parent=spine_04, loc=<3.2456 9.3309 -1.2237>, rot=<0.0603 -0.4033 0.3318 -0.8507>}
 deform_pec_aim_l {parent=deform_pec_aim_base_l, loc=<0 0 0>, rot=<0 0 0 1>}
 deform_pec_base_l {parent=deform_pec_aim_l, loc=<9.3665 0 0>, rot=<0 0 0 1>}
 deform_pec_l {parent=deform_pec_base_l, loc=<0 0 0>, rot=<0 0 0 1>}
 deform_pec_aim_base_r {parent=spine_04, loc=<3.2452 9.3309 1.2237>, rot=<-0.4033 -0.0603 0.8507 0.3318>}
 deform_pec_aim_r {parent=deform_pec_aim_base_r, loc=<0 0 0>, rot=<0 0 0 1>}
 deform_pec_base_r {parent=deform_pec_aim_r, loc=<-9.3667 -0.0003 -0.0004>, rot=<0 0 0 1>}
 deform_pec_r {parent=deform_pec_base_r, loc=<0 0 0>, rot=<0 0 0 1>}
 MSTR_bag {parent=spine_04, loc=<-0.2333 -0.4446 -3.2089>, rot=<0.0332 -0.9466 0.0062 -0.3207>}
 MSTR_belt_bk {parent=MSTR_bag, loc=<0.7573 -13.1313 0.6218>, rot=<0.996 0.08 0.0312 0.0229>}
 belt2 {parent=MSTR_belt_bk, loc=<0 0 0>, rot=<0 0 0 1>}
 belt3 {parent=belt2, loc=<13.8248 0 0>, rot=<0.2512 -0.026 0.1529 -0.9554>}
 belt6 {parent=belt3, loc=<11.3877 0 0>, rot=<-0.6225 0.1194 0.2164 -0.7426>}
 belt7 {parent=belt6, loc=<10.042 0 0>, rot=<0 0.0735 0 -0.9973>}
 def_egg01 {parent=belt3, loc=<2.8618 0.6044 0.1173>, rot=<0.8751 0.3102 -0.3474 -0.1318>}
 def_egg02 {parent=belt3, loc=<-2.6741 -0.3173 -0.2399>, rot=<0.8751 0.3102 -0.3474 -0.1318>}
 MSTR_belt_fr {parent=MSTR_bag, loc=<-1.5776 12.5049 -0.7061>, rot=<0.9964 0.0798 0.0202 0.0221>}
 belt1 {parent=MSTR_belt_fr, loc=<0 0 0>, rot=<0 0 0 1>}
 belt02 {parent=belt1, loc=<5.9985 -0.0772 0.0118>, rot=<0 0 0 1>}
 belt4 {parent=belt02, loc=<5.3596 0.0772 -0.0118>, rot=<0.002 -0.0145 -0.0189 -0.9997>}
 belt5 {parent=belt4, loc=<8.2035 0 0>, rot=<0 0 0 1>}
 def_egg08 {parent=belt4, loc=<-1.7599 -0.0281 -0.2657>, rot=<0.9103 0.0548 -0.4096 0.0247>}
 def_egg07 {parent=belt4, loc=<3.8945 -0.3818 0.178>, rot=<0.9103 0.0548 -0.4096 0.0247>}
 def_egg03 {parent=spine_04, loc=<4.7301 10.9652 2.0877>, rot=<0.8968 -0.0188 -0.4305 -0.1>}
 MSTR_jacket_side_l {parent=spine_03, loc=<-0.7423 -0.8474 -16.5355>, rot=<-0.0796 0.7029 0.7069 -0.0021>}
 dyn_jacket_side_01_l {parent=MSTR_jacket_side_l, loc=<0 0 0>, rot=<0 0 0 1>}
 dyn_jacket_side_02_l {parent=dyn_jacket_side_01_l, loc=<11.0425 0 0>, rot=<0.0002 0.0033 -0.061 -0.9981>}
 dyn_jacket_side_03_l {parent=dyn_jacket_side_02_l, loc=<8.6388 0 0>, rot=<0 0 0.0205 -0.9998>}
 MSTR_jacket_front_l {parent=spine_03, loc=<2.0598 10.7763 -12.0069>, rot=<-0.0232 0.2575 0.9656 -0.0273>}
 dyn_jacket_front_01_l {parent=MSTR_jacket_front_l, loc=<0 0 0>, rot=<0 0 0 1>}
 dyn_jacket_front_02_l {parent=dyn_jacket_front_01_l, loc=<14.997 0 0>, rot=<0.0007 0.0023 -0.0747 -0.9972>}
 dyn_jacket_front_03_l {parent=dyn_jacket_front_02_l, loc=<9.7652 0 0>, rot=<0 0 0.0497 -0.9988>}
 dyn_chicketnbag {parent=dyn_jacket_front_01_l, loc=<18.7317 -2.7328 -6.118>, rot=<0.7003 -0.5629 -0.3962 0.1892>}
 MSTR_jacket_side_r {parent=spine_03, loc=<-0.7426 -0.8474 16.5355>, rot=<0.0796 -0.7029 0.7069 -0.0021>}
 dyn_jacket_side_01_r {parent=MSTR_jacket_side_r, loc=<0 0 0>, rot=<0 0 0 1>}
 dyn_jacket_side_02_r {parent=dyn_jacket_side_01_r, loc=<11.0427 0 0>, rot=<-0.0002 -0.0033 -0.061 -0.9981>}
 dyn_jacket_side_03_r {parent=dyn_jacket_side_02_r, loc=<8.6384 0 0>, rot=<0 0 0.0205 -0.9998>}
 MSTR_jacket_front_r {parent=spine_03, loc=<2.0596 10.7763 12.0069>, rot=<0.0232 -0.2575 0.9656 -0.0273>}
 dyn_jacket_front_01_r {parent=MSTR_jacket_front_r, loc=<0 0 0>, rot=<0 0 0 -1>}
 dyn_jacket_front_02_r {parent=dyn_jacket_front_01_r, loc=<14.9971 0 0>, rot=<-0.0007 -0.0023 -0.0747 -0.9972>}
 dyn_jacket_front_03_r {parent=dyn_jacket_front_02_r, loc=<9.7648 0 0>, rot=<0 0 0.0497 -0.9988>}
 def_altSpineFixLOD {parent=spine_03, loc=<0 0 0>, rot=<0 0 0 1>}
 MSTR_jacket_bk_l {parent=spine_01, loc=<1.6007 -12.7324 -12.6272>, rot=<-0.0173 -0.9239 -0.3822 0.0032>}
 dyn_jacket_bk_01_l {parent=MSTR_jacket_bk_l, loc=<0 0 0>, rot=<0 0 0 1>}
 dyn_jacket_bk_02_l {parent=dyn_jacket_bk_01_l, loc=<9.4897 0 0>, rot=<0 0 0.0201 -0.9998>}
 MSTR_jacket_bk {parent=spine_01, loc=<1.7796 -14.7318 0>, rot=<0.0152 -0.9999 0 0>}
 dyn_jacket_bk_01 {parent=MSTR_jacket_bk, loc=<0 0 0>, rot=<0 0 0 1>}
 dyn_jacket_bk_02 {parent=dyn_jacket_bk_01, loc=<9.4959 0 0>, rot=<0 0 0 1>}
 MSTR_jacket_bk_r {parent=spine_01, loc=<1.601 -12.7324 12.6272>, rot=<-0.0173 -0.9239 0.3822 -0.0032>}
 dyn_jacket_bk_01_r {parent=MSTR_jacket_bk_r, loc=<0 0 0>, rot=<0 0 0 -1>}
 dyn_jacket_bk_02_r {parent=dyn_jacket_bk_01_r, loc=<9.4901 0 0>, rot=<0 0 0.0201 -0.9998>}
 thigh_l {parent=pelvis, loc=<-1.0855 -0.2109 -8.7053>, rot=<0.0009 0.0355 0.0155 -0.9993>}
 calf_l {parent=thigh_l, loc=<-42.5284 -0.0019 -0.7297>, rot=<0.0006 -0.0161 -0.0391 -0.9991>}
 foot_l {parent=calf_l, loc=<-46.7401 -2.5014 -0.592>, rot=<0.0427 -0.0238 0.0353 -0.9982>}
 ball_l {parent=foot_l, loc=<-3.5658 10.8066 -0.1544>, rot=<0.0001 -0.0001 -0.7063 -0.7079>}
 dyn_footLace_01_l {parent=foot_l, loc=<2.7904 6.5993 -3.7699>, rot=<-0.0731 -0.9473 -0.0613 -0.3059>}
 dyn_footLace_02_l {parent=foot_l, loc=<3.2935 7.4512 2.8443>, rot=<0.3519 0.253 0.8958 0.0981>}
 calf_twist_02_l {parent=calf_l, loc=<-19.1546 -1.7943 -0.2638>, rot=<0.0237 0.0095 -0.0035 -0.9997>}
 calf_twist_01_l {parent=calf_l, loc=<-35.8483 -1.9197 -0.5794>, rot=<0.0237 0.0095 -0.0035 -0.9997>}
 def_foot_corr_bk_l {parent=calf_twist_01_l, loc=<-8.8928 -5.1013 0.3778>, rot=<0.6744 -0.7373 -0.0393 0.003>}
 def_foot_corr_fr_l {parent=calf_twist_01_l, loc=<-5.783 3.7574 -0.0543>, rot=<-0.0032 -0.0316 -0.6757 -0.7365>}
 deform_knee_in_base_l {parent=calf_l, loc=<0.8031 -5.0231 0.2456>, rot=<0 0 0 -1>}
 deform_knee_in_l {parent=deform_knee_in_base_l, loc=<0 0 0>, rot=<0 0 0 1>}
 deform_knee_out_base_l {parent=calf_l, loc=<0 4 0>, rot=<0 0 0 -1>}
 deform_knee_out_l {parent=deform_knee_out_base_l, loc=<0 0 0>, rot=<0 0 0 1>}
 thigh_twist_01_l {parent=thigh_l, loc=<-27.952 0 0>, rot=<0.0013 0 0 -1>}
 def_legBag_l {parent=thigh_twist_01_l, loc=<5.8086 -0.8474 -7.7461>, rot=<-0.0012 -0.9986 -0.0362 -0.0373>}
 deform_groin_base_l {parent=thigh_l, loc=<0.4801 5.6776 -0.2486>, rot=<0 0 0 -1>}
 deform_groin_l {parent=deform_groin_base_l, loc=<0 0 0>, rot=<0 0 0 1>}
 MSTR_pants_bk_l {parent=thigh_l, loc=<-29.5013 -10.6931 0.6339>, rot=<0.0204 -0.0002 0.9998 0.0022>}
 dyn_pants_bk_l {parent=MSTR_pants_bk_l, loc=<0 0 0>, rot=<0 0 0 1>}
 dyn_pants_fr_l {parent=thigh_l, loc=<-30.1899 9.8426 0.154>, rot=<0.008 0.0229 0.9997 0.0036>}
 dyn_pants_out_l {parent=thigh_l, loc=<-30.656 -0.9409 -7.6924>, rot=<-0.0048 -0.708 -0.7061 0.0098>}
 dyn_pants_in_l {parent=thigh_l, loc=<-29.6733 -1.2793 9.1671>, rot=<0.012 -0.6907 0.7229 0.0148>}
 thigh_r {parent=pelvis, loc=<-1.0855 -0.2109 8.7053>, rot=<0.0355 -0.0009 0.9993 0.0155>}
 calf_r {parent=thigh_r, loc=<42.5284 0.0019 0.7297>, rot=<0.0006 -0.0161 -0.0391 -0.9991>}
 foot_r {parent=calf_r, loc=<46.7401 2.5014 0.592>, rot=<0.0427 -0.0238 0.0353 -0.9982>}
 ball_r {parent=foot_r, loc=<3.5658 -10.8066 0.1544>, rot=<0.0001 -0.0001 -0.7063 -0.7079>}
 dyn_footLace_01_r {parent=foot_r, loc=<-2.7904 -6.5993 3.7699>, rot=<-0.0613 0.3059 0.0731 -0.9473>}
 dyn_footLace_02_r {parent=foot_r, loc=<-3.2935 -7.4512 -2.8443>, rot=<0.8958 -0.0981 -0.3519 0.253>}
 calf_twist_02_r {parent=calf_r, loc=<19.1546 1.7943 0.2638>, rot=<0.0237 0.0095 -0.0035 -0.9997>}
 calf_twist_01_r {parent=calf_r, loc=<35.8483 1.9197 0.5794>, rot=<0.0237 0.0095 -0.0035 -0.9997>}
 def_foot_corr_fr_r {parent=calf_twist_01_r, loc=<5.783 -3.7574 0.0543>, rot=<0.0316 -0.0032 0.7365 -0.6757>}
 def_foot_corr_bk_r {parent=calf_twist_01_r, loc=<8.9166 5.0994 -0.3781>, rot=<0.7373 0.6744 -0.003 -0.0393>}
 deform_knee_in_base_r {parent=calf_r, loc=<-0.8031 5.0231 -0.2457>, rot=<0 0 0 -1>}
 deform_knee_in_r {parent=deform_knee_in_base_r, loc=<0 0 0>, rot=<0 0 0 1>}
 deform_knee_out_base_r {parent=calf_r, loc=<0 -4 0>, rot=<0 0 0 -1>}
 deform_knee_out_r {parent=deform_knee_out_base_r, loc=<0 0 0>, rot=<0 0 0 1>}
 thigh_twist_01_r {parent=thigh_r, loc=<27.952 0 0>, rot=<0.0013 0 0 -1>}
 def_regBag_r {parent=thigh_twist_01_r, loc=<-5.8086 0.8474 7.7461>, rot=<-0.0012 -0.9986 -0.0362 -0.0373>}
 def_pocketChick {parent=def_regBag_r, loc=<8.8023 5.6587 -3.5387>, rot=<0.7562 -0.0228 0.6534 -0.0264>}
 deform_groin_base_r {parent=thigh_r, loc=<-0.4801 -5.6776 0.2486>, rot=<0 0 0 -1>}
 deform_groin_r {parent=deform_groin_base_r, loc=<0 0 0>, rot=<0 0 0 1>}
 MSTR_pants_bk_r {parent=thigh_r, loc=<29.5014 10.6931 -0.6339>, rot=<-0.0002 -0.0204 0.0022 -0.9998>}
 dyn_pants_bk_r {parent=MSTR_pants_bk_r, loc=<0 0 0>, rot=<0 0 0 1>}
 dyn_pants_fr_r {parent=thigh_r, loc=<30.1899 -9.8426 -0.154>, rot=<0.0229 -0.008 0.0036 -0.9997>}
 dyn_pants_in_r {parent=thigh_r, loc=<29.6733 1.2792 -9.1671>, rot=<-0.6907 -0.012 0.0148 -0.7229>}
 dyn_pants_out_r {parent=thigh_r, loc=<30.656 0.9409 7.6924>, rot=<0.708 -0.0048 -0.0098 -0.7061>}
 deform_glute_base_l {parent=pelvis, loc=<-7.3466 -9.0213 -9.1245>, rot=<0.0172 0.031 -0.4477 -0.8935>}
 deform_glute_l {parent=deform_glute_base_l, loc=<0 0 0>, rot=<0 0 0 1>}
 deform_glute_base_r {parent=pelvis, loc=<-7.3464 -9.0213 9.1245>, rot=<0.031 -0.0172 0.8935 -0.4477>}
 deform_glute_r {parent=deform_glute_base_r, loc=<0 0 0>, rot=<0 0 0 1>}
 def_beltRing {parent=pelvis, loc=<0.2363 9.6851 0>, rot=<0 0 -0.6943 -0.7197>}
]
variables [
 CalfL {type=double, default=1}
 CalfR {type=double, default=1}
 isBaseBallSliding {type=bool, default=False}
 isEmote {type=bool, default=False}
 isFrontEnd {type=bool, default=False}
 isParachute {type=bool, default=False}
 isRBAN {type=bool, default=False}
 isSkyDiveFloat {type=bool, default=False}
 isSliding {type=bool, default=False}
 isSwingAxe {type=bool, default=False}
]
