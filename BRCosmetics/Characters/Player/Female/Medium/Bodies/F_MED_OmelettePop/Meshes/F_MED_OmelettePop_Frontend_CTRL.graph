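# Frontend Control Rig for the female OmelettePop body that nudges cosmetic master bones, including the left and right jacket fronts.
# A single Modify Transforms unit applies fixed offsets to those bones.
# Its weight targets on when isFrontend is true, isEmote is false, and isSkyDive and isSkyDiving are not both true.
# An Alpha Interp smooths that boolean result, so the correction blends in and out rather than popping.
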ControlRigBlueprint /BRCosmetics/Characters/Player/Female/Medium/Bodies/F_MED_OmelettePop/Meshes/F_MED_OmelettePop_Frontend_CTRL
graphs [
 RigVMFunctionLibrary {links [], nodes []}
 RigVMFunction_MathBoolAnd_2_ContainedGraph {nodes [
   n1 = RigVMFunctionEntryNode()
   RigVMFunction_MathBoolAnd = &RigVMFunction_MathBoolAnd(A=@n1.A, B=@n1.B)
   RigVMFunctionReturnNode(Result=&RigVMFunction_MathBoolAnd(B=@n1.C, A=@RigVMFunction_MathBoolAnd.Result).Result)
  ]}
 RigVMModel {nodes [
   RigUnit_BeginExecution = &RigUnit_BeginExecution()
   RigVMFunction_MathBoolNot = &RigVMFunction_MathBoolNot(Value=RigVMVariableNode() {variable=isEmote}.Value)
   n1 = RigVMVariableNode() {variable=isSkyDive}
   n2 = RigVMVariableNode() {variable=isSkyDiving}
   RigVMFunction_MathBoolAnd = RigVMAggregateNode(
    A=@RigVMFunction_MathBoolNot.Result
    C=&RigVMFunction_MathBoolNot(Value=&RigVMFunction_MathBoolAnd(A=@n1.Value, B=@n2.Value).Result).Result
    B=RigVMVariableNode() {variable=isFrontend}.Value
   )
   Cast = &RigVMFunction_MathBoolToFloat(Value=@RigVMFunction_MathBoolAnd.Result)
   RigVMFunction_AlphaInterp = &RigVMFunction_AlphaInterp(InterpSpeedDecreasing=10, InterpSpeedIncreasing=10, Scale=1, bClampResult=False, bInterpResult=True, Value=@Cast.Result)
   &RigUnit_ModifyTransforms(
    ItemToModify="((Item=(Type=Bone,Name=\"MSTR_jacket_front_l\"),Transform=(Rotation=(X=0,Y=0,Z=0,W=1),Translation=(X=0,Y=2,Z=3),Scale3D=(X=1,Y=1,Z=1))),(Item=(Type=Bone,Name=\"MSTR_jacket_front_r\"),Transform=(Rotation=(X=0,Y=0,Z=0,W=1),Translation=(X=0,Y=1,Z=-3),Scale3D=(X=1 ...(truncated)"
    WeightMinimum=0
    ExecutePin=@RigUnit_BeginExecution.ExecutePin
    Weight=@RigVMFunction_AlphaInterp.Result
   )
  ]}
]
hierarchy [
 root {loc=<0 0 0>, rot=<0 0 0 1>}
 pelvis {parent=root, loc=<0 -0.5301 98.0571>, rot=<0.0127 0.707 -0.0127 -0.707>}
 spine_01 {parent=pelvis, loc=<6.1208 -0.9328 0>, rot=<0 0 0.034 -0.9994>}
 spine_02 {parent=spine_01, loc=<6.5997 -0.2095 0>, rot=<0 0 0.0238 -0.9997>}
 spine_03 {parent=spine_02, loc=<6.4914 -0.1237 0>, rot=<0 0 0.0149 -0.9999>}
 spine_04 {parent=spine_03, loc=<6.9598 0.0794 0>, rot=<0 0 -0.0132 -0.9999>}
 spine_05 {parent=spine_04, loc=<6.9987 0 0>, rot=<0 0 -0.0015 -1>}
 neck_01 {parent=spine_05, loc=<11.3138 0.1529 0>, rot=<0 0 -0.2003 -0.9797>}
 neck_02 {parent=neck_01, loc=<4.5627 0 0>, rot=<0 0 0.0504 -0.9987>}
 head {parent=neck_02, loc=<5.3914 0 0>, rot=<0 0 0.115 -0.9934>}
 clavicle_l {parent=spine_05, loc=<5.4831 -0.1053 -3.2946>, rot=<0.0465 -0.7289 -0.009 -0.6829>}
 upperarm_l {parent=clavicle_l, loc=<9.8658 -1.2136 -0.4105>, rot=<0.0164 -0.3245 -0.0269 -0.9454>}
 lowerarm_l {parent=upperarm_l, loc=<26.465 -0.8345 0.259>, rot=<0.0036 0.0161 -0.2298 -0.9731>}
 hand_l {parent=lowerarm_l, loc=<22.7355 -0.3203 -0.6105>, rot=<-0.6784 -0.0223 0.0735 -0.7307>}
 thumb_01_l {parent=hand_l, loc=<1.3118 -0.3191 -2.459>, rot=<0.6546 -0.4042 0.0664 -0.6354>}
 thumb_02_l {parent=thumb_01_l, loc=<5.9336 -0.0448 0.0401>, rot=<-0.0037 0.0213 0.298 -0.9543>}
 thumb_03_l {parent=thumb_02_l, loc=<2.2223 0.3682 -0.0907>, rot=<0.0118 -0.013 0.0066 -0.9998>}
 index_metacarpal_l {parent=hand_l, loc=<1.6535 0.0183 -2.96>, rot=<0.0158 -0.0124 -0.0173 -0.9996>}
 index_01_l {parent=index_metacarpal_l, loc=<8.3008 0.0357 -0.0524>, rot=<-0.0044 0.0485 0.0924 -0.9945>}
 index_02_l {parent=index_01_l, loc=<3.3587 -0.3004 0.0452>, rot=<-0.002 0.0027 0.3225 -0.9466>}
 index_03_l {parent=index_02_l, loc=<2.5578 -0.123 -0.0252>, rot=<-0.002 0.0067 0.2558 -0.9667>}
 middle_metacarpal_l {parent=hand_l, loc=<2.1856 -0.1725 -0.5984>, rot=<0.006 0.0168 -0.0329 -0.9993>}
 middle_01_l {parent=middle_metacarpal_l, loc=<7.3275 -0.2447 0.0129>, rot=<-0.0208 0.0042 0.0858 -0.9961>}
 middle_02_l {parent=middle_01_l, loc=<4.0123 0.1401 0.1175>, rot=<-0.0218 0.0258 0.3512 -0.9357>}
 middle_03_l {parent=middle_02_l, loc=<2.6077 0.0534 0.1373>, rot=<0.0082 0.0188 0.1708 -0.9851>}
 ring_metacarpal_l {parent=hand_l, loc=<2.2074 -0.3625 1.3638>, rot=<-0.0284 0.0722 -0.0458 -0.9959>}
 ring_01_l {parent=ring_metacarpal_l, loc=<7.0295 -0.6197 -0.1446>, rot=<-0.0489 -0.0252 0.1892 -0.9804>}
 ring_02_l {parent=ring_01_l, loc=<3.6174 0.4647 -0.1027>, rot=<0 0 0.2806 -0.9598>}
 ring_03_l {parent=ring_02_l, loc=<2.4152 0.114 0.0199>, rot=<0 0 0.1189 -0.9929>}
 pinky_metacarpal_l {parent=hand_l, loc=<1.7616 -0.9753 3.3814>, rot=<-0.0397 0.0814 -0.069 -0.9935>}
 pinky_01_l {parent=pinky_metacarpal_l, loc=<7.1709 -0.4831 -0.1094>, rot=<-0.0654 -0.0035 0.2242 -0.9723>}
 pinky_02_l {parent=pinky_01_l, loc=<2.8274 0.0893 -0.1043>, rot=<0.0037 0.0011 0.2854 -0.9584>}
 pinky_03_l {parent=pinky_02_l, loc=<2.2046 0.0307 -0.0078>, rot=<-0.0002 0.0098 0.1445 -0.9895>}
 deform_wrist_dn_base_l {parent=hand_l, loc=<0 -1.5326 0>, rot=<0 0 0 -1>}
 deform_wrist_dn_l {parent=deform_wrist_dn_base_l, loc=<0 0 0>, rot=<0 0 0 1>}
 deform_wrist_up_base_l {parent=hand_l, loc=<0 2.3009 0>, rot=<0 0 0 -1>}
 deform_wrist_up_l {parent=deform_wrist_up_base_l, loc=<0 0 0>, rot=<0 0 0 1>}
 lowerarm_twist_02_l {parent=lowerarm_l, loc=<10.6707 0 0>, rot=<0 0 0 1>}
 lowerarm_twist_01_l {parent=lowerarm_l, loc=<18.316 0 0>, rot=<0 0 0 1>}
 deform_elbow_in_base_l {parent=lowerarm_l, loc=<0.1834 3.6062 -0.2572>, rot=<0 0 0 -1>}
 deform_elbow_in_l {parent=deform_elbow_in_base_l, loc=<0 0 0>, rot=<0 0 0 1>}
 deform_elbow_out_base_l {parent=lowerarm_l, loc=<0 -4 0>, rot=<0 0 0 1>}
 deform_elbow_out_l {parent=deform_elbow_out_base_l, loc=<0 0 0>, rot=<0 0 0 1>}
 upperarm_twist_01_l {parent=upperarm_l, loc=<7.2135 -0.1134 0>, rot=<0 0 0 1>}
 deform_upperarm_bk_base_l {parent=upperarm_twist_01_l, loc=<-4.3788 -6.3931 -1.4877>, rot=<0 0 0 -1>}
 deform_upperarm_bk_l {parent=deform_upperarm_bk_base_l, loc=<0 0 0>, rot=<0 0 0 1>}
 deform_upperarm_fr_base_l {parent=upperarm_twist_01_l, loc=<-3.4526 7.5068 -0.2668>, rot=<0 0 0 -1>}
 deform_upperarm_fr_l {parent=deform_upperarm_fr_base_l, loc=<0 0 0>, rot=<0 0 0 1>}
 deform_upperarm_up_base_l {parent=upperarm_twist_01_l, loc=<-6.3503 0.4825 4.8307>, rot=<0 0 0 -1>}
 deform_upperarm_up_l {parent=deform_upperarm_up_base_l, loc=<0 0 0>, rot=<0 0 0 1>}
 upperarm_twist_02_l {parent=upperarm_l, loc=<19.7998 -0.1134 0>, rot=<0 0 0 1>}
 dyn_armRing_l {parent=upperarm_twist_02_l, loc=<1.5155 -1.4039 7.4429>, rot=<0.6778 0.0803 0.0961 -0.7245>}
 clavicle_r {parent=spine_05, loc=<5.4831 -0.1053 3.2946>, rot=<0.7289 0.0465 -0.6829 0.009>}
 upperarm_r {parent=clavicle_r, loc=<-9.8658 1.2136 0.4105>, rot=<0.0164 -0.3245 -0.0269 -0.9454>}
 lowerarm_r {parent=upperarm_r, loc=<-26.465 0.8345 -0.259>, rot=<0.0036 0.0161 -0.2298 -0.9731>}
 hand_r {parent=lowerarm_r, loc=<-22.7355 0.3203 0.6106>, rot=<-0.6784 -0.0223 0.0735 -0.7307>}
 thumb_01_r {parent=hand_r, loc=<-1.3118 0.3191 2.459>, rot=<0.6546 -0.4042 0.0664 -0.6354>}
 thumb_02_r {parent=thumb_01_r, loc=<-5.9336 0.0448 -0.0401>, rot=<-0.0037 0.0213 0.298 -0.9543>}
 thumb_03_r {parent=thumb_02_r, loc=<-2.2223 -0.3682 0.0907>, rot=<0.0118 -0.013 0.0066 -0.9998>}
 index_metacarpal_r {parent=hand_r, loc=<-1.6535 -0.0183 2.96>, rot=<0.0158 -0.0124 -0.0173 -0.9996>}
 index_01_r {parent=index_metacarpal_r, loc=<-8.3008 -0.0357 0.0524>, rot=<-0.0044 0.0485 0.0924 -0.9945>}
 index_02_r {parent=index_01_r, loc=<-3.3587 0.3004 -0.0452>, rot=<-0.002 0.0027 0.3225 -0.9466>}
 index_03_r {parent=index_02_r, loc=<-2.5578 0.123 0.0252>, rot=<-0.002 0.0067 0.2558 -0.9667>}
 middle_metacarpal_r {parent=hand_r, loc=<-2.1856 0.1725 0.5984>, rot=<0.006 0.0168 -0.0329 -0.9993>}
 middle_01_r {parent=middle_metacarpal_r, loc=<-7.3275 0.2447 -0.0129>, rot=<-0.0208 0.0042 0.0858 -0.9961>}
 middle_02_r {parent=middle_01_r, loc=<-4.0123 -0.1401 -0.1175>, rot=<-0.0218 0.0258 0.3512 -0.9357>}
 middle_03_r {parent=middle_02_r, loc=<-2.6077 -0.0534 -0.1373>, rot=<0.0082 0.0188 0.1708 -0.9851>}
 ring_metacarpal_r {parent=hand_r, loc=<-2.2074 0.3625 -1.3638>, rot=<-0.0284 0.0722 -0.0458 -0.9959>}
 ring_01_r {parent=ring_metacarpal_r, loc=<-7.0295 0.6197 0.1446>, rot=<-0.0489 -0.0252 0.1892 -0.9804>}
 ring_02_r {parent=ring_01_r, loc=<-3.6174 -0.4647 0.1027>, rot=<0 0 0.2806 -0.9598>}
 ring_03_r {parent=ring_02_r, loc=<-2.4152 -0.114 -0.0199>, rot=<0 0 0.1189 -0.9929>}
 pinky_metacarpal_r {parent=hand_r, loc=<-1.7616 0.9753 -3.3814>, rot=<-0.0397 0.0814 -0.069 -0.9935>}
 pinky_01_r {parent=pinky_metacarpal_r, loc=<-7.1709 0.4831 0.1094>, rot=<-0.0654 -0.0035 0.2242 -0.9723>}
 pinky_02_r {parent=pinky_01_r, loc=<-2.8274 -0.0893 0.1043>, rot=<0.0037 0.0011 0.2854 -0.9584>}
 pinky_03_r {parent=pinky_02_r, loc=<-2.2046 -0.0307 0.0078>, rot=<-0.0002 0.0098 0.1445 -0.9895>}
 deform_wrist_dn_base_r {parent=hand_r, loc=<-0.0001 1.5328 0>, rot=<0 0 0 -1>}
 deform_wrist_dn_r {parent=deform_wrist_dn_base_r, loc=<0 0 0>, rot=<0 0 0 1>}
 deform_wrist_up_base_r {parent=hand_r, loc=<-0.0002 -2.3007 -0.0001>, rot=<0 0 0 -1>}
 deform_wrist_up_r {parent=deform_wrist_up_base_r, loc=<0 0 0>, rot=<0 0 0 1>}
 lowerarm_twist_02_r {parent=lowerarm_r, loc=<-10.6707 0 0>, rot=<0 0 0 -1>}
 lowerarm_twist_01_r {parent=lowerarm_r, loc=<-18.316 0 0>, rot=<0 0 0 -1>}
 deform_elbow_in_base_r {parent=lowerarm_r, loc=<-0.1834 -3.6062 0.2572>, rot=<0 0 0 -1>}
 deform_elbow_in_r {parent=deform_elbow_in_base_r, loc=<0 0 0>, rot=<0 0 0 1>}
 deform_elbow_out_base_r {parent=lowerarm_r, loc=<0 4 -0.0001>, rot=<0 0 0 -1>}
 deform_elbow_out_r {parent=deform_elbow_out_base_r, loc=<0 0 0>, rot=<0 0 0 1>}
 upperarm_twist_01_r {parent=upperarm_r, loc=<-7.2135 0.1134 0>, rot=<0 0 0 -1>}
 deform_upperarm_bk_base_r {parent=upperarm_twist_01_r, loc=<3.849 5.824 -0.333>, rot=<0 0 0 -1>}
 deform_upperarm_bk_r {parent=deform_upperarm_bk_base_r, loc=<0 0 0>, rot=<0 0 0 1>}
 deform_upperarm_fr_base_r {parent=upperarm_twist_01_r, loc=<3 -5.554 -0.675>, rot=<0 0 0 -1>}
 deform_upperarm_fr_r {parent=deform_upperarm_fr_base_r, loc=<0 0 0>, rot=<0 0 0 1>}
 deform_upperarm_up_base_r {parent=upperarm_twist_01_r, loc=<6.35 -0.482 -4.831>, rot=<0 0 0 -1>}
 deform_upperarm_up_r {parent=deform_upperarm_up_base_r, loc=<0 0 0>, rot=<0 0 0 1>}
 upperarm_twist_02_r {parent=upperarm_r, loc=<-19.7998 0.1134 0>, rot=<0 0 0 -1>}
 dyn_armRing_r {parent=upperarm_twist_02_r, loc=<-1.5154 1.4039 -7.443>, rot=<-0.0803 0.6778 0.7245 0.0961>}
 def_egg06 {parent=clavicle_r, loc=<-2.992 10.753 3.7529>, rot=<0.9271 0.0685 -0.3448 0.13>}
 def_egg05 {parent=clavicle_r, loc=<-6.4914 8.9097 -0.18>, rot=<0.9271 0.0685 -0.3448 0.13>}
 def_egg04 {parent=clavicle_r, loc=<-1.9777 -10.1365 2.68>, rot=<-0.0586 -0.295 -0.0027 -0.9537>}
 attach_backpack {parent=spine_05, loc=<-0.123 -13.473 0>, rot=<-0.0093 -0.707 0.0093 -0.707>}
 attach_cape {parent=spine_05, loc=<9.017 -11.714 0>, rot=<0.0616 -0.7044 -0.0616 -0.7044>}
 dyn_hoodie {parent=spine_05, loc=<12.3798 -9.6335 -0.1322>, rot=<0.4574 -0.5346 0.5564 -0.4422>}
 deform_pec_aim_base_l {parent=spine_04, loc=<3.2456 9.3309 -1.2237>, rot=<0.0603 -0.4033 0.3318 -0.8507>}
 deform_pec_aim_l {parent=deform_pec_aim_base_l, loc=<0 0 0>, rot=<0 0 0 1>}
 deform_pec_base_l {parent=deform_pec_aim_l, loc=<9.3665 0 0>, rot=<0 0 0 1>}
 deform_pec_l {parent=deform_pec_base_l, loc=<0 0 0>, rot=<0 0 0 1>}
 deform_pec_aim_base_r {parent=spine_04, loc=<3.2452 9.3309 1.2237>, rot=<-0.4033 -0.0603 0.8507 0.3318>}
 deform_pec_aim_r {parent=deform_pec_aim_base_r, loc=<0 0 0>, rot=<0 0 0 1>}
 deform_pec_base_r {parent=deform_pec_aim_r, loc=<-9.3667 -0.0003 -0.0004>, rot=<0 0 0 1>}
 deform_pec_r {parent=deform_pec_base_r, loc=<0 0 0>, rot=<0 0 0 1>}
 MSTR_bag {parent=spine_04, loc=<-0.2333 -0.4446 -3.2089>, rot=<0.0332 -0.9466 0.0062 -0.3207>}
 MSTR_belt_bk {parent=MSTR_bag, loc=<0.7573 -13.1313 0.6218>, rot=<0.996 0.08 0.0312 0.0229>}
 belt2 {parent=MSTR_belt_bk, loc=<0 0 0>, rot=<0 0 0 1>}
 belt3 {parent=belt2, loc=<13.8248 0 0>, rot=<0.2512 -0.026 0.1529 -0.9554>}
 belt6 {parent=belt3, loc=<11.3877 0 0>, rot=<-0.6225 0.1194 0.2164 -0.7426>}
 belt7 {parent=belt6, loc=<10.042 0 0>, rot=<0 0.0735 0 -0.9973>}
 def_egg02 {parent=belt3, loc=<-2.6741 -0.3173 -0.2399>, rot=<0.8751 0.3102 -0.3474 -0.1318>}
 def_egg01 {parent=belt3, loc=<2.8618 0.6044 0.1173>, rot=<0.8751 0.3102 -0.3474 -0.1318>}
 MSTR_belt_fr {parent=MSTR_bag, loc=<-1.5776 12.5049 -0.7061>, rot=<0.9964 0.0798 0.0202 0.0221>}
 belt1 {parent=MSTR_belt_fr, loc=<0 0 0>, rot=<0 0 0 1>}
 belt02 {parent=belt1, loc=<5.9985 -0.0772 0.0118>, rot=<0 0 0 1>}
 belt4 {parent=belt02, loc=<5.3596 0.0772 -0.0118>, rot=<0.002 -0.0145 -0.0189 -0.9997>}
 belt5 {parent=belt4, loc=<8.2035 0 0>, rot=<0 0 0 1>}
 def_egg08 {parent=belt4, loc=<-1.7599 -0.0281 -0.2657>, rot=<0.9103 0.0548 -0.4096 0.0247>}
 def_egg07 {parent=belt4, loc=<3.8945 -0.3818 0.178>, rot=<0.9103 0.0548 -0.4096 0.0247>}
 def_egg03 {parent=spine_04, loc=<4.7301 10.9652 2.0877>, rot=<0.8968 -0.0188 -0.4305 -0.1>}
 MSTR_jacket_front_l {parent=spine_03, loc=<2.0598 10.7763 -12.0069>, rot=<-0.0232 0.2575 0.9656 -0.0273>}
 dyn_jacket_front_01_l {parent=MSTR_jacket_front_l, loc=<0 0 0>, rot=<0 0 0 1>}
 dyn_jacket_front_02_l {parent=dyn_jacket_front_01_l, loc=<14.997 0 0>, rot=<0.0007 0.0023 -0.0747 -0.9972>}
 dyn_jacket_front_03_l {parent=dyn_jacket_front_02_l, loc=<9.7652 0 0>, rot=<0 0 0.0497 -0.9988>}
 dyn_chicketnbag {parent=dyn_jacket_front_01_l, loc=<18.7317 -2.7328 -6.118>, rot=<0.7003 -0.5629 -0.3962 0.1892>}
 MSTR_jacket_side_l {parent=spine_03, loc=<-0.7423 -0.8474 -16.5355>, rot=<-0.0796 0.7029 0.7069 -0.0021>}
 dyn_jacket_side_01_l {parent=MSTR_jacket_side_l, loc=<0 0 0>, rot=<0 0 0 1>}
 dyn_jacket_side_02_l {parent=dyn_jacket_side_01_l, loc=<11.0425 0 0>, rot=<0.0002 0.0033 -0.061 -0.9981>}
 dyn_jacket_side_03_l {parent=dyn_jacket_side_02_l, loc=<8.6388 0 0>, rot=<0 0 0.0205 -0.9998>}
 MSTR_jacket_front_r {parent=spine_03, loc=<2.0596 10.7763 12.0069>, rot=<0.0232 -0.2575 0.9656 -0.0273>}
 dyn_jacket_front_01_r {parent=MSTR_jacket_front_r, loc=<0 0 0>, rot=<0 0 0 1>}
 dyn_jacket_front_02_r {parent=dyn_jacket_front_01_r, loc=<14.9971 0 0>, rot=<-0.0007 -0.0023 -0.0747 -0.9972>}
 dyn_jacket_front_03_r {parent=dyn_jacket_front_02_r, loc=<9.7648 0 0>, rot=<0 0 0.0497 -0.9988>}
 MSTR_jacket_side_r {parent=spine_03, loc=<-0.7426 -0.8474 16.5355>, rot=<0.0796 -0.7029 0.7069 -0.0021>}
 dyn_jacket_side_01_r {parent=MSTR_jacket_side_r, loc=<0 0 0>, rot=<0 0 0 1>}
 dyn_jacket_side_02_r {parent=dyn_jacket_side_01_r, loc=<11.0427 0 0>, rot=<-0.0002 -0.0033 -0.061 -0.9981>}
 dyn_jacket_side_03_r {parent=dyn_jacket_side_02_r, loc=<8.6384 0 0>, rot=<0 0 0.0205 -0.9998>}
 def_altSpineFixLOD {parent=spine_03, loc=<0 0 0>, rot=<0 0 0 1>}
 def_backBeltAlt {parent=spine_03, loc=<0.6145 -13.0804 -9.9103>, rot=<0.0387 -0.706 -0.0387 -0.706>}
 MSTR_jacket_bk {parent=spine_01, loc=<1.7796 -14.7318 0>, rot=<0.0152 -0.9999 0 0>}
 dyn_jacket_bk_01 {parent=MSTR_jacket_bk, loc=<0 0 0>, rot=<0 0 0 1>}
 dyn_jacket_bk_02 {parent=dyn_jacket_bk_01, loc=<9.4959 0 0>, rot=<0 0 0 1>}
 MSTR_jacket_bk_l {parent=spine_01, loc=<1.6007 -12.7324 -12.6272>, rot=<-0.0173 -0.9239 -0.3822 0.0032>}
 dyn_jacket_bk_01_l {parent=MSTR_jacket_bk_l, loc=<0 0 0>, rot=<0 0 0 1>}
 dyn_jacket_bk_02_l {parent=dyn_jacket_bk_01_l, loc=<9.4897 0 0>, rot=<0 0 0.0201 -0.9998>}
 MSTR_jacket_bk_r {parent=spine_01, loc=<1.601 -12.7324 12.6272>, rot=<-0.0173 -0.9239 0.3822 -0.0032>}
 dyn_jacket_bk_01_r {parent=MSTR_jacket_bk_r, loc=<0 0 0>, rot=<0 0 0 1>}
 dyn_jacket_bk_02_r {parent=dyn_jacket_bk_01_r, loc=<9.4901 0 0>, rot=<0 0 0.0201 -0.9998>}
 thigh_l {parent=pelvis, loc=<-1.0855 -0.2109 -8.7053>, rot=<0.0009 0.0355 0.0155 -0.9993>}
 calf_l {parent=thigh_l, loc=<-42.5284 -0.0019 -0.7297>, rot=<0.0006 -0.0161 -0.0391 -0.9991>}
 foot_l {parent=calf_l, loc=<-46.7401 -2.5014 -0.592>, rot=<0.0427 -0.0238 0.0353 -0.9982>}
 ball_l {parent=foot_l, loc=<-3.5658 10.8066 -0.1544>, rot=<0.0001 -0.0001 -0.7063 -0.7079>}
 dyn_footLace_01_l {parent=foot_l, loc=<2.7904 6.5993 -3.7699>, rot=<-0.0731 -0.9473 -0.0613 -0.3059>}
 dyn_footLace_02_l {parent=foot_l, loc=<3.2935 7.4512 2.8443>, rot=<0.3519 0.253 0.8958 0.0981>}
 calf_twist_02_l {parent=calf_l, loc=<-19.1546 -1.7943 -0.2638>, rot=<0.0237 0.0095 -0.0035 -0.9997>}
 calf_twist_01_l {parent=calf_l, loc=<-35.8483 -1.9197 -0.5794>, rot=<0.0237 0.0095 -0.0035 -0.9997>}
 def_foot_corr_bk_l {parent=calf_twist_01_l, loc=<-8.8928 -5.1013 0.3778>, rot=<0.6744 -0.7373 -0.0393 0.003>}
 def_foot_corr_fr_l {parent=calf_twist_01_l, loc=<-5.783 3.7574 -0.0543>, rot=<-0.0032 -0.0316 -0.6757 -0.7365>}
 deform_knee_in_base_l {parent=calf_l, loc=<0.8031 -5.0231 0.2456>, rot=<0 0 0 -1>}
 deform_knee_in_l {parent=deform_knee_in_base_l, loc=<0 0 0>, rot=<0 0 0 1>}
 deform_knee_out_base_l {parent=calf_l, loc=<0 4 0>, rot=<0 0 0 -1>}
 deform_knee_out_l {parent=deform_knee_out_base_l, loc=<0 0 0>, rot=<0 0 0 1>}
 thigh_twist_01_l {parent=thigh_l, loc=<-27.952 0 0>, rot=<0.0013 0 0 -1>}
 def_legBag_l {parent=thigh_twist_01_l, loc=<5.8086 -0.8474 -7.7461>, rot=<-0.0012 -0.9986 -0.0362 -0.0373>}
 deform_groin_base_l {parent=thigh_l, loc=<0.4801 5.6776 -0.2486>, rot=<0 0 0 -1>}
 deform_groin_l {parent=deform_groin_base_l, loc=<0 0 0>, rot=<0 0 0 1>}
 MSTR_pants_bk_l {parent=thigh_l, loc=<-29.5013 -10.6931 0.6339>, rot=<0.0204 -0.0002 0.9998 0.0022>}
 dyn_pants_bk_l {parent=MSTR_pants_bk_l, loc=<0 0 0>, rot=<0 0 0 1>}
 dyn_pants_fr_l {parent=thigh_l, loc=<-30.1899 9.8426 0.154>, rot=<0.008 0.0229 0.9997 0.0036>}
 dyn_pants_out_l {parent=thigh_l, loc=<-30.656 -0.9409 -7.6924>, rot=<-0.0048 -0.708 -0.7061 0.0098>}
 dyn_pants_in_l {parent=thigh_l, loc=<-29.6733 -1.2793 9.1671>, rot=<0.012 -0.6907 0.7229 0.0148>}
 thigh_r {parent=pelvis, loc=<-1.0855 -0.2109 8.7053>, rot=<0.0355 -0.0009 0.9993 0.0155>}
 calf_r {parent=thigh_r, loc=<42.5284 0.0019 0.7297>, rot=<0.0006 -0.0161 -0.0391 -0.9991>}
 foot_r {parent=calf_r, loc=<46.7401 2.5014 0.592>, rot=<0.0427 -0.0238 0.0353 -0.9982>}
 ball_r {parent=foot_r, loc=<3.5658 -10.8066 0.1544>, rot=<0.0001 -0.0001 -0.7063 -0.7079>}
 dyn_footLace_01_r {parent=foot_r, loc=<-2.7904 -6.5993 3.7699>, rot=<-0.0613 0.3059 0.0731 -0.9473>}
 dyn_footLace_02_r {parent=foot_r, loc=<-3.2935 -7.4512 -2.8443>, rot=<0.8958 -0.0981 -0.3519 0.253>}
 calf_twist_02_r {parent=calf_r, loc=<19.1546 1.7943 0.2638>, rot=<0.0237 0.0095 -0.0035 -0.9997>}
 calf_twist_01_r {parent=calf_r, loc=<35.8483 1.9197 0.5794>, rot=<0.0237 0.0095 -0.0035 -0.9997>}
 def_foot_corr_fr_r {parent=calf_twist_01_r, loc=<5.783 -3.7574 0.0543>, rot=<0.0316 -0.0032 0.7365 -0.6757>}
 def_foot_corr_bk_r {parent=calf_twist_01_r, loc=<8.9166 5.0994 -0.3781>, rot=<0.7373 0.6744 -0.003 -0.0393>}
 deform_knee_in_base_r {parent=calf_r, loc=<-0.8031 5.0231 -0.2457>, rot=<0 0 0 -1>}
 deform_knee_in_r {parent=deform_knee_in_base_r, loc=<0 0 0>, rot=<0 0 0 1>}
 deform_knee_out_base_r {parent=calf_r, loc=<0 -4 0>, rot=<0 0 0 -1>}
 deform_knee_out_r {parent=deform_knee_out_base_r, loc=<0 0 0>, rot=<0 0 0 1>}
 thigh_twist_01_r {parent=thigh_r, loc=<27.952 0 0>, rot=<0.0013 0 0 -1>}
 def_regBag_r {parent=thigh_twist_01_r, loc=<-5.8086 0.8474 7.7461>, rot=<-0.0012 -0.9986 -0.0362 -0.0373>}
 def_pocketChick {parent=def_regBag_r, loc=<8.8023 5.6587 -3.5387>, rot=<0.7562 -0.0228 0.6534 -0.0264>}
 deform_groin_base_r {parent=thigh_r, loc=<-0.4801 -5.6776 0.2486>, rot=<0 0 0 -1>}
 deform_groin_r {parent=deform_groin_base_r, loc=<0 0 0>, rot=<0 0 0 1>}
 MSTR_pants_bk_r {parent=thigh_r, loc=<29.5014 10.6931 -0.6339>, rot=<-0.0002 -0.0204 0.0022 -0.9998>}
 dyn_pants_bk_r {parent=MSTR_pants_bk_r, loc=<0 0 0>, rot=<0 0 0 1>}
 dyn_pants_fr_r {parent=thigh_r, loc=<30.1899 -9.8426 -0.154>, rot=<0.0229 -0.008 0.0036 -0.9997>}
 dyn_pants_in_r {parent=thigh_r, loc=<29.6733 1.2792 -9.1671>, rot=<-0.6907 -0.012 0.0148 -0.7229>}
 dyn_pants_out_r {parent=thigh_r, loc=<30.656 0.9409 7.6924>, rot=<0.708 -0.0048 -0.0098 -0.7061>}
 deform_glute_base_l {parent=pelvis, loc=<-7.3466 -9.0213 -9.1245>, rot=<0.0172 0.031 -0.4477 -0.8935>}
 deform_glute_l {parent=deform_glute_base_l, loc=<0 0 0>, rot=<0 0 0 1>}
 deform_glute_base_r {parent=pelvis, loc=<-7.3464 -9.0213 9.1245>, rot=<0.031 -0.0172 0.8935 -0.4477>}
 deform_glute_r {parent=deform_glute_base_r, loc=<0 0 0>, rot=<0 0 0 1>}
 def_beltRing {parent=pelvis, loc=<0.2363 9.6851 0>, rot=<0 0 -0.6943 -0.7197>}
]
variables [
 isEmote {type=bool, default=False}
 isFrontend {type=bool, default=False}
 isSkyDive {type=bool, default=False}
 isSkyDiving {type=bool, default=False}
]
