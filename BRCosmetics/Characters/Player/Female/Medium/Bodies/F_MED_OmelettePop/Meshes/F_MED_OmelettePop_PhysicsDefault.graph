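# Physics asset for the OmelettePop medium female body cosmetic, setting up its jacket for secondary motion.
# Capsule chains for jacket front, sides and back use limited swing and twist, with front and side chains anchored to small MSTR spheres.
# Cross constraints, mostly free-rotation, tie neighbouring hem tips together around the jacket; only the back-left chain links to spine_01.
# Pelvis, spine, thighs, bags and pocket chick have unconstrained or collision-style shapes.
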PhysicsAsset /BRCosmetics/Characters/Player/Female/Medium/Bodies/F_MED_OmelettePop/Meshes/F_MED_OmelettePop_PhysicsDefault
bodies {
 MSTR_jacket_front_l [{shapeType=Sphere, radius=0.505, shapeName=MSTR_jacket_front_l_sphere}]
 MSTR_jacket_front_r [{shapeType=Sphere, radius=0.505, shapeName=MSTR_jacket_front_r_sphere}]
 MSTR_jacket_side_l [{shapeType=Sphere, radius=0.505, shapeName=MSTR_jacket_side_l_sphere}]
 MSTR_jacket_side_r [{shapeType=Sphere, radius=0.505, shapeName=MSTR_jacket_side_r_sphere}]
 def_legBag_l [{
   shapeType=Capsule
   length=4.0112
   radius=8.5192
   shapeName=def_legBag_l_capsule
   center=<-3.6044 -0.6197 0.2824>
   rotation=<rot -0.0988 0.7358 -90.0007>
  }]
 def_pocketChick [{shapeType=Sphere, radius=5.0378, shapeName=def_pocketChick_sphere, center=<0.3291 -1.8039 -0.143>}]
 def_regBag_r [{
   shapeType=Capsule
   length=5.5612
   radius=8.5192
   shapeName=def_legBag_l_capsule
   center=<3.5528 -0.5925 -0.8046>
   rotation=<rot -0.0988 0.7358 -90.0007>
  }]
 dyn_jacket_bk_01 [{shapeType=Capsule, length=14.5758, radius=4.5373, shapeName=dyn_jacket_front_03_l_capsule, center=<3.9779 -2.6136 -0.2826>}]
 dyn_jacket_bk_01_l [{
   shapeType=Capsule
   length=12.1666
   radius=4.5373
   shapeName=dyn_jacket_front_03_l_capsule
   center=<3.6196 -1.0297 -0.0633>
   rotation=<rot -8.4964 -95.8627 2.5206>
  }]
 dyn_jacket_bk_01_r [{
   shapeType=Capsule
   length=12.4184
   radius=4.5373
   shapeName=dyn_jacket_front_03_r_capsule
   center=<3.6266 -0.9421 0.2591>
   rotation=<rot -0.6159 -95.7753 -179.3702>
  }]
 dyn_jacket_bk_02 [{shapeType=Capsule, length=16, radius=4.5373, shapeName=dyn_jacket_front_03_l_capsule, center=<3.9045 -2.4424 -0.2826>}]
 dyn_jacket_bk_02_l [{
   shapeType=Capsule
   length=12.5013
   radius=4.5373
   shapeName=dyn_jacket_front_03_l_capsule
   center=<3.2013 -2.7233 -0.3105>
   rotation=<rot 0.0811 -90.1152 1.6801>
  }]
 dyn_jacket_bk_02_r [{
   shapeType=Capsule
   length=12.2674
   radius=4.5373
   shapeName=dyn_jacket_front_03_l_capsule
   center=<4.1746 -2.7253 0.3391>
   rotation=<rot 14.5784 -90.5547 1.4787>
  }]
 dyn_jacket_front_01_l [{
   shapeType=Capsule
   length=9.3426
   radius=5.4009
   shapeName=dyn_jacket_fr_01_l_capsule
   center=<7.8651 -3.4967 0>
   rotation=<rot 90 -0.8952 0.1956>
  }]
 dyn_jacket_front_01_r [{
   shapeType=Capsule
   length=9.3426
   radius=5
   shapeName=dyn_jacket_fr_01_l_capsule
   center=<7.8651 -3.4967 0>
   rotation=<rot 90 -0.8952 0.1956>
  }]
 dyn_jacket_front_02_l [{
   shapeType=Capsule
   length=13.3865
   radius=4.6026
   shapeName=dyn_jacket_fr_01_r_capsule
   center=<5.0146 -3.4292 -0.0003>
   rotation=<rot 0.3041 175.0557 -2.59>
  }]
 dyn_jacket_front_02_r [{
   shapeType=Capsule
   length=13.3865
   radius=4.6026
   shapeName=dyn_jacket_fr_01_r_capsule
   center=<5.0081 -3.5047 -0.0031>
   rotation=<rot -0.3041 175.0557 0>
  }]
 dyn_jacket_front_03_l [{
   shapeType=Capsule
   length=13.6563
   radius=4.9678
   shapeName=dyn_jacket_front_03_l_capsule
   center=<4.4161 -3.0138 -0.4327>
   rotation=<rot -4.9468 88.3418 -5.3972>
  }]
 dyn_jacket_front_03_r [{
   shapeType=Capsule
   length=13.6563
   radius=4.9678
   shapeName=dyn_jacket_front_03_l_capsule
   center=<4.3997 -3.5089 0.4115>
   rotation=<rot 2.4495 88.1057 5.382>
  }]
 dyn_jacket_side_01_l [{shapeType=Capsule, length=5.9621, radius=5.5945, shapeName=dyn_jacket_fr_01_l_capsule, center=<5.9634 -2.2773 0>, rotation=<rot 90 0 0>}]
 dyn_jacket_side_01_r [{shapeType=Capsule, length=5.9621, radius=5.5945, shapeName=dyn_jacket_fr_01_r_capsule, center=<5.9634 -2.2773 0>, rotation=<rot 90 0 180>}]
 dyn_jacket_side_02_l [{
   shapeType=Capsule
   length=12.6313
   radius=4.7903
   shapeName=dyn_jacket_front_02_l_capsule
   center=<3.2818 -1.6544 -0.1693>
   rotation=<rot 5.8331 -90.0029 0.0647>
  }]
 dyn_jacket_side_02_r [{
   shapeType=Capsule
   length=11.9319
   radius=5.4897
   shapeName=dyn_jacket_front_02_r_capsule
   center=<3.2817 -2.3838 0.1169>
   rotation=<rot 4.2341 -90.0047 -179.9353>
  }]
 dyn_jacket_side_03_l [{
   shapeType=Capsule
   length=13.3452
   radius=4.5373
   shapeName=dyn_jacket_front_03_l_capsule
   center=<3.6785 -4.0472 1.4248>
   rotation=<rot 6.6201 -89.9227 1.6912>
  }]
 dyn_jacket_side_03_r [{
   shapeType=Capsule
   length=12.1459
   radius=4.5373
   shapeName=dyn_jacket_front_03_r_capsule
   center=<3.6406 -4.1441 -0.1457>
   rotation=<rot 2.7572 -90.3 179.6206>
  }]
 pelvis [
  {shapeType=Capsule, length=4.2782, radius=12.8927, shapeName=pelvis_capsule, center=<-1.0028 -1.1332 0.017>, rotation=<rot -0.5854 -0.0001 0>}
  {shapeType=Capsule, length=13.5166, radius=10.9626, shapeName=pelvis_capsule, center=<-10.2563 -3.4304 0>}
 ]
 spine_01 [
  {shapeType=Capsule, length=6.1916, radius=11.258, shapeName=pelvis_capsule, center=<0.6596 -0.7248 0>, rotation=<rot -0.5854 -0.0001 0>}
  {shapeType=Capsule, length=9.3864, radius=9.5698, shapeName=spine_01_capsule, center=<0 2.9882 0>}
 ]
 thigh_twist_01_l [{
   shapeType=Capsule
   length=19.8788
   radius=10.492
   shapeName=thigh_twist_01_l_capsule
   center=<10.4323 -1.4044 1.4407>
   rotation=<rot -89.914 -179.9963 179.9963>
  }]
 thigh_twist_01_r [{
   shapeType=Capsule
   length=19.8788
   radius=10.492
   shapeName=thigh_twist_01_r_capsule
   center=<-10.4323 1.4044 -1.4407>
   rotation=<rot -89.914 -179.9963 179.9963>
  }]
}
constraints [
 {bone1Name=dyn_jacket_bk_02, bone2Name=dyn_jacket_bk_01, swing1LimitDegrees=20, swing2LimitDegrees=10, twistLimitDegrees=10}
 {bone1Name=dyn_jacket_bk_02_l, bone2Name=dyn_jacket_bk_01_l, swing1LimitDegrees=20, swing2LimitDegrees=10, twistLimitDegrees=5}
 {
  bone1Name=dyn_jacket_side_01_l
  bone2Name=MSTR_jacket_side_l
  swing1LimitDegrees=20
  swing2LimitDegrees=10
  twistLimitDegrees=5
  twistMotion=Locked
 }
 {bone1Name=dyn_jacket_side_03_l, bone2Name=dyn_jacket_side_02_l, swing1LimitDegrees=15, swing2LimitDegrees=5, twistLimitDegrees=10}
 {bone1Name=dyn_jacket_front_02_l, bone2Name=dyn_jacket_front_01_l, swing1LimitDegrees=30, swing2LimitDegrees=30, twistLimitDegrees=15}
 {bone1Name=dyn_jacket_front_03_l, bone2Name=dyn_jacket_front_02_l, swing1LimitDegrees=30, swing2LimitDegrees=10, twistLimitDegrees=10}
 {bone1Name=dyn_jacket_front_01_l, bone2Name=MSTR_jacket_front_l, swing1LimitDegrees=20, swing2LimitDegrees=30, twistLimitDegrees=10}
 {bone1Name=dyn_jacket_bk_02_r, bone2Name=dyn_jacket_bk_01_r, swing1LimitDegrees=20, swing2LimitDegrees=10, twistLimitDegrees=5}
 {bone1Name=dyn_jacket_front_01_r, bone2Name=MSTR_jacket_front_r, swing1LimitDegrees=20, swing2LimitDegrees=30, twistLimitDegrees=10}
 {bone1Name=dyn_jacket_front_02_r, bone2Name=dyn_jacket_front_01_r, swing1LimitDegrees=30, swing2LimitDegrees=30, twistLimitDegrees=10}
 {bone1Name=dyn_jacket_front_03_r, bone2Name=dyn_jacket_front_02_r, swing1LimitDegrees=15, swing2LimitDegrees=10, twistLimitDegrees=10}
 {
  bone1Name=dyn_jacket_side_01_r
  bone2Name=MSTR_jacket_side_r
  swing1LimitDegrees=20
  swing2LimitDegrees=10
  twistLimitDegrees=5
  twistMotion=Locked
 }
 {
  bone1Name=dyn_jacket_side_02_r
  bone2Name=dyn_jacket_side_01_r
  swing1LimitDegrees=30
  swing2LimitDegrees=15
  twistLimitDegrees=5
  twistMotion=Locked
 }
 {bone1Name=dyn_jacket_side_03_r, bone2Name=dyn_jacket_side_02_r, swing1LimitDegrees=15, swing2LimitDegrees=5, twistLimitDegrees=10}
 {
  bone1Name=dyn_jacket_side_02_l
  bone2Name=dyn_jacket_side_01_l
  swing1LimitDegrees=30
  swing2LimitDegrees=15
  twistLimitDegrees=10
  twistMotion=Locked
 }
 {
  bone1Name=dyn_jacket_bk_02_l
  bone2Name=dyn_jacket_side_03_l
  swing1LimitDegrees=45
  swing1Motion=Locked
  swing2LimitDegrees=45
  swing2Motion=Locked
  twistLimitDegrees=45
  twistMotion=Locked
 }
 {
  bone1Name=dyn_jacket_front_03_r
  bone2Name=dyn_jacket_side_03_r
  swing1LimitDegrees=45
  swing1Motion=Free
  swing2LimitDegrees=45
  swing2Motion=Free
  twistLimitDegrees=45
  twistMotion=Free
 }
 {
  bone1Name=dyn_jacket_front_03_r
  bone2Name=dyn_jacket_side_03_r
  swing1LimitDegrees=45
  swing1Motion=Free
  swing2LimitDegrees=45
  swing2Motion=Free
  twistLimitDegrees=45
  twistMotion=Free
 }
 {
  bone1Name=dyn_jacket_side_03_r
  bone2Name=dyn_jacket_bk_02_r
  swing1LimitDegrees=45
  swing1Motion=Free
  swing2LimitDegrees=45
  swing2Motion=Free
  twistLimitDegrees=45
  twistMotion=Free
 }
 {
  bone1Name=dyn_jacket_bk_02_r
  bone2Name=dyn_jacket_bk_02
  swing1LimitDegrees=45
  swing1Motion=Free
  swing2LimitDegrees=45
  swing2Motion=Free
  twistLimitDegrees=45
  twistMotion=Free
 }
 {
  bone1Name=dyn_jacket_bk_02_l
  bone2Name=dyn_jacket_bk_02
  swing1LimitDegrees=45
  swing1Motion=Free
  swing2LimitDegrees=45
  swing2Motion=Free
  twistLimitDegrees=45
  twistMotion=Free
 }
 {
  bone1Name=dyn_jacket_front_03_l
  bone2Name=dyn_jacket_side_03_l
  swing1LimitDegrees=45
  swing1Motion=Free
  swing2LimitDegrees=45
  swing2Motion=Free
  twistLimitDegrees=45
  twistMotion=Free
 }
 {
  bone1Name=dyn_jacket_side_03_l
  bone2Name=dyn_jacket_bk_02_l
  swing1LimitDegrees=45
  swing1Motion=Free
  swing2LimitDegrees=45
  swing2Motion=Free
  twistLimitDegrees=45
  twistMotion=Free
 }
 {bone1Name=dyn_jacket_bk_01_l, bone2Name=spine_01, swing1LimitDegrees=30, swing2LimitDegrees=15, twistLimitDegrees=10}
]
