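# Post-process style AnimBlueprint for the F_MED_OmelettePop cosmetic body, adding secondary motion and corrective bones to the pose copied from the parent mesh.
# Cached stages chain Control Rig and AnimDynamics, jacket RigidBody (skydive variant), belt RigidBody (backbling variant), then corrections and a frontend Control Rig.
# Egg bones use one of two weighted constraint chains, selected by whether rigid bodies are enabled.
# EventGraph derives leg-bend alphas, LOD-based belt offset and state flags.
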AnimBlueprint /BRCosmetics/Characters/Player/Female/Medium/Bodies/F_MED_OmelettePop/Meshes/F_MED_OmelettePop_AnimBP
target_skeleton=&/F_MED_OmelettePop_Skeleton
graphs [
 AnimGraph {nodes [
   n1 = Anim Dynamics - Bone: dyn_footLace_01_l(Alpha=0, ExternalForce="0,0,0") {
    angularDampingOverride=0.99
    angularSpringConstant=150
    bAngularSpring=true
    bOverrideAngularDamping=true
    bUseGravityOverride=true
    numSolverIterationsPreUpdate=3
    simulationSpace=RootRelative
    alphaBoolBlend {blendOption=Linear}
    boundBone {boneName=dyn_footLace_01_l}
    componentAppliedLinearAccClamp=<3000 3000 3000>
    componentLinearAccScale=<1 1 1>
    componentLinearVelScale=<1 1 1>
    gravityOverride {z=-980}
    physicsBodyDefinitions [{
      boundBone {boneName=dyn_footLace_01_l}
      boxExtents=<10 10 10>
      constraintSetup {angularLimitsMax=<0 10 10>, angularLimitsMin=<0 -25 -10>, angularTarget=<1 0 0>}
      localJointOffset=<0.2 0 0>
     }]
    relativeSpaceBone {boneName=foot_l}
   }
   n2 = Anim Dynamics - Bone: dyn_footLace_02_l(ComponentPose=@n1, Alpha=0, ExternalForce="0,0,0") {
    angularDampingOverride=0.99
    angularSpringConstant=150
    bAngularSpring=true
    bOverrideAngularDamping=true
    bUseGravityOverride=true
    numSolverIterationsPreUpdate=3
    simulationSpace=RootRelative
    alphaBoolBlend {blendOption=Linear}
    boundBone {boneName=dyn_footLace_02_l}
    componentAppliedLinearAccClamp=<3000 3000 3000>
    componentLinearAccScale=<1 1 1>
    componentLinearVelScale=<1 1 1>
    gravityOverride {z=-980}
    physicsBodyDefinitions [{
      boundBone {boneName=dyn_footLace_02_l}
      boxExtents=<10 10 10>
      constraintSetup {angularLimitsMax=<0 10 10>, angularLimitsMin=<0 -25 -10>, angularTarget=<1 0 0>}
      localJointOffset=<0.2 0 0>
     }]
    relativeSpaceBone {boneName=foot_l}
   }
   n3 = Anim Dynamics - Bone: dyn_footLace_01_r(ComponentPose=@n2, Alpha=0, ExternalForce="0,0,0") {
    angularDampingOverride=0.99
    angularSpringConstant=150
    bAngularSpring=true
    bOverrideAngularDamping=true
    bUseGravityOverride=true
    numSolverIterationsPreUpdate=3
    simulationSpace=RootRelative
    alphaBoolBlend {blendOption=Linear}
    boundBone {boneName=dyn_footLace_01_r}
    componentAppliedLinearAccClamp=<3000 3000 3000>
    componentLinearAccScale=<1 1 1>
    componentLinearVelScale=<1 1 1>
    gravityOverride {z=-980}
    physicsBodyDefinitions [{
      boundBone {boneName=dyn_footLace_01_r}
      boxExtents=<10 10 10>
      constraintSetup {angularLimitsMax=<0 10 10>, angularLimitsMin=<0 -25 -10>, angularTarget=<1 0 0>}
      localJointOffset=<0.2 0 0>
     }]
    relativeSpaceBone {boneName=foot_r}
   }
   Anim Dynamics - Bone: dyn_footLace_02_r(ComponentPose=@n3, Alpha=0, ExternalForce="0,0,0") {
    angularDampingOverride=0.99
    angularSpringConstant=150
    bAngularSpring=true
    bOverrideAngularDamping=true
    bUseGravityOverride=true
    numSolverIterationsPreUpdate=3
    simulationSpace=RootRelative
    alphaBoolBlend {blendOption=Linear}
    boundBone {boneName=dyn_footLace_02_r}
    componentAppliedLinearAccClamp=<3000 3000 3000>
    componentLinearAccScale=<1 1 1>
    componentLinearVelScale=<1 1 1>
    gravityOverride {z=-980}
    physicsBodyDefinitions [{
      boundBone {boneName=dyn_footLace_02_r}
      boxExtents=<10 10 10>
      constraintSetup {angularLimitsMax=<0 10 10>, angularLimitsMin=<0 -25 -10>, angularTarget=<1 0 0>}
      localJointOffset=<0.2 0 0>
     }]
    relativeSpaceBone {boneName=foot_r}
   }
   n4 = "Constraint - def_egg01 (PPPPPPP)"!(
    ComponentPose=Local To Component(LocalPose=Use cached pose 'PreParent'())
    Alpha=1
    ConstraintWeights_0=0.0006
    ConstraintWeights_1:float=0.0004
    ConstraintWeights_2:float=0.0003
    ConstraintWeights_3:float=0.0329
    ConstraintWeights_4:float=0.1567
    ConstraintWeights_5:float=0.6618
    ConstraintWeights_6:float=0.1472
   ) -> (Pose) {
    alphaBoolBlend {blendOption=Linear}
    boneToModify {boneName=def_egg01}
    constraintSetup [
     {targetBone {boneName=spine_05}}
     {targetBone {boneName=clavicle_r}}
     {targetBone {boneName=spine_04}}
     {targetBone {boneName=belt7}}
     {targetBone {boneName=belt2}}
     {targetBone {boneName=belt3}}
     {targetBone {boneName=belt6}}
    ]
    constraintWeights [1,1,1,1,1,1,1]
   }
   n5 = "Constraint - def_egg02 (PPPPPPP)"!(
    ComponentPose=@n4
    Alpha=1
    ConstraintWeights_0=0.006
    ConstraintWeights_1:float=0.0007
    ConstraintWeights_2:float=0.0043
    ConstraintWeights_3:float=0.003
    ConstraintWeights_4:float=0.4953
    ConstraintWeights_5:float=0.4514
    ConstraintWeights_6:float=0.0394
   ) -> (Pose) {
    alphaBoolBlend {blendOption=Linear}
    boneToModify {boneName=def_egg02}
    constraintSetup [
     {targetBone {boneName=spine_05}}
     {targetBone {boneName=clavicle_l}}
     {targetBone {boneName=clavicle_r}}
     {targetBone {boneName=spine_04}}
     {targetBone {boneName=belt2}}
     {targetBone {boneName=belt3}}
     {targetBone {boneName=belt6}}
    ]
    constraintWeights [1,1,1,1,1,1,1]
   }
   n6 = "Constraint - def_egg05 (PPPPPP)"!(
    ComponentPose=@n5
    Alpha=1
    ConstraintWeights_0=0.3614
    ConstraintWeights_1:float=0.0006
    ConstraintWeights_2:float=0.5351
    ConstraintWeights_3:float=0.0676
    ConstraintWeights_4:float=0.0291
    ConstraintWeights_5:float=0.0063
   ) -> (Pose) {
    alphaBoolBlend {blendOption=Linear}
    boneToModify {boneName=def_egg05}
    constraintSetup [
     {targetBone {boneName=spine_05}}
     {targetBone {boneName=clavicle_l}}
     {targetBone {boneName=clavicle_r}}
     {targetBone {boneName=spine_04}}
     {targetBone {boneName=upperarm_twist_01_r}}
     {targetBone {boneName=deform_upperarm_up_r}}
    ]
    constraintWeights [1,1,1,1,1,1]
   }
   n7 = "Constraint - def_egg06 (PPPPPPPP)"!(
    ComponentPose=@n6
    Alpha=1
    ConstraintWeights_0=0.5189
    ConstraintWeights_1:float=0.0534
    ConstraintWeights_2:float=0.1359
    ConstraintWeights_3:float=0.2645
    ConstraintWeights_4:float=0.0226
    ConstraintWeights_5:float=0.0035
    ConstraintWeights_6:float=0.0013
    ConstraintWeights_7:float=0.0001
   ) -> (Pose) {
    alphaBoolBlend {blendOption=Linear}
    boneToModify {boneName=def_egg06}
    constraintSetup [
     {targetBone {boneName=spine_05}}
     {targetBone {boneName=clavicle_l}}
     {targetBone {boneName=clavicle_r}}
     {targetBone {boneName=spine_04}}
     {targetBone {boneName=upperarm_twist_01_r}}
     {targetBone {boneName=spine_03}}
     {targetBone {boneName=spine_02}}
     {targetBone {boneName=deform_upperarm_up_r}}
    ]
    constraintWeights [1,1,1,1,1,1,1,1]
   }
   n8 = "Constraint - def_egg04 (PPPPPP)"!(
    ComponentPose=@n7
    Alpha=1
    ConstraintWeights_0=0.5906
    ConstraintWeights_1:float=0
    ConstraintWeights_2:float=0.4093
    ConstraintWeights_3:float=0
    ConstraintWeights_4:float=0
    ConstraintWeights_5:float=0
   ) -> (Pose) {
    alphaBoolBlend {blendOption=Linear}
    boneToModify {boneName=def_egg04}
    constraintSetup [
     {targetBone {boneName=spine_05}}
     {targetBone {boneName=clavicle_r}}
     {targetBone {boneName=spine_04}}
     {targetBone {boneName=upperarm_twist_01_r}}
     {targetBone {boneName=deform_upperarm_fr_r}}
     {targetBone {boneName=deform_upperarm_up_r}}
    ]
    constraintWeights [1,1,1,1,1,1]
   }
   n9 = "Constraint - def_egg03 (PPPPPPPP)"!(
    ComponentPose=@n8
    Alpha=1
    ConstraintWeights_0=0.454
    ConstraintWeights_1:float=0.0001
    ConstraintWeights_2:float=0.5454
    ConstraintWeights_3:float=0.0003
    ConstraintWeights_4:float=0.0002
    ConstraintWeights_5:float=0
    ConstraintWeights_6:float=0
    ConstraintWeights_7:float=0
   ) -> (Pose) {
    alphaBoolBlend {blendOption=Linear}
    boneToModify {boneName=def_egg03}
    constraintSetup [
     {targetBone {boneName=spine_05}}
     {targetBone {boneName=clavicle_r}}
     {targetBone {boneName=spine_04}}
     {targetBone {boneName=spine_03}}
     {targetBone {boneName=spine_02}}
     {targetBone {boneName=dyn_jacket_front_01_l}}
     {targetBone {boneName=dyn_jacket_front_01_r}}
     {targetBone {boneName=dyn_jacket_front_02_r}}
    ]
    constraintWeights [1,1,1,1,1,1,1,1]
   }
   n10 = "Constraint - def_egg08 (PPPPPPPP)"!(
    ComponentPose=@n9
    Alpha=1
    ConstraintWeights_0=0.0013
    ConstraintWeights_1:float=0.0032
    ConstraintWeights_2:float=0
    ConstraintWeights_3:float=0.0017
    ConstraintWeights_4:float=0.0002
    ConstraintWeights_5:float=0.4756
    ConstraintWeights_6:float=0.3037
    ConstraintWeights_7:float=0.2143
   ) -> (Pose) {
    alphaBoolBlend {blendOption=Linear}
    boneToModify {boneName=def_egg08}
    constraintSetup [
     {targetBone {boneName=spine_05}}
     {targetBone {boneName=spine_04}}
     {targetBone {boneName=spine_03}}
     {targetBone {boneName=dyn_jacket_front_01_l}}
     {targetBone {boneName=dyn_chicketnbag}}
     {targetBone {boneName=belt4}}
     {targetBone {boneName=belt1}}
     {targetBone {boneName=belt5}}
    ]
    constraintWeights [1,1,1,1,1,1,1,1]
   }
   n11 = "Constraint - def_egg07 (PPPPPPPP)"!(
    ComponentPose=@n10
    Alpha=1
    ConstraintWeights_0=0.0003
    ConstraintWeights_1:float=0.0008
    ConstraintWeights_2:float=0.0005
    ConstraintWeights_3:float=0
    ConstraintWeights_4:float=0.1769
    ConstraintWeights_5:float=0.2422
    ConstraintWeights_6:float=0.1332
    ConstraintWeights_7:float=0.4462
   ) -> (Pose) {
    alphaBoolBlend {blendOption=Linear}
    boneToModify {boneName=def_egg07}
    constraintSetup [
     {targetBone {boneName=spine_05}}
     {targetBone {boneName=spine_04}}
     {targetBone {boneName=dyn_jacket_front_01_l}}
     {targetBone {boneName=dyn_jacket_front_02_l}}
     {targetBone {boneName=dyn_chicketnbag}}
     {targetBone {boneName=belt4}}
     {targetBone {boneName=belt1}}
     {targetBone {boneName=belt5}}
    ]
    constraintWeights [1,1,1,1,1,1,1,1]
   }
   n12 = "Constraint - def_egg04 (PPPPPPP)"!(
    ComponentPose=Local To Component(LocalPose=Use cached pose 'PreParent'())
    Alpha=1
    ConstraintWeights_0=0.0001
    ConstraintWeights_1:float=0.5122
    ConstraintWeights_2:float=0.4595
    ConstraintWeights_3:float=0.0214
    ConstraintWeights_4:float=0.0055
    ConstraintWeights_5:float=0.0001
    ConstraintWeights_6:float=0.0011
   ) -> (Pose) {
    alphaBoolBlend {blendOption=Linear}
    boneToModify {boneName=def_egg04}
    constraintSetup [
     {targetBone {boneName=spine_03}}
     {targetBone {boneName=spine_04}}
     {targetBone {boneName=spine_05}}
     {targetBone {boneName=clavicle_r}}
     {targetBone {boneName=upperarm_twist_01_r}}
     {targetBone {boneName=deform_upperarm_bk_base_r}}
     {targetBone {boneName=deform_upperarm_up_r}}
    ]
    constraintWeights [1,1,1,1,1,1,1]
   }
   n13 = "Constraint - def_egg03 (PPPPP)"!(
    ComponentPose=@n12
    Alpha=1
    ConstraintWeights_0=0.0024
    ConstraintWeights_1:float=0.7806
    ConstraintWeights_2:float=0.214
    ConstraintWeights_3:float=0.0019
    ConstraintWeights_4:float=0.0011
   ) -> (Pose) {
    alphaBoolBlend {blendOption=Linear}
    boneToModify {boneName=def_egg03}
    constraintSetup [
     {targetBone {boneName=spine_03}}
     {targetBone {boneName=spine_04}}
     {targetBone {boneName=spine_05}}
     {targetBone {boneName=clavicle_r}}
     {targetBone {boneName=upperarm_twist_01_r}}
    ]
    constraintWeights [1,1,1,1,1]
   }
   n14 = "Constraint - def_egg08 (PPPPPPPP)"(
    ComponentPose=@n13
    Alpha=1
    ConstraintWeights_0=0.0438
    ConstraintWeights_1=0.5095
    ConstraintWeights_2=0.0129
    ConstraintWeights_3=0.2105
    ConstraintWeights_4=0.0049
    ConstraintWeights_5=0.0112
    ConstraintWeights_6=0.1875
    ConstraintWeights_7=0.0196
   ) {
    alphaBoolBlend {blendOption=Linear}
    boneToModify {boneName=def_egg08}
    constraintSetup [
     {targetBone {boneName=spine_03}}
     {targetBone {boneName=spine_04}}
     {targetBone {boneName=spine_05}}
     {targetBone {boneName=spine_02}}
     {targetBone {boneName=thigh_twist_01_l}}
     {targetBone {boneName=spine_01}}
     {targetBone {boneName=pelvis}}
     {targetBone {boneName=deform_groin_l}}
    ]
    constraintWeights [1,1,1,1,1,1,1,1]
   }
   n15 = "Constraint - def_egg07 (PPPPPPPP)"(
    ComponentPose=@n14
    Alpha=1
    ConstraintWeights_0=0.032
    ConstraintWeights_1=0.3218
    ConstraintWeights_2=0.0046
    ConstraintWeights_3=0.2273
    ConstraintWeights_4=0.0459
    ConstraintWeights_5=0.008
    ConstraintWeights_6=0.3195
    ConstraintWeights_7=0.041
   ) {
    alphaBoolBlend {blendOption=Linear}
    boneToModify {boneName=def_egg07}
    constraintSetup [
     {targetBone {boneName=spine_03}}
     {targetBone {boneName=spine_04}}
     {targetBone {boneName=spine_05}}
     {targetBone {boneName=spine_02}}
     {targetBone {boneName=thigh_twist_01_l}}
     {targetBone {boneName=spine_01}}
     {targetBone {boneName=pelvis}}
     {targetBone {boneName=deform_groin_l}}
    ]
    constraintWeights [1,1,1,1,1,1,1,1]
   }
   n16 = "Constraint - def_egg01 (PPPPPPPP)"!(
    ComponentPose=@n15
    Alpha=1
    ConstraintWeights_0=0.1184
    ConstraintWeights_1:float=0.0356
    ConstraintWeights_2:float=0.0349
    ConstraintWeights_3:float=0.1643
    ConstraintWeights_4:float=0.0568
    ConstraintWeights_5:float=0.4816
    ConstraintWeights_6:float=0.0259
    ConstraintWeights_7:float=0.0824
   ) -> (Pose) {
    alphaBoolBlend {blendOption=Linear}
    boneToModify {boneName=def_egg01}
    constraintSetup [
     {targetBone {boneName=spine_03}}
     {targetBone {boneName=spine_04}}
     {targetBone {boneName=spine_05}}
     {targetBone {boneName=spine_02}}
     {targetBone {boneName=thigh_twist_01_l}}
     {targetBone {boneName=pelvis}}
     {targetBone {boneName=deform_glute_l}}
     {targetBone {boneName=def_backBeltAlt}}
    ]
    constraintWeights [1,1,1,1,1,1,1,1]
   }
   n17 = "Constraint - def_egg02 (PPPPPPPP)"!(
    ComponentPose=@n16
    Alpha=1
    ConstraintWeights_0=0.0698
    ConstraintWeights_1:float=0.0337
    ConstraintWeights_2:float=0.0303
    ConstraintWeights_3:float=0.0763
    ConstraintWeights_4:float=0.0102
    ConstraintWeights_5:float=0.1891
    ConstraintWeights_6:float=0.0079
    ConstraintWeights_7:float=0.5827
   ) -> (Pose) {
    alphaBoolBlend {blendOption=Linear}
    boneToModify {boneName=def_egg02}
    constraintSetup [
     {targetBone {boneName=spine_03}}
     {targetBone {boneName=spine_04}}
     {targetBone {boneName=spine_05}}
     {targetBone {boneName=spine_02}}
     {targetBone {boneName=thigh_twist_01_l}}
     {targetBone {boneName=pelvis}}
     {targetBone {boneName=deform_glute_l}}
     {targetBone {boneName=def_backBeltAlt}}
    ]
    constraintWeights [1,1,1,1,1,1,1,1]
   }
   n18 = "Constraint - def_egg06 (PPPPPPP)"!(
    ComponentPose=@n17
    Alpha=1
    ConstraintWeights_0=0.0078
    ConstraintWeights_1:float=0.2558
    ConstraintWeights_2:float=0.4399
    ConstraintWeights_3:float=0.2247
    ConstraintWeights_4:float=0.0306
    ConstraintWeights_5:float=0.041
    ConstraintWeights_6:float=0.0003
   ) -> (Pose) {
    alphaBoolBlend {blendOption=Linear}
    boneToModify {boneName=def_egg06}
    constraintSetup [
     {targetBone {boneName=spine_03}}
     {targetBone {boneName=spine_04}}
     {targetBone {boneName=spine_05}}
     {targetBone {boneName=clavicle_r}}
     {targetBone {boneName=upperarm_twist_01_r}}
     {targetBone {boneName=deform_upperarm_bk_base_r}}
     {targetBone {boneName=deform_upperarm_up_r}}
    ]
    constraintWeights [1,1,1,1,1,1,1]
   }
   n19 = "Constraint - def_egg05 (PPPPPPP)"!(
    ComponentPose=@n18
    Alpha=1
    ConstraintWeights_0=0.0007
    ConstraintWeights_1:float=0.0992
    ConstraintWeights_2:float=0.2848
    ConstraintWeights_3:float=0.434
    ConstraintWeights_4:float=0.0678
    ConstraintWeights_5:float=0.0936
    ConstraintWeights_6:float=0.0197
   ) -> (Pose) {
    alphaBoolBlend {blendOption=Linear}
    boneToModify {boneName=def_egg05}
    constraintSetup [
     {targetBone {boneName=spine_03}}
     {targetBone {boneName=spine_04}}
     {targetBone {boneName=spine_05}}
     {targetBone {boneName=clavicle_r}}
     {targetBone {boneName=upperarm_twist_01_r}}
     {targetBone {boneName=deform_upperarm_bk_base_r}}
     {targetBone {boneName=deform_upperarm_up_r}}
    ]
    constraintWeights [1,1,1,1,1,1,1]
   }
   Output Pose(Result=Blend Poses by bool(BlendPose_0=Component To Local(ComponentPose=@n11), BlendPose_1=Component To Local(ComponentPose=@n19), bActiveValue=Get bRigidBodyEnabled()) {blendType=Linear}) {layerGroup=None}
   n20 = "Constraint - def_backBeltAlt (PPPPPPPP)"!(
    ComponentPose=Local To Component(LocalPose=Use cached pose 'BeltPhysics'())
    Alpha=Get RBANDisabled() -> (RBANDisabled:double)
    ConstraintWeights_0=0.072
    ConstraintWeights_1:float=0.2907
    ConstraintWeights_2:float=0.0373
    ConstraintWeights_3:float=0.1722
    ConstraintWeights_4:float=0.3688
    ConstraintWeights_5:float=0.0327
    ConstraintWeights_6:float=0.0008
    ConstraintWeights_7:float=0.0255
   ) -> (Pose) {
    alphaBoolBlend {blendOption=Linear}
    boneToModify {boneName=def_backBeltAlt}
    constraintSetup [
     {targetBone {boneName=spine_05}}
     {targetBone {boneName=spine_04}}
     {targetBone {boneName=upperarm_twist_01_l}}
     {targetBone {boneName=spine_03}}
     {targetBone {boneName=spine_02}}
     {targetBone {boneName=spine_01}}
     {targetBone {boneName=dyn_jacket_bk_01}}
     {targetBone {boneName=dyn_jacket_bk_01_l}}
    ]
    constraintWeights [1,1,1,1,1,1,1,1]
   }
   n21 = "Transform (Modify) Bone - Bone: def_backBeltAlt"(ComponentPose=@n20, Alpha=1, Translation=Get BeltBackOffset() -> (BeltBackOffset:Vector), Rotation="0,0,0", Scale="1,1,1") {translationMode=BMM_Additive, translationSpace=BCS_BoneSpace, alphaBoolBlend {blendOption=Linear}, boneToModify {boneName=def_backBeltAlt}}
   n22 = "Transform (Modify) Bone - Bone: lowerarm_r"(ComponentPose=@n21, Alpha=1, Translation="0,1,0", Rotation="0,0,0", Scale="1,1,1") {
    scaleMode=BMM_Replace
    scaleSpace=BCS_BoneSpace
    translationMode=BMM_Additive
    alphaBoolBlend {blendOption=Linear}
    boneToModify {boneName=lowerarm_r}
    translation {y=1}
   }
   n23 = "Transform (Modify) Bone - Bone: lowerarm_twist_02_r"(ComponentPose=@n22, Alpha=1, Translation="0,1,0", Rotation="0,0,0", Scale="0.7,0.7,0.7") {
    scaleMode=BMM_Replace
    scaleSpace=BCS_BoneSpace
    translationMode=BMM_Additive
    alphaBoolBlend {blendOption=Linear}
    boneToModify {boneName=lowerarm_twist_02_r}
    scale=<0.7 0.7 0.7>
    translation {y=1}
   }
   n24 = "Transform (Modify) Bone - Bone: lowerarm_twist_01_r"(ComponentPose=@n23, Alpha=1, Translation="0.5,-1,2", Rotation="0,0,0", Scale="0.6,0.6,0.6") {
    scaleMode=BMM_Replace
    scaleSpace=BCS_BoneSpace
    translationMode=BMM_Additive
    alphaBoolBlend {blendOption=Linear}
    boneToModify {boneName=lowerarm_twist_01_r}
    scale=<0.6 0.6 0.6>
    translation=<0.5 -1 2>
   }
   n25 = "Transform (Modify) Bone - Bone: lowerarm_l"(ComponentPose=@n24, Alpha=1, Translation="0,1,0", Rotation="0,0,0", Scale="1,1,1") {
    scaleMode=BMM_Replace
    scaleSpace=BCS_BoneSpace
    translationMode=BMM_Additive
    alphaBoolBlend {blendOption=Linear}
    boneToModify {boneName=lowerarm_l}
    translation {y=1}
   }
   n26 = "Transform (Modify) Bone - Bone: lowerarm_twist_02_l"(ComponentPose=@n25, Alpha=1, Translation="0,1,0", Rotation="0,0,0", Scale="0.7,0.7,0.7") {
    scaleMode=BMM_Replace
    scaleSpace=BCS_BoneSpace
    alphaBoolBlend {blendOption=Linear}
    boneToModify {boneName=lowerarm_twist_02_l}
    scale=<0.7 0.7 0.7>
    translation {y=1}
   }
   n27 = "Transform (Modify) Bone - Bone: lowerarm_twist_01_l"(ComponentPose=@n26, Alpha=1, Translation="0.5,-1,2", Rotation="0,0,0", Scale="0.6,0.6,0.6") {
    scaleMode=BMM_Replace
    scaleSpace=BCS_BoneSpace
    translationMode=BMM_Additive
    alphaBoolBlend {blendOption=Linear}
    boneToModify {boneName=lowerarm_twist_01_l}
    scale=<0.6 0.6 0.6>
    translation=<0.5 -1 2>
   }
   n28 = Control Rig!(
    Source=Component To Local(ComponentPose=@n27)
    bAlphaBoolEnabled:bool=True
    isEmote:bool=Get bIsPlayingEmote()
    isFrontend:bool=Get bIsInFrontEnd()
    isSkyDive:bool=Get bIsSkydiveDiveMode()
    isSkyDiving:bool=Get bIsSkydiveFloating()
   ) -> (Pose) {
    alphaInputType=Bool
    alphaBoolBlend {blendOption=Linear}
    controlRigAssetReference {blueprintRigClass=&/F_MED_OmelettePop_Frontend_CTRL.F_MED_OmelettePop_Frontend_CTRL_C}
   }
   Save cached pose 'PreParent'(Pose=@n28) {cacheName=PreParent}
   n29 = Control Rig!(
    Source=Copy Pose From Mesh() {bUseAttachedParent=true}
    Alpha:float
    CalfL:double=Get CalfBendL() -> (CalfBendL:double)
    CalfR:double=Get CalfBendR() -> (CalfBendR:double)
    isSliding:bool=Get bIsSliding()
    isBaseBallSliding:bool=Get bIsBaseballSliding()
    isSkyDiveFloat:bool=Get bIsSkydiveFloating()
    isParachute:bool=Get isParachute() -> (isParachute:bool)
    isRBAN:bool=Get bRigidBodyEnabled()
    isSwingAxe:bool=Get IsPickaxeAttack() -> (IsPickaxeAttack:bool)
    isFrontEnd:bool=false
    isEmote=false
   ) -> (Pose) {
    alphaBoolBlend {blendOption=Linear}
    controlRigAssetReference {blueprintRigClass=&/F_MED_OmelettePop_CtrlRig.F_MED_OmelettePop_CtrlRig_C}
   }
   n30 = "Constraint - dyn_armRing_l (PPPPPPPP)"!(
    ComponentPose=Local To Component(LocalPose=@n29)
    Alpha=1
    ConstraintWeights_0=0.065
    ConstraintWeights_1:float=0.6603
    ConstraintWeights_2:float=0.2527
    ConstraintWeights_3:float=0.0076
    ConstraintWeights_4:float=0
    ConstraintWeights_5:float=0.0138
    ConstraintWeights_6:float=0
    ConstraintWeights_7:float=0.0005
   ) -> (Pose) {
    alphaBoolBlend {blendOption=Linear}
    boneToModify {boneName=dyn_armRing_l}
    constraintSetup [
     {targetBone {boneName=upperarm_twist_01_l}}
     {targetBone {boneName=upperarm_l}}
     {targetBone {boneName=upperarm_twist_02_l}}
     {targetBone {boneName=lowerarm_l}}
     {targetBone {boneName=lowerarm_twist_01_l}}
     {targetBone {boneName=lowerarm_twist_02_l}}
     {targetBone {boneName=deform_elbow_out_l}}
     {targetBone {boneName=deform_elbow_in_l}}
    ]
    constraintWeights [1,1,1,1,1,1,1,1]
   }
   n31 = "Constraint - dyn_armRing_r (PPPPPPPP)"!(
    ComponentPose=@n30
    Alpha=1
    ConstraintWeights_0=0.0805
    ConstraintWeights_1:float=0.5638
    ConstraintWeights_2:float=0.3211
    ConstraintWeights_3:float=0
    ConstraintWeights_4:float=0.012
    ConstraintWeights_5:float=0.0218
    ConstraintWeights_6:float=0
    ConstraintWeights_7:float=0.0008
   ) -> (Pose) {
    alphaBoolBlend {blendOption=Linear}
    boneToModify {boneName=dyn_armRing_r}
    constraintSetup [
     {targetBone {boneName=upperarm_twist_01_r}}
     {targetBone {boneName=upperarm_r}}
     {targetBone {boneName=upperarm_twist_02_r}}
     {targetBone {boneName=lowerarm_twist_01_r}}
     {targetBone {boneName=lowerarm_r}}
     {targetBone {boneName=lowerarm_twist_02_r}}
     {targetBone {boneName=deform_elbow_out_r}}
     {targetBone {boneName=deform_elbow_in_r}}
    ]
    constraintWeights [1,1,1,1,1,1,1,1]
   }
   n32 = Anim Dynamics - Bone: dyn_armRing_l!(SimulationSpace:AnimPhysSimSpaceType Enum, ComponentPose=@n31, Alpha=1, ExternalForce="0,0,0") -> (Pose) {
    angularDampingOverride=0.99
    angularSpringConstant=200
    bAngularSpring=true
    bOverrideAngularDamping=true
    bUseGravityOverride=true
    numSolverIterationsPreUpdate=3
    simulationSpace=RootRelative
    alphaBoolBlend {blendOption=Linear}
    boundBone {boneName=dyn_armRing_l}
    componentAppliedLinearAccClamp=<3000 3000 3000>
    componentLinearAccScale=<3 3 3>
    componentLinearVelScale=<3 3 3>
    gravityOverride {z=-980}
    physicsBodyDefinitions [{
      boundBone {boneName=dyn_armRing_l}
      boxExtents=<10 10 10>
      constraintSetup {angularLimitsMax=<0 0 3>, angularLimitsMin=<0 0 -75>, angularTarget=<0 0 0>}
      localJointOffset=<0.2 0 0>
     }]
   }
   n33 = Anim Dynamics - Bone: dyn_armRing_r!(SimulationSpace:AnimPhysSimSpaceType Enum, ComponentPose=@n32, Alpha=1, ExternalForce="0,0,0") -> (Pose) {
    angularDampingOverride=0.99
    angularSpringConstant=200
    bAngularSpring=true
    bGravityOverrideInSimSpace=true
    bOverrideAngularDamping=true
    bUseGravityOverride=true
    numSolverIterationsPreUpdate=3
    simulationSpace=RootRelative
    alphaBoolBlend {blendOption=Linear}
    boundBone {boneName=dyn_armRing_r}
    componentAppliedLinearAccClamp=<3000 3000 3000>
    componentLinearAccScale=<3 3 3>
    componentLinearVelScale=<3 3 3>
    gravityOverride {z=-980}
    physicsBodyDefinitions [{
      boundBone {boneName=dyn_armRing_r}
      boxExtents=<10 10 10>
      constraintSetup {angularLimitsMax=<0 0 3>, angularLimitsMin=<0 0 70>, angularTarget=<0 0 0>}
      localJointOffset=<0.2 0 0>
     }]
   }
   n34 = Anim Dynamics - Bone: dyn_pants_fr_l!(SimulationSpace:AnimPhysSimSpaceType Enum, ComponentPose=@n33, Alpha=1, ExternalForce="0,0,0") -> (Pose) {
    angularDampingOverride=0.99
    angularSpringConstant=200
    bAngularSpring=true
    bOverrideAngularDamping=true
    bOverrideLinearDamping=true
    bUseGravityOverride=true
    linearDampingOverride=0.9
    numSolverIterationsPreUpdate=3
    simulationSpace=RootRelative
    alphaBoolBlend {blendOption=Linear}
    boundBone {boneName=dyn_pants_fr_l}
    componentAppliedLinearAccClamp=<5000 5000 5000>
    componentLinearAccScale=<3 3 3>
    componentLinearVelScale=<3 3 3>
    gravityOverride {z=-980}
    physicsBodyDefinitions [{
      boundBone {boneName=dyn_pants_fr_l}
      boxExtents=<10 10 10>
      constraintSetup {angularLimitsMax=<0 5 10>, angularLimitsMin=<0 -5 -30>, angularTarget=<0 0 0>, linearAxesMax=<0 2 1>, linearAxesMin=<0 0 -1>}
      localJointOffset=<1 0 0>
     }]
    relativeSpaceBone {boneName=thigh_l}
   }
   n35 = Anim Dynamics - Bone: dyn_pants_out_l!(SimulationSpace:AnimPhysSimSpaceType Enum, ComponentPose=@n34, Alpha=1, ExternalForce="0,0,0") -> (Pose) {
    angularDampingOverride=0.99
    angularSpringConstant=200
    bAngularSpring=true
    bOverrideAngularDamping=true
    bOverrideLinearDamping=true
    bUseGravityOverride=true
    linearDampingOverride=0.9
    numSolverIterationsPreUpdate=3
    simulationSpace=RootRelative
    alphaBoolBlend {blendOption=Linear}
    boundBone {boneName=dyn_pants_out_l}
    componentAppliedLinearAccClamp=<5000 5000 5000>
    componentLinearAccScale=<3 3 3>
    componentLinearVelScale=<3 3 3>
    gravityOverride {z=-980}
    physicsBodyDefinitions [{
      boundBone {boneName=dyn_pants_out_l}
      boxExtents=<10 10 10>
      constraintSetup {angularLimitsMax=<0 15 15>, angularLimitsMin=<0 -15 -60>, angularTarget=<0 0 0>}
      localJointOffset=<1 0 0>
     }]
    relativeSpaceBone {boneName=thigh_l}
   }
   n36 = Anim Dynamics - Bone: dyn_pants_in_l!(SimulationSpace:AnimPhysSimSpaceType Enum, ComponentPose=@n35, Alpha=1, ExternalForce="0,0,0") -> (Pose) {
    angularDampingOverride=0.99
    angularSpringConstant=200
    bAngularSpring=true
    bOverrideAngularDamping=true
    bOverrideLinearDamping=true
    bUseGravityOverride=true
    linearDampingOverride=0.9
    numSolverIterationsPreUpdate=3
    simulationSpace=RootRelative
    alphaBoolBlend {blendOption=Linear}
    boundBone {boneName=dyn_pants_in_l}
    componentAppliedLinearAccClamp=<5000 5000 5000>
    componentLinearAccScale=<3 3 3>
    componentLinearVelScale=<3 3 3>
    gravityOverride {z=-980}
    physicsBodyDefinitions [{
      boundBone {boneName=dyn_pants_in_l}
      boxExtents=<10 10 10>
      constraintSetup {angularLimitsMax=<0 15 10>, angularLimitsMin=<0 -15 -30>, angularTarget=<0 0 0>}
      localJointOffset=<1 0 0>
     }]
    relativeSpaceBone {boneName=thigh_l}
   }
   n37 = Anim Dynamics - Bone: dyn_pants_bk_l!(SimulationSpace:AnimPhysSimSpaceType Enum, ComponentPose=@n36, Alpha=Get CalfBendL(), ExternalForce="0,0,0") -> (Pose) {
    angularDampingOverride=0.99
    angularSpringConstant=200
    bAngularSpring=true
    bOverrideAngularDamping=true
    bOverrideLinearDamping=true
    bUseGravityOverride=true
    linearDampingOverride=0.9
    numSolverIterationsPreUpdate=3
    simulationSpace=RootRelative
    alphaBoolBlend {blendOption=Linear}
    boundBone {boneName=dyn_pants_bk_l}
    componentAppliedLinearAccClamp=<5000 5000 5000>
    componentLinearAccScale=<3 3 3>
    componentLinearVelScale=<3 3 3>
    gravityOverride {z=-980}
    physicsBodyDefinitions [{
      boundBone {boneName=dyn_pants_bk_l}
      boxExtents=<10 10 10>
      constraintSetup {angularLimitsMax=<0 15 10>, angularLimitsMin=<0 -15 -30>, angularTarget=<0 0 0>, linearAxesMax=<0 1 1>, linearAxesMin=<0 -1 -1>}
      localJointOffset=<1 0 0>
     }]
    relativeSpaceBone {boneName=thigh_l}
   }
   n38 = Anim Dynamics - Bone: dyn_pants_fr_r!(SimulationSpace:AnimPhysSimSpaceType Enum, ComponentPose=@n37, Alpha=1, ExternalForce="0,0,0") -> (Pose) {
    angularDampingOverride=0.99
    angularSpringConstant=200
    bAngularSpring=true
    bOverrideAngularDamping=true
    bOverrideLinearDamping=true
    bUseGravityOverride=true
    linearDampingOverride=0.9
    numSolverIterationsPreUpdate=3
    simulationSpace=RootRelative
    alphaBoolBlend {blendOption=Linear}
    boundBone {boneName=dyn_pants_fr_r}
    componentAppliedLinearAccClamp=<5000 5000 5000>
    componentLinearAccScale=<3 3 3>
    componentLinearVelScale=<3 3 3>
    gravityOverride {z=-980}
    physicsBodyDefinitions [{
      boundBone {boneName=dyn_pants_fr_r}
      boxExtents=<10 10 10>
      constraintSetup {angularLimitsMax=<0 5 10>, angularLimitsMin=<0 -5 -30>, angularTarget=<0 0 0>, linearAxesMax=<0 2 1>, linearAxesMin=<0 0 -1>}
      localJointOffset=<1 0 0>
     }]
    relativeSpaceBone {boneName=thigh_l}
   }
   n39 = Anim Dynamics - Bone: dyn_pants_out_r!(SimulationSpace:AnimPhysSimSpaceType Enum, ComponentPose=@n38, Alpha=1, ExternalForce="0,0,0") -> (Pose) {
    angularDampingOverride=0.99
    angularSpringConstant=200
    bAngularSpring=true
    bOverrideAngularDamping=true
    bOverrideLinearDamping=true
    bUseGravityOverride=true
    linearDampingOverride=0.9
    numSolverIterationsPreUpdate=3
    simulationSpace=RootRelative
    alphaBoolBlend {blendOption=Linear}
    boundBone {boneName=dyn_pants_out_r}
    componentAppliedLinearAccClamp=<5000 5000 5000>
    componentLinearAccScale=<3 3 3>
    componentLinearVelScale=<3 3 3>
    gravityOverride {z=-980}
    physicsBodyDefinitions [{
      boundBone {boneName=dyn_pants_out_r}
      boxExtents=<10 10 10>
      constraintSetup {angularLimitsMax=<0 15 15>, angularLimitsMin=<0 -15 -60>, angularTarget=<0 0 0>}
      localJointOffset=<1 0 0>
     }]
    relativeSpaceBone {boneName=thigh_l}
   }
   n40 = Anim Dynamics - Bone: dyn_pants_in_r!(SimulationSpace:AnimPhysSimSpaceType Enum, ComponentPose=@n39, Alpha=1, ExternalForce="0,0,0") -> (Pose) {
    angularDampingOverride=0.99
    angularSpringConstant=200
    bAngularSpring=true
    bOverrideAngularDamping=true
    bOverrideLinearDamping=true
    bUseGravityOverride=true
    linearDampingOverride=0.9
    numSolverIterationsPreUpdate=3
    simulationSpace=RootRelative
    alphaBoolBlend {blendOption=Linear}
    boundBone {boneName=dyn_pants_in_r}
    componentAppliedLinearAccClamp=<5000 5000 5000>
    componentLinearAccScale=<3 3 3>
    componentLinearVelScale=<3 3 3>
    gravityOverride {z=-980}
    physicsBodyDefinitions [{
      boundBone {boneName=dyn_pants_in_r}
      boxExtents=<10 10 10>
      constraintSetup {angularLimitsMax=<0 15 10>, angularLimitsMin=<0 -15 -30>, angularTarget=<0 0 0>}
      localJointOffset=<1 0 0>
     }]
    relativeSpaceBone {boneName=thigh_l}
   }
   n41 = Anim Dynamics - Bone: dyn_pants_bk_r!(SimulationSpace:AnimPhysSimSpaceType Enum, ComponentPose=@n40, Alpha=Get CalfBendR(), ExternalForce="0,0,0") -> (Pose) {
    angularDampingOverride=0.99
    angularSpringConstant=200
    bAngularSpring=true
    bOverrideAngularDamping=true
    bOverrideLinearDamping=true
    bUseGravityOverride=true
    linearDampingOverride=0.9
    numSolverIterationsPreUpdate=3
    simulationSpace=RootRelative
    alphaBoolBlend {blendOption=Linear}
    boundBone {boneName=dyn_pants_bk_r}
    componentAppliedLinearAccClamp=<5000 5000 5000>
    componentLinearAccScale=<3 3 3>
    componentLinearVelScale=<3 3 3>
    gravityOverride {z=-980}
    physicsBodyDefinitions [{
      boundBone {boneName=dyn_pants_bk_r}
      boxExtents=<10 10 10>
      constraintSetup {angularLimitsMax=<0 15 10>, angularLimitsMin=<0 -15 -30>, angularTarget=<0 0 0>, linearAxesMax=<0 1 1>, linearAxesMin=<0 -1 -1>}
      localJointOffset=<1 0 0>
     }]
    relativeSpaceBone {boneName=thigh_l}
   }
   n42 = Anim Dynamics - Bone: dyn_hoodie!(
    SimulationSpace:AnimPhysSimSpaceType Enum=Get Simulation Space() -> (Simulation Space:AnimPhysSimSpaceType Enum)
    ComponentPose=@n41
    Alpha=1
    ExternalForce="0,0,0"
   ) -> (Pose) {
    angularDampingOverride=1
    angularSpringConstant=200
    bOverrideAngularDamping=true
    bOverrideLinearDamping=true
    bUseGravityOverride=true
    linearDampingOverride=1
    linearSpringConstant=100
    numSolverIterationsPreUpdate=3
    simulationSpace=RootRelative
    alphaBoolBlend {blendOption=Linear}
    boundBone {boneName=dyn_hoodie}
    componentAppliedLinearAccClamp=<3000 3000 3000>
    componentLinearAccScale=<0.5 0.5 0.5>
    componentLinearVelScale=<0.5 0.5 0.5>
    gravityOverride {z=-980}
    physicsBodyDefinitions [{
      boundBone {boneName=dyn_hoodie}
      boxExtents=<20 20 20>
      constraintSetup {angularLimitsMax=<5 20 10>, angularLimitsMin=<-5 -5 -10>, angularTarget=<1 0 0>, linearAxesMax=<1 1.5 0.5>, linearAxesMin=<0 -1.5 -0.5>}
      localJointOffset=<0.1 0 0>
     }]
   }
   Save cached pose 'AnimDynamics'(Pose=Component To Local(ComponentPose=@n42)) {cacheName=AnimDynamics}
   n43 = RigidBody!(ComponentPose=Local To Component(LocalPose=Use cached pose 'AnimDynamics'()), Alpha, OverrideWorldGravity:Vector="0,0,0", ExternalForce) -> (Pose) {
    bClampLinearTranslationLimitToRefPose=true
    bOverrideWorldGravity=true
    overridePhysicsAsset=&/F_MED_OmelettePop_PhysicsDefault
    simulationSpace=BaseBoneSpace
    alphaBoolBlend {blendOption=Linear}
    baseBoneRef {boneName=pelvis}
    simSpaceSettings {
     dampingAlpha=0
     maxAngularAcceleration=25
     maxAngularVelocity=5
     maxLinearAcceleration=500
     maxLinearVelocity=350
     velocityScaleZ=0.8
     worldAlpha=1
     externalLinearDragV=<0.5 0.5 0.5>
    }
   }
   n44 = RigidBody(ComponentPose=Local To Component(LocalPose=Use cached pose 'AnimDynamics'()), OverrideWorldGravity="0,0,0") {
    bClampLinearTranslationLimitToRefPose=true
    bOverrideWorldGravity=true
    overridePhysicsAsset=&/F_MED_OmelettePop_PhysicsSkyDive
    simulationSpace=BaseBoneSpace
    alphaBoolBlend {blendOption=Linear}
    baseBoneRef {boneName=spine_01}
    simSpaceSettings {
     dampingAlpha=0
     maxAngularAcceleration=25
     maxAngularVelocity=5
     maxLinearAcceleration=150
     maxLinearVelocity=30
     velocityScaleZ=0.7
     worldAlpha=1
     externalLinearDragV=<0.5 0.5 0.5>
    }
   }
   Save cached pose 'PhysicsJacket'(Pose=Blend Poses by bool(BlendPose_0=Component To Local(ComponentPose=@n44), BlendPose_1=Component To Local(ComponentPose=@n43), bActiveValue=Get bIsSkyDiving()) {blendType=Linear}) {cacheName=PhysicsJacket}
   n45 = RigidBody(ComponentPose=Local To Component(LocalPose=Use cached pose 'PhysicsJacket'()), OverrideWorldGravity="0,0,0") {
    bOverrideWorldGravity=true
    overridePhysicsAsset=&/F_MED_OmelettePop_Belt_Physics1
    simulationSpace=BaseBoneSpace
    alphaBoolBlend {blendOption=Linear}
    baseBoneRef {boneName=spine_03}
    simSpaceSettings {
     dampingAlpha=0.5
     maxAngularAcceleration=24
     maxAngularVelocity=6
     maxLinearAcceleration=100
     maxLinearVelocity=10
     velocityScaleZ=0.7
     worldAlpha=1
     externalLinearDragV=<0.5 0.5 0.5>
    }
   }
   n46 = RigidBody(ComponentPose=Local To Component(LocalPose=Use cached pose 'PhysicsJacket'()), OverrideWorldGravity="0,0,0") {
    bOverrideWorldGravity=true
    overridePhysicsAsset=&/F_MED_OmelettePop_BeltBackPac_Physics
    simulationSpace=BaseBoneSpace
    alphaBoolBlend {blendOption=Linear}
    baseBoneRef {boneName=spine_03}
    simSpaceSettings {
     dampingAlpha=0.5
     maxAngularAcceleration=24
     maxAngularVelocity=6
     maxLinearAcceleration=100
     maxLinearVelocity=10
     velocityScaleZ=0.7
     worldAlpha=1
     externalLinearDragV=<0.5 0.5 0.5>
    }
   }
   Save cached pose 'BeltPhysics'(Pose=Blend Poses by bool(
     BlendPose_0=Component To Local(ComponentPose=@n46)
     BlendPose_1=Component To Local(ComponentPose=@n45)
     bActiveValue=Get Is Wearing Backbling()
    ) {blendType=Linear}) {cacheName=BeltPhysics}
  ]}
 EventGraph {nodes [
   Event BlueprintUpdateAnimation()
   > n1 = Gravity Override State Function DEPRECATED(CharacterPartAnimInstance=Self-Reference(), GameTimeInSeconds=GetGameTimeInSeconds(), GravityOverrideParams=Get Gravity Override Params())
   > Set Gravity Override(Gravity Override=@n1)
   n2 = Get Custom Base Mesh() -> (Custom Base Mesh:Skeletal Mesh Component Object Reference)
   > Set CalfBendL(CalfBendL:double=Remap Float(
     Input=GetSocketTransform!(self=@n2, InSocketName=calf_l, TransformSpace=RTS_ParentBoneSpace) -> (
      ReturnValue_Location:Vector
      ReturnValue_Rotation_Roll:float
      ReturnValue_Rotation_Pitch:float
      ReturnValue_Rotation_Yaw:float
      ReturnValue_Scale:Vector
     ).ReturnValue_Rotation_Yaw
     OldRange=(X=80,Y=110)
     NewRange=(X=1,Y=0)
     Clamp=true
    ) -> (Output=~)) -> (Output_Get:double)
   > Set CalfBendR(CalfBendR:double=Remap Float(
     Input=GetSocketTransform(self=@n2, InSocketName=calf_r, TransformSpace=RTS_ParentBoneSpace).ReturnValue_Rotation_Yaw
     OldRange=(X=80,Y=122)
     NewRange=(X=1,Y=0)
     Clamp=true
    ) -> (Output=~)) -> (Output_Get:double)
   n3 = Remap Float(
    Input=GetSocketTransform(self=@n2, InSocketName=thigh_l, TransformSpace=RTS_ParentBoneSpace).ReturnValue_Rotation_Yaw
    OldRange=(X=-50,Y=-100)
    NewRange=(X=0,Y=1)
    Clamp=true
   ) -> (Output=~)
   > Set Thigh Bend L(Thigh Bend L:double=@n3) -> (Output_Get:double)
   n4 = Remap Float(
    Input=GetSocketTransform(self=@n2, InSocketName=thigh_r, TransformSpace=RTS_ParentBoneSpace).ReturnValue_Rotation_Yaw
    OldRange=(X=140,Y=110)
    NewRange=(X=0,Y=1)
    Clamp=true
   ) -> (Output=~)
   > Set ThighBendR(ThighBendR:double=@n4) -> (Output_Get:double)
   > Set BothThighUp(BothThighUp:double="Clamp (Float)"(Value=float + float(A=@n3, B=@n4), Min=0, Max=1)) -> (Output_Get:double)
   n5 = Get bRigidBodyEnabled()
   > Set RBANDisabled(RBANDisabled:double="To Float (Boolean)"(InBool=NOT Boolean(A=@n5))) -> (Output_Get:double)
   > Set BeltBackOffset(BeltBackOffset:Vector=Select!(Option 0:Vector="0,-1,0", Option 1:Vector="0,-3,0", Option 2:Vector="-3,-10,0", Option 3:Vector="-3,-10,0", Index:int=Get CurrentLOD()) -> (ReturnValue:Vector)) -> (Output_Get:Vector)
   n6 = Event OnCharacterCustomizationLoaded()
   Set Custom Base Mesh(execute=[@n6.then, @n7.then], Custom Base Mesh:Skeletal Mesh Component Object Reference=@n6.CustomBaseMesh) -> (Output_Get:Skeletal Mesh Component Object Reference)
   > Set Is Wearing Backbling(Is Wearing Backbling=CheckForBackbling(self=Self-Reference()))
   > Set Simulation Space(Simulation Space:AnimPhysSimSpaceType Enum=Select!(Option 0:AnimPhysSimSpaceType Enum, Option 1:AnimPhysSimSpaceType Enum=World, Index:bool=Get bIsInFrontEnd()) -> (ReturnValue:AnimPhysSimSpaceType Enum)) -> (Output_Get:AnimPhysSimSpaceType Enum)
   n7 = Event BlueprintInitializeAnimation()
   Event OnParachuteOpened()
   > Set isParachute(isParachute:bool=true) -> (Output_Get:bool)
   Event OnParachuteClosed()
   > Set isParachute(isParachute=false)
   Event OnBeginPickaxeSwing() {is=title K2Node_Event}
   > Set IsPickaxeAttack(IsPickaxeAttack:bool=true) -> (Output_Get:bool)
   Event OnEndPickaxeSwing() {is=title K2Node_Event}
   > Set IsPickaxeAttack(IsPickaxeAttack=false)
  ]}
]
variables [
 BeltBackOffset {type=Vector, default=<0 -1 0>}
 BothThighUp {type=number}
 CalfBendL {type=number}
 CalfBendR {type=number}
 Custom Base Mesh {type=SkeletalMeshComponent}
 Gravity Override {type=Vector, default=<0 0 -980>}
 Gravity Override Params {
  type=GravityOverrideParamsStruct
  default {
   onGround_Moving {windAmplitude=<0 0 200>, windFrequency=<0 0 3>, windOffset=<0 -500 0>}
   skydive_FloatingDown {gravityOverride=<0 -700 -600>}
   skydive_GoingUp {gravityOverride=<0 -300 800>}
   skydive_Parachuting {windAmplitude=<200 500 400>}
  }
 }
 Is Wearing Backbling {type=boolean}
 isParachute {type=boolean}
 IsPickaxeAttack {type=boolean}
 RBANDisabled {type=number}
 Simulation Space {type=AnimPhysSimSpaceType, default=RootRelative}
 Thigh Bend L {type=number}
 ThighBendR {type=number}
]
