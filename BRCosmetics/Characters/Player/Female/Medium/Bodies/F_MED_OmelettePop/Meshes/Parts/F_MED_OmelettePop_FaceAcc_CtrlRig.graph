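# Control Rig for the F_MED_OmelettePop face accessory that applies pose-driven corrections to head flap, curl and back-hair bones.
# Spherical pose readers on each upper arm weight per-side offsets, suppressed while isDBNO, plus a combined both-arms correction.
# Head rotation relative to spine_05 offsets and squashes back hair, while isDBNO rotates both flaps.
# Both distance variables measure head to upperarm_r; the curl parent constraints are outside the execution chain.
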ControlRigBlueprint /BRCosmetics/Characters/Player/Female/Medium/Bodies/F_MED_OmelettePop/Meshes/Parts/F_MED_OmelettePop_FaceAcc_CtrlRig
graphs [
 RigVMFunctionLibrary {links [], nodes []}
 RigVMModel {nodes [
   RigUnit_BeginExecution = &RigUnit_BeginExecution()
   n1 = &RigUnit_GetTransform(Item=(Type=Bone,Name="head"), bInitial=False)
   n2 = &RigUnit_GetTransform(Item=(Type=Bone,Name="upperarm_r"), bInitial=False)
   RigVMVariableNode(
    Value=&RigVMFunction_MathDoubleRemap(
     SourceMaximum=16
     SourceMinimum=21
     TargetMaximum=1
     TargetMinimum=0
     Value=&RigVMFunction_MathVectorDistance(A=@n1.Transform.Translation, B=@n2.Transform.Translation).Result
    ).Result
    ExecuteContext=@RigUnit_BeginExecution.ExecutePin
   ) {variable=DistanceToLeft}
   RigVMFunction_MathVectorDistance = &RigVMFunction_MathVectorDistance(
    A=&RigUnit_GetTransform(Item=(Type=Bone,Name="head"), bInitial=False).Transform.Translation
    B=&RigUnit_GetTransform(Item=(Type=Bone,Name="upperarm_r"), bInitial=False).Transform.Translation
   )
   > RigVMVariableNode(Value=&RigVMFunction_MathDoubleRemap(SourceMaximum=16, SourceMinimum=21, TargetMaximum=1, TargetMinimum=0, Value=@RigVMFunction_MathVectorDistance.Result).Result) {variable=DistanceToRights}
   > n3 = &RigUnit_SphericalPoseReader(
    ActiveRegionScaleFactors=(PositiveWidth=3,NegativeWidth=3,PositiveHeight=1,NegativeHeight=3)
    ActiveRegionSize=0.1
    Debug=(bDrawDebug=false,bDraw2D=False,bDrawLocalAxes=False,DebugScale=25,DebugSegments=20,DebugThickness=0.25)
    DriverAxis=(X=1,Y=0,Z=0)
    DriverItem=(Type=Bone,Name="upperarm_l")
    FalloffRegionScaleFactors=(PositiveWidth=1,NegativeWidth=1,PositiveHeight=1,NegativeHeight=1)
    FalloffSize=0.4
    FlipHeightScaling=False
    FlipWidthScaling=False
    OptionalParentItem=(Type=Bone,Name="None")
    RotationOffset=(X=60,Y=0,Z=-10)
   )
   RigVMFunction_MathBoolNot = &RigVMFunction_MathBoolNot(Value=RigVMVariableNode() {variable=isDBNO}.Value)
   n4 = &RigVMFunction_MathBoolToFloat(Value=@RigVMFunction_MathBoolNot.Result)
   > &RigUnit_ModifyTransforms(
    ItemToModify="((Item=(Type=Bone,Name=\"MSTR_flap_l\"),Transform=(Rotation=(X=0,Y=0,Z=0.2164,W=0.9763),Translation=(X=0,Y=-1.5,Z=1),Scale3D=(X=1,Y=1,Z=1))),(Item=(Type=Bone,Name=\"dyn_curl_fr_01_l\"),Transform=(Rotation=(X=0,Y=0,Z=-0.0441,W=0.999),Translation=(X=0.032,Y=1.01 ...(truncated)"
    WeightMinimum=0
    Weight=&RigVMFunction_MathDoubleMul(B=@n4.Result, A=@n3.OutputParam).Result
   )
   > n5 = &RigUnit_SphericalPoseReader(
    ActiveRegionScaleFactors=(PositiveWidth=3,NegativeWidth=3,PositiveHeight=1,NegativeHeight=3)
    ActiveRegionSize=0.1
    Debug=(bDrawDebug=false,bDraw2D=False,bDrawLocalAxes=False,DebugScale=25,DebugSegments=20,DebugThickness=0.25)
    DriverAxis=(X=-1,Y=0,Z=0)
    DriverItem=(Type=Bone,Name="upperarm_r")
    FalloffRegionScaleFactors=(PositiveWidth=1,NegativeWidth=1,PositiveHeight=1,NegativeHeight=1)
    FalloffSize=0.4
    FlipHeightScaling=true
    FlipWidthScaling=false
    OptionalParentItem=(Type=None,Name="upperarm_r")
    RotationOffset=(X=120,Y=0,Z=-160)
   )
   Multiply = &RigVMFunction_MathDoubleMul(B=@n4.Result, A=@n5.OutputParam)
   > &RigUnit_ModifyTransforms(
    ItemToModify="((Item=(Type=Bone,Name=\"MSTR_flap_r\"),Transform=(Rotation=(X=0.3105,Y=0.0163,Z=-0.2312,W=0.9219),Translation=(X=-2.3376,Y=-1.681,Z=-1.1316),Scale3D=(X=1,Y=1,Z=1))),(Item=(Type=Bone,Name=\"dyn_curl_fr_21_r\"),Transform=(Rotation=(X=0.1769,Y=0.0392,Z=0.2129,W= ...(truncated)"
    WeightMinimum=0
    Weight=@Multiply.Result
   )
   Remap = &RigVMFunction_MathDoubleRemap(
    SourceMaximum=-50
    SourceMinimum=0
    TargetMaximum=1
    TargetMinimum=0
    Value=&RigVMFunction_MathQuaternionToEuler(
     RotationOrder=ZYX
     Value=&RigUnit_GetRelativeTransformForItem(Child=(Type=Bone,Name="head"), Parent=(Type=Bone,Name="spine_05"), bChildInitial=False, bParentInitial=False).RelativeTransform.Rotation
    ).Result.Z
   )
   > &RigUnit_ModifyTransforms(
    ItemToModify=((Item=(Type=Bone,Name="MSTR_hair_bk"),Transform=(Rotation=(X=0,Y=-0.0872,Z=0,W=0.9962),Translation=(X=-3,Y=0,Z=0),Scale3D=(X=0.6,Y=1,Z=1))))
    WeightMinimum=0
    Weight=@Remap.Result
   )
   Cast = &RigVMFunction_MathBoolToFloat(Value=RigVMVariableNode() {variable=isDBNO}.Value)
   > &RigUnit_ModifyTransforms(
    ItemToModify=((Item=(Type=Bone,Name="MSTR_flap_l"),Transform=(Rotation=(X=0,Y=-0.1305,Z=0,W=0.9914),Translation=(X=0,Y=0,Z=0),Scale3D=(X=1,Y=1,Z=1))))
    WeightMinimum=0
    Weight=@Cast.Result
   )
   > &RigUnit_ModifyTransforms(
    ItemToModify=((Item=(Type=Bone,Name="MSTR_flap_r"),Transform=(Rotation=(X=0,Y=0.1305,Z=0,W=0.9914),Translation=(X=0,Y=0,Z=0),Scale3D=(X=1,Y=1,Z=1))))
    WeightMinimum=0
    Weight=@Cast.Result
   )
   > n6 = &RigUnit_SphericalPoseReader(
    ActiveRegionScaleFactors=(PositiveWidth=1,NegativeWidth=2,PositiveHeight=2,NegativeHeight=2)
    ActiveRegionSize=0.1
    Debug=(bDrawDebug=false,bDraw2D=False,bDrawLocalAxes=False,DebugScale=25,DebugSegments=20,DebugThickness=0.25)
    DriverAxis=(X=1,Y=0,Z=0)
    DriverItem=(Type=Bone,Name="upperarm_l")
    FalloffRegionScaleFactors=(PositiveWidth=0.7,NegativeWidth=1,PositiveHeight=1,NegativeHeight=1)
    FalloffSize=0.3
    FlipHeightScaling=False
    FlipWidthScaling=False
    OptionalParentItem=(Type=Bone,Name="clavicle_l")
    RotationOffset=(X=0,Y=-55,Z=0)
   )
   > n7 = &RigUnit_SphericalPoseReader(
    ActiveRegionScaleFactors=(PositiveWidth=1,NegativeWidth=2,PositiveHeight=2,NegativeHeight=2)
    ActiveRegionSize=0.1
    Debug=(bDrawDebug=false,bDraw2D=False,bDrawLocalAxes=False,DebugScale=25,DebugSegments=20,DebugThickness=0.25)
    DriverAxis=(X=-1,Y=0,Z=0)
    DriverItem=(Type=Bone,Name="upperarm_r")
    FalloffRegionScaleFactors=(PositiveWidth=0.7,NegativeWidth=1,PositiveHeight=1,NegativeHeight=1)
    FalloffSize=0.3
    FlipHeightScaling=true
    FlipWidthScaling=false
    OptionalParentItem=(Type=Bone,Name="clavicle_r")
    RotationOffset=(X=0,Y=125,Z=0)
   )
   > &RigUnit_ModifyTransforms(
    ItemToModify="((Item=(Type=Bone,Name=\"MSTR_flap_l\"),Transform=(Rotation=(X=-0.034,Y=-0.0748,Z=0.1271,W=0.9885),Translation=(X=0.1005,Y=-0.6166,Z=0.4354),Scale3D=(X=1,Y=1,Z=1))),(Item=(Type=Bone,Name=\"MSTR_flap_r\"),Transform=(Rotation=(X=-0.0533,Y=0.0715,Z=0.1334,W=0.987 ...(truncated)"
    WeightMinimum=0
    Weight=&RigVMFunction_MathDoubleMul(A=@n6.OutputParam, B=@n7.OutputParam).Result
   )
   &RigUnit_ParentConstraint(
    AdvancedSettings=(InterpolationType=Average,RotationOrderForFilter=XZY)
    Child=(Type=Bone,Name="dyn_curl_fr_12_l")
    ExecutePin=()
    Filter=(TranslationFilter=(bX=true,bY=true,bZ=true),RotationFilter=(bX=True,bY=True,bZ=True),ScaleFilter=(bX=True,bY=True,bZ=True))
    Parents=((Item=(Type=Bone,Name="head"),Weight=0),(Item=(Type=Bone,Name="dyn_flap01_l"),Weight=1))
    Weight=1
   )
   &RigUnit_ParentConstraint(
    AdvancedSettings=(InterpolationType=Average,RotationOrderForFilter=XZY)
    Child=(Type=Bone,Name="dyn_curl_fr_21_r")
    ExecutePin=()
    Filter=(TranslationFilter=(bX=true,bY=true,bZ=true),RotationFilter=(bX=True,bY=True,bZ=True),ScaleFilter=(bX=True,bY=True,bZ=True))
    Parents=((Item=(Type=Bone,Name="head"),Weight=0),(Item=(Type=Bone,Name="dyn_flap01_r"),Weight=1))
    bMaintainOffset=true
    Weight=RigVMVariableNode() {variable=DistanceToRights}.Value
   )
   RigVMVariableNode() {variable=DistanceToLeft}
   RigVMVariableNode() {variable=DistanceToRights}
   RigVMVariableNode() {variable=DistanceToLeft}
  ]}
]
hierarchy [
 root {loc=<0 0 0>, rot=<0 0 0 1>}
 pelvis {parent=root, loc=<0 -0.5301 98.0571>, rot=<0.0127 0.707 -0.0127 -0.707>}
 spine_01 {parent=pelvis, loc=<6.1208 -0.9328 0>, rot=<0 0 0.034 -0.9994>}
 spine_02 {parent=spine_01, loc=<6.5997 -0.2095 0>, rot=<0 0 0.0238 -0.9997>}
 spine_03 {parent=spine_02, loc=<6.4914 -0.1237 0>, rot=<0 0 0.0149 -0.9999>}
 spine_04 {parent=spine_03, loc=<6.9598 0.0794 0>, rot=<0 0 -0.0132 -0.9999>}
 spine_05 {parent=spine_04, loc=<6.9987 0 0>, rot=<0 0 -0.0015 -1>}
 neck_01 {parent=spine_05, loc=<11.3138 0.1529 0>, rot=<0 0 -0.2003 -0.9797>}
 neck_02 {parent=neck_01, loc=<4.5627 0 0>, rot=<0 0 0.0504 -0.9987>}
 head {parent=neck_02, loc=<5.3914 0 0>, rot=<0 0 0.115 -0.9934>}
 dyn_curl_fr_01_r {parent=head, loc=<-0.0154 5.7505 6.8811>, rot=<-0.1427 -0.6637 0.7217 -0.135>}
 dyn_curl_fr_01_l {parent=head, loc=<0.3006 5.8914 -6.7312>, rot=<0.3049 0.6011 0.7287 0.1212>}
 dyn_curl_fr_01 {parent=head, loc=<8.0221 9.9855 0.7299>, rot=<-0.316 0.02 0.9476 0.0427>}
 dyn_curl_fr_11_r {parent=head, loc=<5.0177 8.1277 6.8576>, rot=<0.0614 -0.6288 0.7739 0.0446>}
 dyn_curl_fr_12_r {parent=dyn_curl_fr_11_r, loc=<2.5084 -0.3343 -0.314>, rot=<-0.6794 -0.1352 -0.2091 -0.6903>}
 MSTR_flap_r {parent=head, loc=<0.6321 0.2106 8.0373>, rot=<-0.0862 -0.7458 0.6339 0.186>}
 dyn_flap01_r {parent=MSTR_flap_r, loc=<0 0 0>, rot=<0 0 1 0>}
 dyn_flap02_r {parent=dyn_flap01_r, loc=<-5.6347 -0.0001 0.2916>, rot=<0 0 0 1>}
 dyn_race01_r {parent=dyn_flap02_r, loc=<-4.5385 -0.1396 -0.0001>, rot=<0 0.0171 -0.0028 -0.9998>}
 dyn_race02_r {parent=dyn_race01_r, loc=<-3.8278 -0.0001 0.0001>, rot=<0 0 0 1>}
 dyn_race03_r {parent=dyn_race02_r, loc=<-2.8851 0 0>, rot=<0 0 0 1>}
 MSTR_flap_l {parent=head, loc=<0.6316 0.2106 -8.0373>, rot=<-0.0862 -0.7458 -0.6339 -0.186>}
 dyn_flap01_l {parent=MSTR_flap_l, loc=<0 0 0>, rot=<0 0 0 1>}
 dyn_flap02_l {parent=dyn_flap01_l, loc=<5.634 0 -0.2913>, rot=<0 0 0 1>}
 dyn_lace01_l {parent=dyn_flap02_l, loc=<4.5387 0.1396 0>, rot=<0 0.0171 -0.0028 -0.9998>}
 dyn_lace02_l {parent=dyn_lace01_l, loc=<3.8276 0 0>, rot=<0 0 0 1>}
 dyn_lace03_l {parent=dyn_lace02_l, loc=<2.8851 0 0>, rot=<0 0 0 1>}
 dyn_hat {parent=head, loc=<12.1769 -5.6549 0>, rot=<0.8853 -0.4647 0.0107 -0.0117>}
 MSTR_hair_bk {parent=head, loc=<-0.1282 -5.0213 0.03>, rot=<0.4203 -0.5657 0.5777 -0.4117>}
 dyn_hair_bk_01 {parent=MSTR_hair_bk, loc=<4.3501 -0.0931 1.2214>, rot=<0.7061 0.1875 0.0372 -0.6818>}
 dyn_hair_bk_02 {parent=dyn_hair_bk_01, loc=<3.0451 0.0193 -0.0355>, rot=<-0.7059 0.1464 0.0419 -0.6918>}
 dyn_hair_bk_01_r {parent=MSTR_hair_bk, loc=<4.8388 -3.7143 0.1897>, rot=<0.0568 -0.0052 -0.2377 -0.9697>}
 dyn_hair_bk_01_l {parent=MSTR_hair_bk, loc=<4.7227 0.5845 -2.1614>, rot=<0.0063 0.2392 0 -0.9709>}
 dyn_hair_bk_02_l {parent=MSTR_hair_bk, loc=<1.6001 5.1839 -3.1865>, rot=<0.6394 -0.1648 0.1229 -0.7409>}
 dyn_hair_bk_02_r {parent=MSTR_hair_bk, loc=<2.0685 -3.6059 -2.7002>, rot=<-0.0968 -0.0762 -0.1617 -0.9791>}
 dyn_smallHatBelt {parent=head, loc=<8.069 -2.2539 -8.6809>, rot=<-0.5713 0.5489 0.6093 -0.0314>}
 dyn_curl_fr_12_l {parent=head, loc=<-4.2539 3.1163 -6.3792>, rot=<-0.113 -0.7709 -0.6204 -0.0899>}
 dyn_curl_fr_13_l {parent=dyn_curl_fr_12_l, loc=<3.5073 -0.3929 0.128>, rot=<-0.0782 0.0758 -0.146 -0.9833>}
 dyn_curl_fr_21_r {parent=head, loc=<-2.1427 2.0177 6.4481>, rot=<-0.0723 -0.7121 0.6982 0.0106>}
 dyn_curl_fr_22_r {parent=dyn_curl_fr_21_r, loc=<5.4276 -0.0961 -0.6824>, rot=<0.0566 -0.2366 0.1962 -0.9499>}
 dyn_curl_fr_31_r {parent=head, loc=<7.3504 10.2778 4.8402>, rot=<0.4965 -0.4945 0.5131 0.4957>}
 clavicle_l {parent=spine_05, loc=<5.4831 -0.1053 -3.2946>, rot=<0.0465 -0.7289 -0.009 -0.6829>}
 upperarm_l {parent=clavicle_l, loc=<9.8658 -1.2136 -0.4105>, rot=<0.0164 -0.3245 -0.0269 -0.9454>}
 upperarm_twist_01_l {parent=upperarm_l, loc=<7.2135 -0.1134 0>, rot=<0 0 0 1>}
 clavicle_r {parent=spine_05, loc=<5.4831 -0.1053 3.2946>, rot=<0.7289 0.0465 -0.6829 0.009>}
 upperarm_r {parent=clavicle_r, loc=<-9.8658 1.2136 0.4105>, rot=<0.0164 -0.3245 -0.0269 -0.9454>}
 upperarm_twist_01_r {parent=upperarm_r, loc=<-7.2135 0.1134 0>, rot=<0 0 0 -1>}
 def_backBeltAlt {parent=spine_03, loc=<0.6145 -13.0804 -9.9103>, rot=<0.0387 -0.706 -0.0387 -0.706>}
]
variables [
 DistanceToLeft {type=double, default=0}
 DistanceToRights {type=double, default=0}
 Remap_3_Result {type=double, default=0}
 X Rot {type=double, default=1}
 Y Rot {type=double, default=1}
 Z Rot {type=double, default=1}
 isDBNO {type=bool, default=False}
]
