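# Physics asset for the face accessory of the female medium OmelettePop cosmetic character, covering left and right flaps with dangling laces.
# Each side chains a flap capsule to three lace capsules (right-side bones are named "race") through swing-limited, twist-locked constraints.
# Spheres and capsules on the head, neck, upper spine and upper arms supply body shapes, with no constraints linking them to the chains.
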PhysicsAsset /BRCosmetics/Characters/Player/Female/Medium/Bodies/F_MED_OmelettePop/Meshes/Parts/F_MED_OmelettePop_FaceAcc_Physics
bodies {
 dyn_flap02_l [{
   shapeType=Capsule
   length=5.1866
   radius=1.3647
   shapeName=dyn_flap02_l_capsule
   center=<1.3845 0.2534 0.0991>
   rotation=<rot -4.1606 -90 2.6316>
  }]
 dyn_flap02_r [{
   shapeType=Capsule
   length=5.2072
   radius=1.3441
   shapeName=dyn_flap02_r_capsule
   center=<-1.3845 -0.1574 -0.0922>
   rotation=<rot -4.1606 -90 -177.3684>
  }]
 dyn_lace01_l [{
   shapeType=Capsule
   length=2.2848
   radius=0.9682
   shapeName=dyn_lace01_l_capsule
   center=<1.7144 0.0229 -0.006>
   rotation=<rot 0 89.4607 -89.4689>
  }]
 dyn_lace02_l [{shapeType=Capsule, length=2.4388, radius=1.0821, shapeName=dyn_lace02_l_capsule, center=<1.4211 0 0>, rotation=<rot 0 -90 89.9998>}]
 dyn_lace03_l [{
   shapeType=Capsule
   length=3.6158
   radius=1.0877
   shapeName=dyn_lace03_l_capsule
   center=<2.5748 0.0146 0.1991>
   rotation=<rot 0 89.656 -87.1574>
  }]
 dyn_race01_r [{
   shapeType=Capsule
   length=2.2848
   radius=0.9682
   shapeName=dyn_lace01_l_capsule
   center=<-1.8578 0.0565 -0.0391>
   rotation=<rot 0 89.4607 -89.4689>
  }]
 dyn_race02_r [{shapeType=Capsule, length=2.4388, radius=1.0821, shapeName=dyn_lace02_l_capsule, center=<-1.2268 0 0>, rotation=<rot 0 -90 89.9998>}]
 dyn_race03_r [{
   shapeType=Capsule
   length=3.5676
   radius=1.1359
   shapeName=dyn_race03_r_capsule
   center=<-2.462 -0.0146 -0.1963>
   rotation=<rot 0 89.6834 -87.2199>
  }]
 head [{shapeType=Sphere, radius=8.6292, shapeName=head_sphere, center=<2.9934 1.9908 0.0168>}]
 neck_01 [{shapeType=Capsule, length=2.8007, radius=6.1918, shapeName=neck_01_capsule, center=<2.9846 1.6632 0>, rotation=<rot 0 -90.0002 90>}]
 spine_04 [{shapeType=Sphere, radius=6.7802, shapeName=spine_04_sphere, center=<6.2206 10.4171 3.283>}]
 spine_05 [{shapeType=Capsule, length=20.1046, radius=9.78, shapeName=spine_05_capsule, center=<0 -0.7103 0>, rotation=<rot 0 -90 0>}]
 upperarm_l [{
   shapeType=Capsule
   length=21.4438
   radius=7.5287
   shapeName=upperarm_r_capsule
   center=<12.877 0.0017 -1.421>
   rotation=<rot -89.595 178.9581 1.0485>
  }]
 upperarm_r [{
   shapeType=Capsule
   length=20.0292
   radius=7.5287
   shapeName=upperarm_r_capsule
   center=<-12.877 -0.0017 1.421>
   rotation=<rot -89.595 178.9581 1.0485>
  }]
}
constraints [
 {bone1Name=dyn_lace02_l, bone2Name=dyn_lace01_l, swing1LimitDegrees=60, swing2LimitDegrees=60, twistLimitDegrees=5, twistMotion=Locked}
 {bone1Name=dyn_lace03_l, bone2Name=dyn_lace02_l, swing1LimitDegrees=30, swing2LimitDegrees=30, twistLimitDegrees=5, twistMotion=Locked}
 {bone1Name=dyn_race02_r, bone2Name=dyn_race01_r, swing1LimitDegrees=60, swing2LimitDegrees=60, twistLimitDegrees=5, twistMotion=Locked}
 {bone1Name=dyn_race03_r, bone2Name=dyn_race02_r, swing1LimitDegrees=30, swing2LimitDegrees=30, twistLimitDegrees=5, twistMotion=Locked}
 {bone1Name=dyn_lace01_l, bone2Name=dyn_flap02_l, swing1LimitDegrees=45, swing2LimitDegrees=45, twistLimitDegrees=45, twistMotion=Locked}
 {bone1Name=dyn_race01_r, bone2Name=dyn_flap02_r, swing1LimitDegrees=45, swing2LimitDegrees=45, twistLimitDegrees=45, twistMotion=Locked}
]
